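# Experimental Niagara system for the Black Flame vehicle boost cosmetic, drawing flame ribbon trails from filtered skeletal sockets.
# The aura emitter spawns backward-moving ribbons, offset by game mode, with three renderers and decaying inherited movement.
# Painted-variant overrides supply per-color curves to five color channels, several scaled by user emissive and alpha powers.
# Age-driven sine offsets make ribbons swirl, mirrored by spawn group; a camera-aware Base1 emitter is disabled.
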NiagaraSystem /VehicleCosmetics/Boosts/Boost_BlackFlame/FX/NS_BlackFlame_Experiment
dependencies {
 usedDataInterfaces [
  &NiagaraDataInterfaceColorCurve
  &NiagaraDataInterfaceCurve
  &NiagaraDataInterfaceArrayFloat
  &NiagaraDataInterfaceCamera
  &NiagaraDataInterfaceSkeletalMesh
 ]
 usedDynamicInputs []
 usedModules [
  &self:SystemSpawnScript.SystemScriptSource.SystemScriptGraph.NiagaraNodeAssignment_0.SetVariables_#1
  &SystemState
 ]
}
emitters {
 NE_BlacFlame_Aura {
  emitterUpdateScript [
   &EmitterState()
   &VC_FilteredSocketCount()
   &VC_RateQualityHandler(Epic Spawn Rate=30, High Spawn Rate=30, Low Spawn Rate=30, Use Custom Rate?=true)
   &VC_SpawnRibbon_v1(Number of Ribbons=$Emitter.PortCount, Spawn Time=-1, SpawnRate=$Emitter.Spawn Rate)
  ]
  particleSpawnScript [
   &VC_PaintedVariantsOverride(
    Black=<rgba 0.01 0.01 0.01 1.25>
    BurntSienna=<rgba 0.0955 0.0295 0.0118 1>
    Cobalt=<rgba 0 0.0066 0.3559 1>
    Crimson=<rgba 0.0816 0 0.0022 1.5>
    ForestGreen=<rgba 0 0.0608 0.0031 1.5>
    Grey=<rgba 0.1649 0.1649 0.1649 0.4435>
    Lime=<rgba 0.0334 0.1565 0.025 1.5>
    Orange=<rgba 0.2205 0.0372 0 1.5>
    Pink=<rgba 0.3003 0.0771 0.2854 1>
    Purple=<rgba 0.0223 0 0.1101 2>
    R1 Black:LinearColor=<rgba 0.0469 0.0469 0.0469 1>
    R1 Cobalt:LinearColor=<rgba 0 0.0185 1 1>
    R1 ForestGreen:LinearColor=<rgba 0 0.2319 0.004 1>
    R1 Grey:LinearColor=<rgba 0.1267 0.1267 0.1267 1>
    R1 Orange:LinearColor=<rgba 0.6724 0.2346 0 1>
    R1 Pink:LinearColor=<rgba 0.5052 0.1552 0.3779 1>
    R1 Purple:LinearColor=<rgba 0.0699 0 0.2552 1>
    R1 SkyBlue:LinearColor=<rgba 0.2058 0.4258 0.5365 1>
    Saffron=<rgba 0.1372 0.1052 0 1>
    SkyBlue=<rgba 0 0.0765 0.1719 1.5>
    TitaniumWhite=<rgba 0.1406 0.1406 0.1406 1>
    Use R1 Painted Overrides=true
   )
   &VC_InitializeRibbon_v1(
    Color=$Particles.Color
    Custom Offset=true
    Enable Skeletal Position=true
    Lifetime=0.8
    Local Vector Offset=/VehicleCosmetics/FX/Niagara/Modules/GameMode/VC_SetVectorByGameModeID(
     SameValueForLockerAndBR?:NiagaraBool=false
     Locker Vector:Vector3f=<30 0 2.5>
     Battle Royale Vector:Vector3f=<180 0 2.5>
     Rocket Racing Vector:Vector3f=<180 0 2.5>
     GameModeID:NiagaraInt32=$System.GameModeID
    )
    Offset Amount=1
    Ribbon Width=70
    Write Material Random=false
   )
   &AddVelocity(Velocity=<-1750 0 0>) {module=AddVelocity001}
  ]
  particleUpdateScript [
   &VC_PaintedVariantsOverride(
    R1 Black:LinearColor=&ColorFromCurve(ColorCurve=DataInterface() {propertyValues {
       AlphaCurve {
        defaultValue=0
        keys [<at 0 0>, {arriveTangent=15.9916, interpMode=RCIM_Cubic, leaveTangent=15.9916, time=0.1339, value=8}, <at 0.6253 10>, <at 0.9771 0>]
       }
       BlueCurve {keys [<at 1 0.1094 RCIM_Cubic>]}
       GreenCurve {keys [<at 1 0.1094 RCIM_Cubic>]}
       RedCurve {keys [<at 1 0.1094 RCIM_Cubic>]}
      }})
    R1 BurntSienna:LinearColor=&ColorFromCurve(ColorCurve=DataInterface() {propertyValues {
       AlphaCurve {
        defaultValue=0
        keys [<at 0 0>, {arriveTangent=15.9916, interpMode=RCIM_Cubic, leaveTangent=15.9916, time=0.1756, value=8}, <at 0.6253 10>, <at 0.9771 0>]
       }
       BlueCurve {keys [<at 1 0.0686 RCIM_Cubic>]}
       GreenCurve {keys [<at 1 0.0858 RCIM_Cubic>]}
       RedCurve {keys [<at 1 0.1667 RCIM_Cubic>]}
      }})
    R1 Cobalt:LinearColor=&ColorFromCurve(ColorCurve=DataInterface() {propertyValues {
       AlphaCurve {
        defaultValue=0
        keys [<at 0 0>, {arriveTangent=15.9916, interpMode=RCIM_Cubic, leaveTangent=15.9916, time=0.1756, value=8}, <at 0.6253 10>, <at 0.9771 0>]
       }
       BlueCurve {keys [<at 1 0.3802 RCIM_Cubic>]}
       GreenCurve {keys [<at 1 0.0931 RCIM_Cubic>]}
       RedCurve {keys [<at 1 0.0931 RCIM_Cubic>]}
      }})
    R1 Crimson:LinearColor=&ColorFromCurve(ColorCurve=DataInterface() {propertyValues {
       AlphaCurve {
        defaultValue=0
        keys [<at 0 0>, {arriveTangent=15.9916, interpMode=RCIM_Cubic, leaveTangent=15.9916, time=0.1756, value=8}, <at 0.6253 10>, <at 0.9771 0>]
       }
       BlueCurve {keys [<at 1 0.0131 RCIM_Cubic>]}
       GreenCurve {keys [<at 1 0 RCIM_Cubic>]}
       RedCurve {keys [<at 1 0.2348 RCIM_Cubic>]}
      }})
    R1 DefaultColor=&ColorFromCurve(ColorCurve=DataInterface() {propertyValues {
       AlphaCurve {keys [<at 0 0>, {arriveTangent=15.9916, interpMode=RCIM_Cubic, leaveTangent=15.9916, time=0.1039, value=8}, <at 0.6253 10>, <at 0.9771 0>]}
       BlueCurve {keys [<at 0.6966 1>]}
       GreenCurve {keys [<at 0.6966 0.5>]}
       RedCurve {keys [<at 0.6966 0>]}
      }})
    R1 ForestGreen:LinearColor=&ColorFromCurve(ColorCurve=DataInterface() {propertyValues {
       AlphaCurve {
        defaultValue=0
        keys [<at 0 0>, {arriveTangent=15.9916, interpMode=RCIM_Cubic, leaveTangent=15.9916, time=0.1756, value=8}, <at 0.6253 10>, <at 0.9771 0>]
       }
       BlueCurve {keys [<at 1 0.0198 RCIM_Cubic>]}
       GreenCurve {keys [<at 1 0.2031 RCIM_Cubic>]}
       RedCurve {keys [<at 1 0 RCIM_Cubic>]}
      }})
    R1 Grey:LinearColor=&ColorFromCurve(ColorCurve=DataInterface() {propertyValues {
       AlphaCurve {
        defaultValue=0
        keys [<at 0 0>, {arriveTangent=15.9916, interpMode=RCIM_Cubic, leaveTangent=15.9916, time=0.1756, value=8}, <at 0.6253 10>, <at 0.9771 0>]
       }
       BlueCurve {keys [<at 1 0.3542 RCIM_Cubic>]}
       GreenCurve {keys [<at 1 0.3542 RCIM_Cubic>]}
       RedCurve {keys [<at 1 0.3542 RCIM_Cubic>]}
      }})
    R1 Lime:LinearColor=&ColorFromCurve(ColorCurve=DataInterface() {propertyValues {
       AlphaCurve {
        defaultValue=0
        keys [<at 0 0>, {arriveTangent=15.9916, interpMode=RCIM_Cubic, leaveTangent=15.9916, time=0.1756, value=8}, <at 0.6253 10>, <at 0.9771 0>]
       }
       BlueCurve {keys [<at 1 0.0865 RCIM_Cubic>]}
       GreenCurve {keys [<at 1 0.2865 RCIM_Cubic>]}
       RedCurve {keys [<at 1 0.1697 RCIM_Cubic>]}
      }})
    R1 Orange:LinearColor=&ColorFromCurve(ColorCurve=DataInterface() {propertyValues {
       AlphaCurve {
        defaultValue=0
        keys [<at 0 0>, {arriveTangent=15.9916, interpMode=RCIM_Cubic, leaveTangent=15.9916, time=0.1756, value=8}, <at 0.6253 10>, <at 0.9771 0>]
       }
       BlueCurve {keys [<at 1 0 RCIM_Cubic>]}
       GreenCurve {keys [<at 1 0.2961 RCIM_Cubic>]}
       RedCurve {keys [<at 1 0.9271 RCIM_Cubic>]}
      }})
    R1 Pink:LinearColor=&ColorFromCurve(ColorCurve=DataInterface() {propertyValues {
       AlphaCurve {
        defaultValue=0
        keys [<at 0 0>, {arriveTangent=15.9916, interpMode=RCIM_Cubic, leaveTangent=15.9916, time=0.1756, value=8}, <at 0.6253 10>, <at 0.9771 0>]
       }
       BlueCurve {keys [<at 1 0.2004 RCIM_Cubic>]}
       GreenCurve {keys [<at 1 0 RCIM_Cubic>]}
       RedCurve {keys [<at 1 0.2188 RCIM_Cubic>]}
      }})
    R1 Purple:LinearColor=&ColorFromCurve(ColorCurve=DataInterface() {propertyValues {
       AlphaCurve {
        defaultValue=0
        keys [<at 0 0>, {arriveTangent=15.9916, interpMode=RCIM_Cubic, leaveTangent=15.9916, time=0.1756, value=8}, <at 0.6253 10>, <at 0.9771 0>]
       }
       BlueCurve {keys [<at 1 0.375 RCIM_Cubic>]}
       GreenCurve {keys [<at 1 0 RCIM_Cubic>]}
       RedCurve {keys [<at 1 0.1484 RCIM_Cubic>]}
      }})
    R1 Saffron:LinearColor=&ColorFromCurve(ColorCurve=DataInterface() {propertyValues {
       AlphaCurve {
        defaultValue=0
        keys [<at 0 0>, {arriveTangent=15.9916, interpMode=RCIM_Cubic, leaveTangent=15.9916, time=0.1756, value=8}, <at 0.6253 10>, <at 0.9771 0>]
       }
       BlueCurve {keys [<at 1 0 RCIM_Cubic>]}
       GreenCurve {keys [<at 1 0.2395 RCIM_Cubic>]}
       RedCurve {keys [<at 1 0.2812 RCIM_Cubic>]}
      }})
    R1 SkyBlue:LinearColor=&ColorFromCurve(ColorCurve=DataInterface() {propertyValues {
       AlphaCurve {
        defaultValue=0
        keys [<at 0 0>, {arriveTangent=15.9916, interpMode=RCIM_Cubic, leaveTangent=15.9916, time=0.1756, value=8}, <at 0.6253 10>, <at 0.9771 0>]
       }
       BlueCurve {keys [<at 1 0.3802 RCIM_Cubic>]}
       GreenCurve {keys [<at 1 0.1905 RCIM_Cubic>]}
       RedCurve {keys [<at 1 0 RCIM_Cubic>]}
      }})
    R1 TitaniumWhite:LinearColor=&ColorFromCurve(ColorCurve=DataInterface() {propertyValues {
       AlphaCurve {
        defaultValue=0
        keys [<at 0 0>, {arriveTangent=15.9916, interpMode=RCIM_Cubic, leaveTangent=15.9916, time=0.1756, value=8}, <at 0.6253 10>, <at 0.9771 0>]
       }
       BlueCurve {keys [<at 1 0.6406 RCIM_Cubic>]}
       GreenCurve {keys [<at 1 0.6406 RCIM_Cubic>]}
       RedCurve {keys [<at 1 0.6406 RCIM_Cubic>]}
      }})
    R2 Black:LinearColor=&ColorFromCurve(ColorCurve=DataInterface() {propertyValues {
       AlphaCurve {keys [<at 0 7>,<at 0.2232 1>,<at 0.3261 0.05>,<at 0.5495 0.2>,<at 0.7007 0.2>,<at 0.8474 0 RCIM_Cubic>]}
       BlueCurve {keys [<at 0 0.0194 RCIM_Cubic>,<at 0.1492 0.0333>,<at 0.2681 0.1104>,<at 0.4863 0.9274>,<at 0.6088 0.0305>]}
       GreenCurve {keys [<at 0 0.0194 RCIM_Cubic>,<at 0.1492 0.0333>,<at 0.2681 0.1104>,<at 0.4863 0.9274>,<at 0.6088 0.0305>]}
       RedCurve {keys [<at 0 0.0194 RCIM_Cubic>,<at 0.1492 0.0333>,<at 0.2681 0.1104>,<at 0.4863 0.9274>,<at 0.6088 0.0305>]}
      }})
    R2 BurntSienna:LinearColor=&ColorFromCurve(ColorCurve=DataInterface() {propertyValues {
       AlphaCurve {keys [<at 0 4>,<at 0.1907 1>,<at 0.3261 0.05>,<at 0.593 0.2>,<at 0.8474 0 RCIM_Cubic>]}
       BlueCurve {keys [<at 0 0.008 RCIM_Cubic>,<at 0.2839 0.0202>,<at 0.4863 0.2907>,<at 0.5956 0.008>]}
       GreenCurve {keys [<at 0 0.0203 RCIM_Cubic>,<at 0.2839 0.0323>,<at 0.4863 0.4394>,<at 0.5956 0.0203>]}
       RedCurve {keys [<at 0 0.0648 RCIM_Cubic>,<at 0.2839 0.0762>,<at 0.4863 0.9792>,<at 0.5956 0.0648>]}
      }})
    R2 Cobalt:LinearColor=&ColorFromCurve(ColorCurve=DataInterface() {propertyValues {
       AlphaCurve {keys [<at 0 7>,<at 0.1907 1>,<at 0.3261 0.05>,<at 0.5495 0.6>,<at 0.7007 0.2>,<at 0.8474 0 RCIM_Cubic>]}
       BlueCurve {keys [<at 0 0.6458 RCIM_Cubic>,<at 0.2575 0.9985>,<at 0.4863 1>,<at 0.6115 0.8177>]}
       GreenCurve {keys [<at 0 0.0119 RCIM_Cubic>,<at 0.2575 0.0367>,<at 0.4863 0.2997>,<at 0.6115 0.0527>]}
       RedCurve {keys [<at 0 0 RCIM_Cubic>,<at 0.2575 0.0186>,<at 0.4863 0.2865>,<at 0.6115 0.0383>]}
      }})
    R2 Crimson:LinearColor=&ColorFromCurve(ColorCurve=DataInterface() {propertyValues {
       AlphaCurve {keys [<at 0 7>,<at 0.1907 1>,<at 0.3261 0.05>,<at 0.5495 0.2>,<at 0.7007 0.2>,<at 0.8474 0 RCIM_Cubic>]}
       BlueCurve {keys [<at 0 0 RCIM_Cubic>,<at 0.1492 0.0104>,<at 0.4863 0.2995>,<at 0.6458 0.0894>]}
       GreenCurve {keys [<at 0 0 RCIM_Cubic>,<at 0.1492 0.0104>,<at 0.4863 0.2995>,<at 0.6458 0.0894>]}
       RedCurve {keys [<at 0 0.4323 RCIM_Cubic>,<at 0.1492 0.9992>,<at 0.4863 0.9274>,<at 0.6458 0.993>]}
      }})
    R2 DefaultColor:LinearColor=&ColorFromCurve(ColorCurve=DataInterface() {propertyValues {
       AlphaCurve {keys [<at 0 1>,<at 0.0914 0.2857>,<at 0.308 0.05>,<at 0.5592 0.2>,<at 0.8632 0>]}
       BlueCurve {keys [<at 0.0009 0>,<at 0.2668 -0.2829>,<at 0.5114 0.8849>,<at 0.619 1>,<at 0.682 0.383>]}
       GreenCurve {keys [<at 0.0009 0.0767>,<at 0.2742 -0.1994>,<at 0.5114 0.8849>,<at 0.619 1>,<at 0.682 0.2585>]}
       RedCurve {keys [<at 0.0009 3.2194>,<at 0.0016 3.2194>,<at 0.4479 0.8849>,<at 0.5114 0.8849>,<at 0.619 1>]}
      }})
    R2 ForestGreen:LinearColor=&ColorFromCurve(ColorCurve=DataInterface() {propertyValues {
       AlphaCurve {keys [<at 0 6>,<at 0.1907 1>,<at 0.3261 0.05>,<at 0.5495 0.2>,<at 0.7007 0.2>,<at 0.8474 0 RCIM_Cubic>]}
       BlueCurve {keys [<at 0 0.0172 RCIM_Cubic>,<at 0.2179 0.0322>,<at 0.4863 0.4656>,<at 0.6537 0.1027>]}
       GreenCurve {keys [<at 0 0.3438 RCIM_Cubic>,<at 0.2179 0.245>,<at 0.4863 0.875>,<at 0.6537 0.2987>]}
       RedCurve {keys [<at 0 0 RCIM_Cubic>,<at 0.2179 0.021>,<at 0.4863 0.4466>,<at 0.6537 0.0924>]}
      }})
    R2 Grey:LinearColor=&ColorFromCurve(ColorCurve=DataInterface() {propertyValues {
       AlphaCurve {keys [<at 0 4>,<at 0.1907 1>,<at 0.3261 0.05>,<at 0.5495 0.2>,<at 0.7007 0.2>,<at 0.8474 0 RCIM_Cubic>]}
       BlueCurve {keys [<at 0 0.0692 RCIM_Cubic>,<at 0.2285 0.1336>,<at 0.4863 0.9274>,<at 0.6405 0.132>]}
       GreenCurve {keys [<at 0 0.0781 RCIM_Cubic>,<at 0.2285 0.1336>,<at 0.4863 0.9274>,<at 0.6405 0.132>]}
       RedCurve {keys [<at 0 0.0767 RCIM_Cubic>,<at 0.2285 0.1406>,<at 0.4863 0.9274>,<at 0.6405 0.1354>]}
      }})
    R2 Lime:LinearColor=&ColorFromCurve(ColorCurve=DataInterface() {propertyValues {
       AlphaCurve {keys [<at 0 6>,<at 0.1907 1>,<at 0.3261 0.05>,<at 0.5495 0.2>,<at 0.8122 0 RCIM_Cubic>]}
       BlueCurve {keys [<at 0 0 RCIM_Cubic>,<at 0.2179 0.0459>,<at 0.4863 0.9274>,<at 0.6299 0.0724>]}
       GreenCurve {keys [<at 0 0.6875 RCIM_Cubic>,<at 0.2179 0.4635>,<at 0.4863 0.9274>,<at 0.6299 0.7708>]}
       RedCurve {keys [<at 0 0.3447 RCIM_Cubic>,<at 0.2179 0.2553>,<at 0.4863 0.9274>,<at 0.6299 0.4226>]}
      }})
    R2 Orange:LinearColor=&ColorFromCurve(ColorCurve=DataInterface() {propertyValues {
       AlphaCurve {keys [<at 0 7>,<at 0.1907 1>,<at 0.3261 0.05>,<at 0.5495 0.2>,<at 0.7007 0.2>,<at 0.8474 0 RCIM_Cubic>]}
       BlueCurve {keys [<at 0 0 RCIM_Cubic>,<at 0.2364 0.0598>,<at 0.4863 0.4263>,<at 0.6217 0.1191>]}
       GreenCurve {keys [<at 0 0.2834 RCIM_Cubic>,<at 0.2364 0.2792>,<at 0.4863 0.5719>,<at 0.6217 0.3121>]}
       RedCurve {keys [<at 0 0.8125 RCIM_Cubic>,<at 0.2364 0.6889>,<at 0.4863 0.8438>,<at 0.6217 0.6724>]}
      }})
    R2 Pink:LinearColor=&ColorFromCurve(ColorCurve=DataInterface() {propertyValues {
       AlphaCurve {keys [<at 0 7>,<at 0.1907 1>,<at 0.3261 0.05>,<at 0.5495 0.2>,<at 0.7007 0.2>,<at 0.8474 0 RCIM_Cubic>]}
       BlueCurve {keys [<at 0 0.3419 RCIM_Cubic>,<at 0.2707 0.3564>,<at 0.4863 0.7216>,<at 0.6141 0.5634>]}
       GreenCurve {keys [<at 0 0.0612 RCIM_Cubic>,<at 0.2707 0.1489>,<at 0.4863 0.5789>,<at 0.6141 0.2536>]}
       RedCurve {keys [<at 0 0.5029 RCIM_Cubic>,<at 0.2707 0.5614>,<at 0.4863 0.855>,<at 0.6141 0.8542>]}
      }})
    R2 Purple:LinearColor=&ColorFromCurve(ColorCurve=DataInterface() {propertyValues {
       AlphaCurve {keys [<at 0 7>,<at 0.1907 1>,<at 0.3261 0.05>,<at 0.5495 0.2>,<at 0.7007 0.2>,<at 0.8474 0 RCIM_Cubic>]}
       BlueCurve {keys [<at 0 0.8281 RCIM_Cubic>,<at 0.2919 0.6836>,<at 0.4863 0.8957>,<at 0.6141 0.9565>]}
       GreenCurve {keys [<at 0 0 RCIM_Cubic>,<at 0.2919 0.0563>,<at 0.4863 0.2791>,<at 0.6141 0.1097>]}
       RedCurve {keys [<at 0 0.1924 RCIM_Cubic>,<at 0.2919 0.3349>,<at 0.4863 0.5023>,<at 0.6141 0.3436>]}
      }})
    R2 Saffron:LinearColor=&ColorFromCurve(ColorCurve=DataInterface() {propertyValues {
       AlphaCurve {keys [<at 0 4>,<at 0.1907 1>,<at 0.3261 0.05>,<at 0.5495 0.2>,<at 0.8175 0 RCIM_Cubic>]}
       BlueCurve {keys [<at 0 0 RCIM_Cubic>,<at 0.2697 0.0046>,<at 0.4863 0.3188>,<at 0.6748 0.0798>]}
       GreenCurve {keys [<at 0 0.6105 RCIM_Cubic>,<at 0.2697 0.6881>,<at 0.4863 0.9274>,<at 0.6748 0.6052>]}
       RedCurve {keys [<at 0 0.7969 RCIM_Cubic>,<at 0.2697 0.8967>,<at 0.4863 0.8744>,<at 0.6748 0.7656>]}
      }})
    R2 SkyBlue:LinearColor=&ColorFromCurve(ColorCurve=DataInterface() {propertyValues {
       AlphaCurve {keys [<at 0 7>,<at 0.1907 1>,<at 0.3261 0.05>,<at 0.5495 0.2>,<at 0.8096 0 RCIM_Cubic>]}
       BlueCurve {keys [<at 0 0.5333 RCIM_Cubic>,<at 0.2496 0.5525>,<at 0.4863 0.9323>,<at 0.6141 0.9323>]}
       GreenCurve {keys [<at 0 0.3564 RCIM_Cubic>,<at 0.2496 0.3792>,<at 0.4863 0.8083>,<at 0.6141 0.6988>]}
       RedCurve {keys [<at 0 0 RCIM_Cubic>,<at 0.2496 0.0875>,<at 0.4863 0.5584>,<at 0.6141 0.2282>]}
      }})
    R2 TitaniumWhite:LinearColor=&ColorFromCurve(ColorCurve=DataInterface() {propertyValues {
       AlphaCurve {keys [<at 0 4>,<at 0.1907 1>,<at 0.3261 0.05>,<at 0.5495 0.2>,<at 0.7007 0.2>,<at 0.8175 0 RCIM_Cubic>]}
       BlueCurve {keys [<at 0 0.1846 RCIM_Cubic>,<at 0.2285 0.1336>,<at 0.4863 0.9274>,<at 0.6405 0.3248>]}
       GreenCurve {keys [<at 0 0.2083 RCIM_Cubic>,<at 0.2285 0.1336>,<at 0.4863 0.9274>,<at 0.6405 0.3248>]}
       RedCurve {keys [<at 0 0.2046 RCIM_Cubic>,<at 0.2285 0.1406>,<at 0.4863 0.9274>,<at 0.6405 0.3333>]}
      }})
    R3 Black:LinearColor=&ColorFromCurve(ColorCurve=DataInterface() {propertyValues {
       AlphaCurve {
        defaultValue=0
        keys [
         {tangentMode=RCTM_User, time=0.0672, value=0}
         {
          arriveTangent=13.1575
          arriveTangentWeight=0.0306
          interpMode=RCIM_Cubic
          leaveTangent=13.1575
          leaveTangentWeight=0.2107
          tangentMode=RCTM_User
          time=0.1591
          value=7
         }
         {tangentMode=RCTM_User, time=0.5992, value=7}
         <at 0.7913 0>
        ]
       }
       BlueCurve {keys [<at 0.6966 0.3043>]}
       GreenCurve {keys [<at 0.6966 0.3043>]}
       RedCurve {keys [<at 0.6966 0.3043>]}
      }})
    R3 BurntSienna:LinearColor=&ColorFromCurve(ColorCurve=DataInterface() {propertyValues {
       AlphaCurve {defaultValue=0, keys [{arriveTangent=13.1575, interpMode=RCIM_Cubic, time=0.0863, value=7}, <at 0.5992 7>, <at 0.7913 0>]}
       BlueCurve {keys [<at 0.6966 0.2729>]}
       GreenCurve {keys [<at 0.6966 0.3693>]}
       RedCurve {keys [<at 0.6966 0.6719>]}
      }})
    R3 Cobalt:LinearColor=&ColorFromCurve(ColorCurve=DataInterface() {propertyValues {
       AlphaCurve {
        defaultValue=0
        keys [
         {tangentMode=RCTM_User, time=0.0672, value=0}
         {arriveTangent=13.1575, interpMode=RCIM_Cubic, leaveTangent=13.1575, tangentMode=RCTM_User, time=0.1591, value=7}
         {tangentMode=RCTM_User, time=0.5992, value=7}
         <at 0.7913 0>
        ]
       }
       BlueCurve {keys [<at 0.6966 1.5>]}
       GreenCurve {keys [<at 0.6966 0.0598>]}
       RedCurve {keys [<at 0.6966 0>]}
      }})
    R3 Crimson:LinearColor=&ColorFromCurve(ColorCurve=DataInterface() {propertyValues {
       AlphaCurve {
        defaultValue=0
        keys [
         {tangentMode=RCTM_User, time=0.0672, value=0}
         {arriveTangent=13.1575, interpMode=RCIM_Cubic, leaveTangent=13.1575, tangentMode=RCTM_User, time=0.1591, value=7}
         {tangentMode=RCTM_User, time=0.5992, value=7}
         <at 0.7913 0>
        ]
       }
       BlueCurve {keys [<at 0.6966 0>]}
       GreenCurve {keys [<at 0.6966 0>]}
       RedCurve {keys [<at 0.6966 1>]}
      }})
    R3 DefaultColor:LinearColor=&ColorFromCurve(ColorCurve=DataInterface() {propertyValues {
       AlphaCurve {keys [
         {tangentMode=RCTM_User, time=0, value=0}
         {
          arriveTangent=13.1575
          arriveTangentWeight=0.0306
          interpMode=RCIM_Cubic
          leaveTangent=13.1575
          leaveTangentWeight=0.2538
          tangentMode=RCTM_User
          time=0.1591
          value=7
         }
         {tangentMode=RCTM_User, time=0.5992, value=7}
         <at 0.7913 0>
         <at 0.9206 0>
        ]}
       BlueCurve {keys [<at 0.4821 1>]}
       GreenCurve {keys [<at 0.4821 0.0657>]}
       RedCurve {keys [<at 0.4821 0>]}
      }})
    R3 ForestGreen:LinearColor=&ColorFromCurve(ColorCurve=DataInterface() {propertyValues {
       AlphaCurve {
        defaultValue=0
        keys [
         {tangentMode=RCTM_User, time=0.0672, value=0}
         {arriveTangent=13.1575, interpMode=RCIM_Cubic, leaveTangent=13.1575, tangentMode=RCTM_User, time=0.1591, value=7}
         {tangentMode=RCTM_User, time=0.5992, value=7}
         <at 0.7913 0>
        ]
       }
       BlueCurve {keys [<at 0.6966 0.3387>]}
       GreenCurve {keys [<at 0.6966 1.5>]}
       RedCurve {keys [<at 0.6966 0>]}
      }})
    R3 Grey:LinearColor=&ColorFromCurve(ColorCurve=DataInterface() {propertyValues {
       AlphaCurve {
        defaultValue=0
        keys [
         {tangentMode=RCTM_User, time=0.0672, value=0}
         {arriveTangent=13.1575, interpMode=RCIM_Cubic, leaveTangent=13.1575, tangentMode=RCTM_User, time=0.1591, value=7}
         {tangentMode=RCTM_User, time=0.5992, value=7}
         <at 0.7913 0>
        ]
       }
       BlueCurve {keys [<at 0.6966 0.9375>]}
       GreenCurve {keys [<at 0.6966 0.9375>]}
       RedCurve {keys [<at 0.6966 0.9375>]}
      }})
    R3 Lime:LinearColor=&ColorFromCurve(ColorCurve=DataInterface() {propertyValues {
       AlphaCurve {
        defaultValue=0
        keys [
         {tangentMode=RCTM_User, time=0.0672, value=0}
         {arriveTangent=13.1575, interpMode=RCIM_Cubic, leaveTangent=13.1575, tangentMode=RCTM_User, time=0.1591, value=7}
         {tangentMode=RCTM_User, time=0.5992, value=7}
         <at 0.7913 0>
        ]
       }
       BlueCurve {keys [<at 0.6966 0>]}
       GreenCurve {keys [<at 0.6966 1>]}
       RedCurve {keys [<at 0.6966 0.5901>]}
      }})
    R3 Orange:LinearColor=&ColorFromCurve(ColorCurve=DataInterface() {propertyValues {
       AlphaCurve {
        defaultValue=0
        keys [
         {tangentMode=RCTM_User, time=0, value=0}
         {arriveTangent=13.1575, interpMode=RCIM_Cubic, leaveTangent=13.1575, tangentMode=RCTM_User, time=0.1591, value=7}
         {tangentMode=RCTM_User, time=0.5992, value=7}
         <at 0.7913 0>
        ]
       }
       BlueCurve {keys [<at 0.6966 0.2681>]}
       GreenCurve {keys [<at 0.6966 0.4367>]}
       RedCurve {keys [<at 0.6966 0.9531>]}
      }})
    R3 Pink:LinearColor=&ColorFromCurve(ColorCurve=DataInterface() {propertyValues {
       AlphaCurve {
        defaultValue=0
        keys [
         {tangentMode=RCTM_User, time=0.0672, value=0}
         {arriveTangent=13.1575, interpMode=RCIM_Cubic, leaveTangent=13.1575, tangentMode=RCTM_User, time=0.1591, value=7}
         {tangentMode=RCTM_User, time=0.5992, value=7}
         <at 0.7913 0>
        ]
       }
       BlueCurve {keys [<at 0.6966 1.2188>]}
       GreenCurve {keys [<at 0.6966 0.2158>]}
       RedCurve {keys [<at 0.6966 1.1488>]}
      }})
    R3 Purple:LinearColor=&ColorFromCurve(ColorCurve=DataInterface() {propertyValues {
       AlphaCurve {
        defaultValue=0
        keys [
         {tangentMode=RCTM_User, time=0.0672, value=0}
         {arriveTangent=13.1575, interpMode=RCIM_Cubic, leaveTangent=13.1575, tangentMode=RCTM_User, time=0.1591, value=7}
         {tangentMode=RCTM_User, time=0.5992, value=7}
         <at 0.7913 0>
        ]
       }
       BlueCurve {keys [<at 0.6966 1>]}
       GreenCurve {keys [<at 0.6966 0>]}
       RedCurve {keys [<at 0.6966 0.3627>]}
      }})
    R3 Saffron:LinearColor=&ColorFromCurve(ColorCurve=DataInterface() {propertyValues {
       AlphaCurve {
        defaultValue=0
        keys [
         {tangentMode=RCTM_User, time=0.0672, value=0}
         {arriveTangent=13.1575, interpMode=RCIM_Cubic, leaveTangent=13.1575, tangentMode=RCTM_User, time=0.1591, value=7}
         {tangentMode=RCTM_User, time=0.5992, value=7}
         <at 0.7913 0>
        ]
       }
       BlueCurve {keys [<at 0.6966 0>]}
       GreenCurve {keys [<at 0.6966 1.8157>]}
       RedCurve {keys [<at 0.6966 2>]}
      }})
    R3 SkyBlue:LinearColor=&ColorFromCurve(ColorCurve=DataInterface() {propertyValues {
       AlphaCurve {
        defaultValue=0
        keys [
         {tangentMode=RCTM_User, time=0.0672, value=0}
         {arriveTangent=13.1575, interpMode=RCIM_Cubic, leaveTangent=13.1575, tangentMode=RCTM_User, time=0.1591, value=7}
         {tangentMode=RCTM_User, time=0.5992, value=7}
         <at 0.7913 0>
        ]
       }
       BlueCurve {keys [<at 0.6966 1.5>]}
       GreenCurve {keys [<at 0.6966 0.3855>]}
       RedCurve {keys [<at 0.6966 0>]}
      }})
    R3 TitaniumWhite:LinearColor=&ColorFromCurve(ColorCurve=DataInterface() {propertyValues {
       AlphaCurve {keys [
         {tangentMode=RCTM_User, time=0.0672, value=0}
         {arriveTangent=13.1575, interpMode=RCIM_Cubic, leaveTangent=13.1575, tangentMode=RCTM_User, time=0.1591, value=7}
         {tangentMode=RCTM_User, time=0.5992, value=7}
         <at 0.7913 0>
        ]}
       BlueCurve {keys [<at 0.6966 2.5>]}
       GreenCurve {keys [<at 0.6966 2.5>]}
       RedCurve {keys [<at 0.6966 2.5>]}
      }})
    R4 BurntSienna:LinearColor=<rgba 0.0312 0.0096 0.0039 0.6>
    R4 Cobalt:LinearColor=<rgba 0 0.0115 0.6198 0.7>
    R4 Crimson:LinearColor=<rgba 0.2656 0 0.0072 0.7>
    R4 ForestGreen:LinearColor=<rgba 0 0.2552 0.0129 0.6>
    R4 Grey:LinearColor=<rgba 0.0417 0.0417 0.0417 0.7>
    R4 Lime:LinearColor=<rgba 0.1875 0.375 0 0.6>
    R4 Orange:LinearColor=<rgba 0.5521 0.1706 0 0.45>
    R4 Pink:LinearColor=<rgba 0.3906 0.0468 0.2656 0.5>
    R4 Purple:LinearColor=<rgba 0.1567 0 0.7552 0.5>
    R4 Saffron:LinearColor=<rgba 0.5573 0.4299 0 0.3>
    R4 SkyBlue:LinearColor=<rgba 0.0137 0.2957 0.4375 0.5>
    R4 TitaniumWhite:LinearColor=<rgba 0.224 0.224 0.224 0.6>
    R5 Black:LinearColor=<rgba 0.0469 0.0469 0.0469 0.6>
    R5 BurntSienna:LinearColor=<rgba 0.1042 0.0322 0.0128 0.7>
    R5 Cobalt:LinearColor=<rgba 0.0312 0.0401 1 0.6>
    R5 Crimson:LinearColor=<rgba 0.4635 0.0266 0.0384 0.6>
    R5 DefaultColor:LinearColor=<rgba 1 1 1 0.5>
    R5 ForestGreen:LinearColor=<rgba 0.0704 0.901 0.1124 0.4>
    R5 Grey:LinearColor=<rgba 0.1667 0.1667 0.1667 0.4>
    R5 Lime:LinearColor=<rgba 0.2938 0.5104 0.0771 0.25>
    R5 Orange:LinearColor=<rgba 0.8261 0.2993 0.0172 0.3>
    R5 Pink:LinearColor=<rgba 0.651 0.1763 0.4784 0.4>
    R5 Purple:LinearColor=<rgba 0.2704 0.1155 0.7917 0.3>
    R5 Saffron:LinearColor=<rgba 0.8906 0.6824 0 0.15>
    R5 SkyBlue:LinearColor=<rgba 0.0754 0.3822 0.5365 0.4>
    R5 TitaniumWhite:LinearColor=<rgba 0.5729 0.5729 0.5729 0.3>
    Use R1 Painted Overrides=true
    Use R2 Painted Overrides=true
    Use R3 Painted Overrides=true
    Use R4 Painted Overrides=true
    Use R5 Painted Overrides=true
   ) {module=VC_PaintedVariantsOverride001}
   &ParticleState()
   &InheritSourceMovement(
    Applied Position Scale=&Multiply_VectorByFloat(
     Vector=<1 1 1>
     Float=&FloatFromCurve(FloatCurve=DataInterface() {propertyValues {Curve {keys [{interpMode=RCIM_Cubic, tangentMode=RCTM_Break, time=0, value=1}, {interpMode=RCIM_Cubic, tangentMode=RCTM_Break, time=1, value=0}]}}})
    )
    ApplyForce=false
    ApplyPosition=true
    ApplyPositionFalloff=false
   )
   &SolveForcesAndVelocity(Clamp Velocity=false, Limit Acceleration=false)
   &VC_UpdateRibbon_v1(Enable Skeletal Position=true)
   /SetVariables(
    Particles.RibbonBase3 Color:LinearColor=&ColorFromCurve(ColorCurve=DataInterface() {propertyValues {
       AlphaCurve {keys [<at 0 1 RCIM_Cubic>]}
       BlueCurve {keys [<at 1 0.0411 RCIM_Cubic>]}
       GreenCurve {keys [<at 1 0.0849 RCIM_Cubic>]}
       RedCurve {keys [<at 1 1 RCIM_Cubic>]}
      }})
    Particles.RibbonGlow:LinearColor=&ColorFromCurve(ColorCurve=DataInterface() {propertyValues {
       AlphaCurve {keys [<at 0.0614 0>, {arriveTangent=15.9916, interpMode=RCIM_Cubic, time=0.1756, value=8}, <at 0.6253 10>, <at 0.9771 0>]}
       BlueCurve {keys [<at 0.6966 1>]}
       GreenCurve {keys [<at 0.6966 0.5>]}
       RedCurve {keys [<at 0.6966 0>]}
      }})
    Particles.RibbonGlow 2:LinearColor=&ColorFromCurve(ColorCurve=DataInterface() {propertyValues {
       AlphaCurve {keys [<at 0.0672 0>, {arriveTangent=11.6813, interpMode=RCIM_Cubic, time=0.1591, value=7}, <at 0.5992 7>, <at 0.7913 0>]}
       BlueCurve {keys [<at 0.4821 1>]}
       GreenCurve {keys [<at 0.4821 0.0656>]}
       RedCurve {keys [<at 0.4821 0>]}
      }})
    Particles.Scale Glow 2:NiagaraFloat=&FloatFromCurve(FloatCurve=DataInterface() {propertyValues {Curve {keys [
         {arriveTangent=0.0312, interpMode=RCIM_Cubic, leaveTangent=0.0312, tangentMode=RCTM_User, time=0.0006, value=200}
         {arriveTangent=-1.8376, interpMode=RCIM_Cubic, leaveTangent=-1.8376, tangentMode=RCTM_User, time=0.3015, value=200}
         {arriveTangent=-848.9316, tangentMode=RCTM_User, time=1.002, value=0.35}
        ]}}})
    Particles.curve_color_lines:LinearColor=&MultiplyLinearColor(
     Linear Color=$Particles.color_lines
     Scale Factor=&MakeVector4_FromLinearColor(LinearColor=&MakeLinearColorFromVectorAndFloat(
       "Vector (RGB)"=<1 1 1>
       "Float (Alpha)"=&FloatFromCurve(FloatCurve=DataInterface() {propertyValues {Curve {keys [{interpMode=RCIM_Cubic, tangentMode=RCTM_Break, time=0.5, value=1}, {interpMode=RCIM_Cubic, tangentMode=RCTM_Break, time=1, value=1}]}}})
      ))
    )
   ) {isSetParameters=true}
   /SetVariables(
    Particles.R1_Color_Scaled:LinearColor=&MultiplyLinearColor(
     Linear Color=$Particles.R1 Color
     Scale Factor=&MakeVector4FromVectorAndFloat(
      "Vector (RGB)"=&VectorFromFloat(Value=&FloatFromCurve(
        FloatCurve=DataInterface() {propertyValues {Curve {keys [
            {arriveTangent=1.0866, interpMode=RCIM_Cubic, leaveTangent=0.2704, tangentMode=RCTM_Break, time=0, value=0}
            {arriveTangent=0.5459, interpMode=RCIM_Cubic, leaveTangent=0.5459, time=0.048, value=0.5}
            <at 0.9158 0.5>
            <at 1 0>
           ]}}}
        Scale Curve=&MakeCustomFloatFromBool(Bool=$User.bIsPainted, True Float=$User.Power_EmissivePaintedR4, False Float=2)
       ))
      "Float (Alpha)"=&FloatFromCurve(
       FloatCurve=DataInterface() {propertyValues {Curve {keys [
           {arriveTangent=1.0866, interpMode=RCIM_Cubic, leaveTangent=0.2704, tangentMode=RCTM_Break, time=0.0064, value=-0.02}
           {arriveTangent=0.7367, interpMode=RCIM_Cubic, leaveTangent=0.7367, time=0.066, value=0.63}
           <at 0.9158 0.65>
           <at 1 0>
          ]}}}
       Scale Curve=&MakeCustomFloatFromBool(Bool=$User.bIsPainted, True Float=$User.Power_AlphaPaintedR4, False Float=0.5)
      )
     )
    )
    Particles.R3_Color_Scaled:LinearColor=&MultiplyLinearColor(
     Linear Color=$Particles.R3 Color
     Scale Factor=&MakeVector4FromVectorAndFloat(
      "Vector (RGB)"=&VectorFromFloat(Value=&FloatFromCurve(
        FloatCurve=DataInterface() {propertyValues {Curve {keys [{arriveTangent=1.0866, interpMode=RCIM_Cubic, leaveTangent=0.2704, tangentMode=RCTM_Break, time=0, value=0.5}, <at 0.7538 0.5>, <at 1 0>]}}}
        Scale Curve=&MakeCustomFloatFromBool(Bool=$User.bIsPainted, True Float=$User.Power_EmissivePaintedR4, False Float=2)
       ))
      "Float (Alpha)"=&FloatFromCurve(
       FloatCurve=DataInterface() {propertyValues {Curve {keys [<at 0 1 RCIM_Cubic>,<at 0.7517 0.7>,<at 0.9905 -0.0198>]}}}
       Scale Curve=&MakeCustomFloatFromBool(Bool=$User.bIsPainted, True Float=$User.Power_AlphaPaintedR4, False Float=1)
      )
     )
    )
    Particles.R4_Color_Scaled:LinearColor=&MultiplyLinearColor(
     Linear Color=$Particles.R4 Color
     Scale Factor=&MakeVector4FromVectorAndFloat(
      "Vector (RGB)"=&VectorFromFloat(Value=&FloatFromCurve(
        FloatCurve=DataInterface() {propertyValues {Curve {keys [
            {arriveTangent=1.0866, interpMode=RCIM_Cubic, leaveTangent=6.1307, tangentMode=RCTM_Break, time=0.28, value=0}
            {arriveTangent=7.4465, interpMode=RCIM_Cubic, tangentMode=RCTM_Break, time=0.4231, value=1}
            <at 0.53 1>
            <at 0.62 0>
           ]}}}
        Scale Curve=&MakeCustomFloatFromBool(Bool=$User.bIsPainted, True Float=$User.Power_EmissivePaintedR4, False Float=0.6)
       ))
      "Float (Alpha)"=&FloatFromCurve(
       FloatCurve=DataInterface() {propertyValues {Curve {keys [
           <at -0.0014 1.7739 RCIM_Cubic>
           {arriveTangent=-1.9854, interpMode=RCIM_Cubic, leaveTangent=-1.9854, time=0.3044, value=2.258}
           {arriveTangent=-0.2383, interpMode=RCIM_Cubic, leaveTangent=-1.8935, tangentMode=RCTM_Break, time=0.7964, value=0.1899}
           <at 1.4148 -1.634 RCIM_Cubic>
          ]}}}
       Scale Curve=&MakeCustomFloatFromBool(Bool=$User.bIsPainted, True Float=$User.Power_AlphaPaintedR4, False Float=1.2)
      )
     )
    )
    Particles.R5_Color_Scaled:LinearColor=&MultiplyLinearColor(
     Linear Color=$Particles.R5 Color
     Scale Factor=&MakeVector4FromVectorAndFloat(
      "Vector (RGB)"=&VectorFromFloat(Value=&FloatFromCurve(
        FloatCurve=DataInterface() {propertyValues {Curve {keys [
            <at -0.17 16.4711 RCIM_Cubic>
            {arriveTangent=-27.6507, interpMode=RCIM_Cubic, leaveTangent=-27.6507, time=0.1072, value=17.8443}
            {arriveTangent=-3.2592, interpMode=RCIM_Cubic, leaveTangent=-3.2592, tangentMode=RCTM_User, time=0.3385, value=2.4113}
            {arriveTangent=2.5648, interpMode=RCIM_Cubic, leaveTangent=2.5648, tangentMode=RCTM_User, time=0.64, value=1.973}
            {arriveTangent=11.7336, interpMode=RCIM_Cubic, leaveTangent=11.7336, time=1.0714, value=20.2083}
            {arriveTangent=-34.897, interpMode=RCIM_Cubic, leaveTangent=-34.897, time=1.3482, value=10.2825}
            {arriveTangent=-8.1043, interpMode=RCIM_Cubic, leaveTangent=-8.1043, time=1.5789, value=2.4972}
            <at 1.7789 6.7918 RCIM_Cubic>
           ]}}}
        Scale Curve=&MakeCustomFloatFromBool(Bool=$User.bIsPainted, True Float=$User.Power_EmissivePaintedR4, False Float=0.7)
       ))
      "Float (Alpha)"=&FloatFromCurve(
       FloatCurve=DataInterface() {propertyValues {Curve {keys [
           <at -0.0556 3.887 RCIM_Cubic>
           {arriveTangent=-1.3925, interpMode=RCIM_Cubic, leaveTangent=-1.3925, time=0.111, value=0.7392}
           {arriveTangent=-0.2383, interpMode=RCIM_Cubic, leaveTangent=-0.2383, tangentMode=RCTM_User, time=0.4793, value=3.1423}
           {arriveTangent=-6.3349, interpMode=RCIM_Cubic, leaveTangent=-6.3349, time=0.8329, value=0.2261}
           <at 1.01 -0.2199 RCIM_Cubic>
          ]}}}
       Scale Curve=&MakeCustomFloatFromBool(Bool=$User.bIsPainted, True Float=$User.Power_AlphaPaintedR4, False Float=0.3)
      )
     )
    )
   ) {isSetParameters=true}
   &ScaleRibbonWidth(Ribbon Width Scale=&FloatFromCurve(
     FloatCurve=DataInterface() {propertyValues {Curve {keys [
         {arriveTangent=0.0312, interpMode=RCIM_Cubic, leaveTangent=0.0312, tangentMode=RCTM_User, time=0.0006, value=2}
         {arriveTangent=-1.8376, interpMode=RCIM_Cubic, leaveTangent=-1.8376, tangentMode=RCTM_User, time=0.1806, value=1.3641}
         {arriveTangent=-1.1432, interpMode=RCIM_Cubic, leaveTangent=-1.1432, time=0.4232, value=0.8024}
         {arriveTangent=-0.7817, interpMode=RCIM_Cubic, leaveTangent=-0.7817, time=0.8251, value=0.6273}
         <at 1.002 0.35>
        ]}}}
     Scale Curve=1.2
    ))
   /SetVariables(Particles.WPO:NiagaraPosition=&AddVectorToPosition(
     Position=$Particles.Position
     Vector=&TransformVector(
      Vector=&MakeVector(
       X=0
       Y=&Sine(
        Normalized Angle=&Add_Float(
         A=&Multiply_Float(A=$Particles.NormalizedAge, B=-1)
         B=&Multiply_Float(A=$Emitter.LoopedAge, B=1)
        )
        Period=&MaskFloatBySpawnGroup(Masked Float=-1.04, Passthrough Float=1.04)
        Scale=&FloatFromCurve(
         FloatCurve=DataInterface() {propertyValues {Curve {
            defaultValue=1
            keys [
             {interpMode=RCIM_Cubic, tangentMode=RCTM_User, time=0.0532, value=0}
             {arriveTangent=1.7197, interpMode=RCIM_Cubic, leaveTangent=1.7197, tangentMode=RCTM_User, time=0.183, value=0.5}
             {arriveTangent=-0.6424, interpMode=RCIM_Cubic, leaveTangent=-0.6424, tangentMode=RCTM_User, time=0.7232, value=0.5}
             {arriveTangent=-0.5663, interpMode=RCIM_Cubic, leaveTangent=-0.5663, tangentMode=RCTM_User, time=1.0055, value=0.25}
            ]
           }}}
         Scale Curve=10
        )
        Bias=0
       )
       Z=&Sine(
        Normalized Angle=&Add_Float(
         A=&Multiply_Float(A=$Particles.NormalizedAge, B=-1)
         B=&Multiply_Float(A=$Emitter.LoopedAge, B=1)
        )
        Period=&MaskFloatBySpawnGroup(Masked Float=-1.04, Passthrough Float=1.04)
        Scale=&FloatFromCurve(
         FloatCurve=DataInterface() {propertyValues {Curve {
            defaultValue=1
            keys [
             {arriveTangent=0.3314, interpMode=RCIM_Cubic, leaveTangent=0.3314, tangentMode=RCTM_User, time=0.0695, value=0.1875}
             {arriveTangent=3.1087, interpMode=RCIM_Cubic, leaveTangent=3.1087, tangentMode=RCTM_User, time=0.183, value=1}
             {arriveTangent=-0.6424, interpMode=RCIM_Cubic, leaveTangent=-0.6424, tangentMode=RCTM_User, time=0.7232, value=1}
             {arriveTangent=-0.8576, interpMode=RCIM_Cubic, leaveTangent=-0.8576, tangentMode=RCTM_User, time=0.9931, value=0.7947}
            ]
           }}}
         Scale Curve=10
        )
        Bias=0
       )
      )
      Destination Space=::Simulation
      Apply Scale=true
     )
    )) {isSetParameters=true}
  ]
  rendererClasses [&NiagaraRibbonRendererProperties]
  renderers [
   {rendererClass=&NiagaraRibbonRendererProperties}
   {rendererClass=&NiagaraRibbonRendererProperties, rendererIndex=1}
   {rendererClass=&NiagaraRibbonRendererProperties, rendererIndex=2}
  ]
 }
 NE_BlackFlame_Base1 {
  bEnabled=false
  emitterUpdateScript [
   &EmitterState(Life Cycle Mode=::NewEnumerator1, Loop Duration=1, UseLoopDelay=false)
   &VC_FilteredSocketCount()
   &VC_RateQualityHandler(Epic Spawn Rate=30, High Spawn Rate=30, Low Spawn Rate=30, Use Custom Rate?=true)
   &VC_SpawnRibbon_v1(Number of Ribbons=$Emitter.PortCount, Spawn Time=-1, SpawnRate=$Emitter.Spawn Rate)
  ]
  particleSpawnScript [
   &VC_PaintedVariantsOverride() {enabled=false}
   &VC_InitializeRibbon_v1(
    Color=$Particles.Color
    Custom Offset=true
    Enable Skeletal Position=true
    Lifetime=&MergeHelper_FloatSwitch(0=1, 1=1, 2=1, 3=?, Index=$User.GameModeID, Num=::NewEnumerator1)
    Local Vector Offset=&AddVectorToPosition(Position=<30 0 2.5>, Vector=<0 0 0>)
    Offset Amount=1
    Ribbon Width=55
   )
   &AddVelocity(Velocity=<-2000 0 0>) {module=AddVelocity001}
   /SetVariables(
    Particles.R1_Width:NiagaraFloat=&SelectFloatFromArray(Direct Array Index=&SpawnGroup(), Float Selection Array=DataInterface() {propertyValues {FloatData [75,0]}})
    Particles.R2_Width:NiagaraFloat=&SelectFloatFromArray(Direct Array Index=&SpawnGroup(), Float Selection Array=DataInterface() {propertyValues {FloatData [0,-105]}})
    Particles.R3_Color:LinearColor=<rgba 0.2241 0.5211 0.6609 3>
   ) {enabled=false, isSetParameters=true, module=SetVariables_#2}
  ]
  particleUpdateScript [
   &ParticleState()
   &InheritSourceMovement(
    Applied Position Scale=&Multiply_VectorByFloat(
     Vector=<1 1 1>
     Float=&FloatFromCurve(FloatCurve=DataInterface() {propertyValues {Curve {keys [{interpMode=RCIM_Cubic, tangentMode=RCTM_Break, time=0, value=1}, {interpMode=RCIM_Cubic, tangentMode=RCTM_Break, time=1, value=0}]}}})
    )
    ApplyForce=false
    ApplyPosition=true
    ApplyPositionFalloff=false
   ) {module=InheritSourceMovement001}
   &SolveForcesAndVelocity(Clamp Velocity=false, Limit Acceleration=false)
   &VC_UpdateRibbon_v1(Enable Skeletal Position=true)
   /SetVariables(Particles.DotProduct:NiagaraFloat=&DotProduct(
     Vector A=/Niagara/DynamicInputs/Math/TransformVectorByMatrix(
      Vector:Vector3f=<1 0 0>
      Transformation Matrix:NiagaraMatrix=$Engine.Owner.SystemLocalToWorld
      Transform Vector as Position:NiagaraBool=false
     )
     Vector B=&NormalizeVector(
      Vector To Normalize=&SubtractVectorFromPosition(
       Position=$Engine.Owner.Position
       Vector=&ReturnCameraPosition(Coordinate Space:ENiagaraCoordinateSpace=::Simulation, Camera Query:NiagaraDataInterfaceCamera=DataInterface())
      )
      Safe Normalize=true
      Safe Fallback Threshold=0
      Safe Fallback Vector=<0 0 0.001>
     )
    )) {isSetParameters=true, module=SetVariables_#3}
   /SetVariables(
    Particles.RibbonBack:LinearColor=&ColorFromCurve(ColorCurve=DataInterface() {propertyValues {
       AlphaCurve {keys [<at 0.5096 0>,<at 0.6 2>,<at 0.6566 0>]}
       BlueCurve {keys [<at 0.9243 0>]}
       GreenCurve {keys [<at 0.9243 0>]}
       RedCurve {keys [<at 0.9243 0>]}
      }})
    Particles.RibbonBase:LinearColor=&ColorFromCurve(ColorCurve=DataInterface() {propertyValues {
       AlphaCurve {keys [<at 0.108 0>,<at 0.1426 1>,<at 0.31 1>,<at 0.4762 0>]}
       BlueCurve {keys [<at 0.9243 0>]}
       GreenCurve {keys [<at 0.9243 0>]}
       RedCurve {keys [<at 0.9243 0>]}
      }})
    Particles.RibbonGlow:LinearColor=&ColorFromCurve(ColorCurve=DataInterface() {propertyValues {
       AlphaCurve {keys [<at 0.2665 3>,<at 0.3905 0>,<at 0.5841 0>]}
       BlueCurve {keys [<at 0.6966 2>]}
       GreenCurve {keys [<at 0.6966 0.5201>]}
       RedCurve {keys [<at 0.6966 0>]}
      }})
    Particles.RibbonGradient:LinearColor=&ColorFromCurve(ColorCurve=DataInterface() {propertyValues {
       AlphaCurve {keys [
         <at 0.2819 7>
         <at 0.4191 0>
         <at 0.6302 0>
         <at 0.7212 1>
         {arriveTangent=-6.7757, interpMode=RCIM_Cubic, leaveTangent=-6.7757, time=0.7822, value=1}
         <at 0.8688 0>
        ]}
       BlueCurve {keys [<at 1 0.1683 RCIM_Cubic>]}
       GreenCurve {keys [<at 1 0 RCIM_Cubic>]}
       RedCurve {keys [<at 1 0.0865 RCIM_Cubic>]}
      }})
   ) {isSetParameters=true, module=SetVariables_#4}
   &ScaleRibbonWidth(Ribbon Width Scale=&FloatFromCurve(FloatCurve=DataInterface() {propertyValues {Curve {keys [
         {arriveTangent=-0.2605, interpMode=RCIM_Cubic, leaveTangent=-0.2605, tangentMode=RCTM_User, time=0.0043, value=1.4579}
         {arriveTangent=-1.5709, interpMode=RCIM_Cubic, leaveTangent=-1.5709, tangentMode=RCTM_User, time=0.3137, value=1.1734}
         {arriveTangent=-2.7437, interpMode=RCIM_Cubic, leaveTangent=-2.7437, time=0.5043, value=0.5604}
         {arriveTangent=-1.126, interpMode=RCIM_Cubic, leaveTangent=-1.126, time=0.6133, value=0.3515}
         {arriveTangent=-1.0469, tangentMode=RCTM_User, time=1.002, value=0}
        ]}}}))
   /SetVariables(
    Particles.Ribbon_BlackColor:LinearColor=<rgba 0 0 0 1>
    Particles.WaveScaleCurve:NiagaraFloat=&FloatFromCurve(
     FloatCurve=DataInterface() {propertyValues {Curve {keys [
         {arriveTangent=39.5227, interpMode=RCIM_Cubic, leaveTangent=39.5227, tangentMode=RCTM_User, time=0.0092, value=0}
         {
          arriveTangent=0.7219
          arriveTangentWeight=0.0795
          interpMode=RCIM_Cubic
          leaveTangent=0.1806
          leaveTangentWeight=0.0696
          tangentMode=RCTM_Break
          tangentWeightMode=RCTWM_WeightedBoth
          time=0.125
          value=1.25
         }
         {arriveTangent=0.1988, interpMode=RCIM_Cubic, leaveTangent=0.1988, tangentMode=RCTM_User, time=0.45, value=0.625}
        ]}}}
     Scale Curve=10
    )
   ) {isSetParameters=true, module=SetVariables_#5}
   /SetVariables(
    Particles.CustomPosition_R1:NiagaraPosition=&AddVectorToPosition(
     Position=$Particles.Position
     Vector=&Multiply_VectorByFloat(
      Vector=&MakeVector(
       X=&Sine(
        Normalized Angle=&Add_Float(A=0, B=&Subtract_Float(A=$Emitter.NormalizedLoopAge, B=$Particles.NormalizedAge))
        Period=$User.WavePeriod
        Scale=&Multiply_Float(
         A=&SelectFloatFromArray(Direct Array Index=$Particles.SpawnGroup, Float Selection Array=DataInterface() {propertyValues {FloatData [1,-1]}})
         B=$User.WaveMaxScale
        )
        Bias=0
       )
       Y=&Sine(
        Normalized Angle=&Add_Float(A=0, B=&Subtract_Float(A=$Emitter.NormalizedLoopAge, B=$Particles.NormalizedAge))
        Period=$User.WavePeriod
        Scale=&Multiply_Float(
         A=&SelectFloatFromArray(Direct Array Index=$Particles.SpawnGroup, Float Selection Array=DataInterface() {propertyValues {FloatData [1,-1]}})
         B=$User.WaveMaxScale
        )
        Bias=0
       )
       Z=&Cosine(
        Normalized Angle=&Add_Float(A=0, B=&Subtract_Float(A=$Emitter.NormalizedLoopAge, B=$Particles.NormalizedAge))
        Period=$User.WavePeriod
        Scale=&Multiply_Float(
         A=&SelectFloatFromArray(Direct Array Index=$Particles.SpawnGroup, Float Selection Array=DataInterface() {propertyValues {FloatData [1,-1]}})
         B=$User.WaveMaxScale
        )
        Bias=0
       )
      )
      Float=$Particles.WaveScaleCurve
     )
    )
    Particles.CustomPosition_R2:NiagaraPosition=&AddVectorToPosition(
     Position=$Particles.Position
     Vector=&Multiply_VectorByFloat(
      Vector=&MakeVector(
       X=&Sine(
        Normalized Angle=&Add_Float(
         A=&Multiply_Float(A=$User.WavePeriod, B=0.5)
         B=&Subtract_Float(A=$Emitter.NormalizedLoopAge, B=$Particles.NormalizedAge)
        )
        Period=$User.WavePeriod
        Scale=&Multiply_Float(
         A=&SelectFloatFromArray(Direct Array Index=$Particles.SpawnGroup, Float Selection Array=DataInterface() {propertyValues {FloatData [1,1]}})
         B=$User.WaveMaxScale
        )
        Bias=0
       )
       Y=&Sine(
        Normalized Angle=&Add_Float(
         A=&Multiply_Float(A=$User.WavePeriod, B=0.5)
         B=&Subtract_Float(A=$Emitter.NormalizedLoopAge, B=$Particles.NormalizedAge)
        )
        Period=$User.WavePeriod
        Scale=&Multiply_Float(
         A=&SelectFloatFromArray(Direct Array Index=$Particles.SpawnGroup, Float Selection Array=DataInterface() {propertyValues {FloatData [1,1]}})
         B=$User.WaveMaxScale
        )
        Bias=0
       )
       Z=&Cosine(
        Normalized Angle=&Add_Float(
         A=&Multiply_Float(A=$User.WavePeriod, B=0.5)
         B=&Subtract_Float(A=$Emitter.NormalizedLoopAge, B=$Particles.NormalizedAge)
        )
        Period=$User.WavePeriod
        Scale=&Multiply_Float(
         A=&SelectFloatFromArray(Direct Array Index=$Particles.SpawnGroup, Float Selection Array=DataInterface() {propertyValues {FloatData [-1,1]}})
         B=$User.WaveMaxScale
        )
        Bias=0
       )
      )
      Float=$Particles.WaveScaleCurve
     )
    )
   ) {isSetParameters=true, module=SetVariables_#6}
   /SetVariables(Particles.DynamicMP_R2:Vector4f=&MakeVector4(
     X=&FloatFromCurve(FloatCurve=DataInterface() {propertyValues {Curve {
         defaultValue=1
         keys [
          {tangentMode=RCTM_User, time=0, value=0.5}
          {arriveTangent=-0.0633, interpMode=RCIM_Cubic, leaveTangent=-0.0633, tangentMode=RCTM_User, time=0.65, value=0.5}
          {arriveTangent=-0.0699, interpMode=RCIM_Cubic, leaveTangent=-0.0699, tangentMode=RCTM_User, time=0.75, value=-2}
         ]
        }}})
     Y=0
     Z=0
     W=0
    )) {isSetParameters=true, module=SetVariables_#7}
   &DynamicMaterialParameters(
    Index 0 Param 1=&FloatFromCurve(FloatCurve=DataInterface() {propertyValues {Curve {
        defaultValue=1
        keys [
         {
          arriveTangent=0.0405
          arriveTangentWeight=0.1835
          interpMode=RCIM_Cubic
          leaveTangent=0.0405
          leaveTangentWeight=0.1835
          tangentMode=RCTM_User
          time=0
          value=0
         }
         {
          arriveTangent=0.0201
          arriveTangentWeight=0.1834
          interpMode=RCIM_Cubic
          leaveTangent=0.0201
          leaveTangentWeight=0.1834
          tangentMode=RCTM_User
          time=0.5
          value=1
         }
        ]
       }}})
    Module.Index0Float0Write=true
    Module.Index0Float1Write=true
    Module.Index0Float2Write=true
    Module.Index0Float3Write=true
   )
  ]
  rendererClasses [&NiagaraRibbonRendererProperties]
  renderers [
   {rendererClass=&NiagaraRibbonRendererProperties}
   {rendererClass=&NiagaraRibbonRendererProperties, rendererIndex=1}
   {rendererClass=&NiagaraRibbonRendererProperties, rendererIndex=2}
   {rendererClass=&NiagaraRibbonRendererProperties, rendererIndex=3}
   {rendererClass=&NiagaraRibbonRendererProperties, rendererIndex=4}
  ]
 }
 NE_BlackFlame_Base1001 {
  bEnabled=false
  emitterUpdateScript [
   &EmitterState(Life Cycle Mode=::NewEnumerator1, Loop Duration=1, UseLoopDelay=false)
   &VC_FilteredSocketCount()
   &VC_RateQualityHandler(Epic Spawn Rate=30, High Spawn Rate=30, Low Spawn Rate=30, Use Custom Rate?=true)
   &VC_SpawnRibbon_v1(Number of Ribbons=$Emitter.PortCount, Spawn Time=-1, SpawnRate=$Emitter.Spawn Rate)
  ]
  particleSpawnScript [
   &VC_PaintedVariantsOverride() {enabled=false}
   &VC_InitializeRibbon_v1(
    Color=$Particles.Color
    Custom Offset=true
    Enable Skeletal Position=true
    Lifetime=&MergeHelper_FloatSwitch(0=1, 1=1, 2=1, 3=?, Index=0, Num=::NewEnumerator1)
    Local Vector Offset=&AddVectorToPosition(Position=<30 0 2.5>, Vector=<0 0 0>)
    Offset Amount=1
    Ribbon Width=55
   )
   &AddVelocity(Velocity=<-2000 0 0>) {module=AddVelocity001}
   /SetVariables(
    Particles.R1_Width:NiagaraFloat=&SelectFloatFromArray(Direct Array Index=&SpawnGroup(), Float Selection Array=DataInterface() {propertyValues {FloatData [75,0]}})
    Particles.R2_Width:NiagaraFloat=&SelectFloatFromArray(Direct Array Index=&SpawnGroup(), Float Selection Array=DataInterface() {propertyValues {FloatData [0,-105]}})
    Particles.R3_Color:LinearColor=<rgba 0.2241 0.5211 0.6609 3>
   ) {enabled=false, isSetParameters=true, module=SetVariables_#2}
  ]
  particleUpdateScript [
   &VC_PaintedVariantsOverride(
    R1 Black:LinearColor=&ColorFromCurve(ColorCurve=DataInterface() {propertyValues {
       AlphaCurve {keys [
         <at 0.2819 7>
         <at 0.4191 0>
         <at 0.6302 0>
         <at 0.7212 1>
         {arriveTangent=-6.7757, interpMode=RCIM_Cubic, leaveTangent=-6.7757, time=0.7822, value=1}
         <at 0.8688 0>
        ]}
       BlueCurve {keys [<at 1 0.1094 RCIM_Cubic>]}
       GreenCurve {keys [<at 1 0.1094 RCIM_Cubic>]}
       RedCurve {keys [<at 1 0.1094 RCIM_Cubic>]}
      }})
    R1 BurntSienna:LinearColor=&ColorFromCurve(ColorCurve=DataInterface() {propertyValues {
       AlphaCurve {keys [
         <at 0.2819 7>
         <at 0.4191 0>
         <at 0.6302 0>
         <at 0.7212 1>
         {arriveTangent=-6.7757, interpMode=RCIM_Cubic, leaveTangent=-6.7757, time=0.7822, value=1}
         <at 0.8688 0>
        ]}
       BlueCurve {keys [<at 1 0.0686 RCIM_Cubic>]}
       GreenCurve {keys [<at 1 0.0858 RCIM_Cubic>]}
       RedCurve {keys [<at 1 0.1667 RCIM_Cubic>]}
      }})
    R1 Cobalt:LinearColor=&ColorFromCurve(ColorCurve=DataInterface() {propertyValues {
       AlphaCurve {keys [
         <at 0.2819 7>
         <at 0.4191 0>
         <at 0.6302 0>
         <at 0.7212 1>
         {arriveTangent=-6.7757, interpMode=RCIM_Cubic, leaveTangent=-6.7757, time=0.7822, value=1}
         <at 0.8688 0>
        ]}
       BlueCurve {keys [<at 1 0.3802 RCIM_Cubic>]}
       GreenCurve {keys [<at 1 0.0931 RCIM_Cubic>]}
       RedCurve {keys [<at 1 0.0931 RCIM_Cubic>]}
      }})
    R1 Crimson:LinearColor=&ColorFromCurve(ColorCurve=DataInterface() {propertyValues {
       AlphaCurve {keys [
         <at 0.2819 7>
         <at 0.4191 0>
         <at 0.6302 0>
         <at 0.7212 1>
         {arriveTangent=-6.7757, interpMode=RCIM_Cubic, leaveTangent=-6.7757, time=0.7822, value=1}
         <at 0.8688 0>
        ]}
       BlueCurve {keys [<at 1 0.0131 RCIM_Cubic>]}
       GreenCurve {keys [<at 1 0 RCIM_Cubic>]}
       RedCurve {keys [<at 1 0.2348 RCIM_Cubic>]}
      }})
    R1 DefaultColor=&ColorFromCurve(ColorCurve=DataInterface() {propertyValues {
       AlphaCurve {keys [
         <at 0.2819 7>
         <at 0.4191 0>
         <at 0.6302 0>
         <at 0.7212 1>
         {arriveTangent=-6.7757, interpMode=RCIM_Cubic, leaveTangent=-6.7757, time=0.7822, value=1}
         <at 0.8688 0>
        ]}
       BlueCurve {keys [<at 1 0.1195 RCIM_Cubic>]}
       GreenCurve {keys [<at 1 0 RCIM_Cubic>]}
       RedCurve {keys [<at 1 0.063 RCIM_Cubic>]}
      }})
    R1 ForestGreen:LinearColor=&ColorFromCurve(ColorCurve=DataInterface() {propertyValues {
       AlphaCurve {keys [
         <at 0.2819 7>
         <at 0.4191 0>
         <at 0.6302 0>
         <at 0.7212 1>
         {arriveTangent=-6.7757, interpMode=RCIM_Cubic, leaveTangent=-6.7757, time=0.7822, value=1}
         <at 0.8688 0>
        ]}
       BlueCurve {keys [<at 1 0.0198 RCIM_Cubic>]}
       GreenCurve {keys [<at 1 0.2031 RCIM_Cubic>]}
       RedCurve {keys [<at 1 0 RCIM_Cubic>]}
      }})
    R1 Grey:LinearColor=&ColorFromCurve(ColorCurve=DataInterface() {propertyValues {
       AlphaCurve {keys [
         <at 0.2819 1>
         <at 0.4191 0>
         <at 0.6302 0>
         <at 0.7212 1>
         {arriveTangent=-6.7757, interpMode=RCIM_Cubic, leaveTangent=-6.7757, time=0.7822, value=1}
         <at 0.8688 0>
        ]}
       BlueCurve {keys [<at 1 0.3542 RCIM_Cubic>]}
       GreenCurve {keys [<at 1 0.3542 RCIM_Cubic>]}
       RedCurve {keys [<at 1 0.3542 RCIM_Cubic>]}
      }})
    R1 Lime:LinearColor=&ColorFromCurve(ColorCurve=DataInterface() {propertyValues {
       AlphaCurve {keys [
         <at 0.2819 7>
         <at 0.4191 0>
         <at 0.6302 0>
         <at 0.7212 1>
         {arriveTangent=-6.7757, interpMode=RCIM_Cubic, leaveTangent=-6.7757, time=0.7822, value=1}
         <at 0.8688 0>
        ]}
       BlueCurve {keys [<at 1 0.0865 RCIM_Cubic>]}
       GreenCurve {keys [<at 1 0.2865 RCIM_Cubic>]}
       RedCurve {keys [<at 1 0.1697 RCIM_Cubic>]}
      }})
    R1 Orange:LinearColor=&ColorFromCurve(ColorCurve=DataInterface() {propertyValues {
       AlphaCurve {keys [
         <at 0.2819 7>
         <at 0.4191 0>
         <at 0.6302 0>
         <at 0.7212 1>
         {arriveTangent=-6.7757, interpMode=RCIM_Cubic, leaveTangent=-6.7757, time=0.7822, value=1}
         <at 0.8688 0>
        ]}
       BlueCurve {keys [<at 1 0.0688 RCIM_Cubic>]}
       GreenCurve {keys [<at 1 0.1085 RCIM_Cubic>]}
       RedCurve {keys [<at 1 0.3478 RCIM_Cubic>]}
      }})
    R1 Pink:LinearColor=&ColorFromCurve(ColorCurve=DataInterface() {propertyValues {
       AlphaCurve {keys [
         <at 0.2819 7>
         <at 0.4191 0>
         <at 0.6302 0>
         <at 0.7212 1>
         {arriveTangent=-6.7757, interpMode=RCIM_Cubic, leaveTangent=-6.7757, time=0.7822, value=1}
         <at 0.8688 0>
        ]}
       BlueCurve {keys [<at 1 0.2004 RCIM_Cubic>]}
       GreenCurve {keys [<at 1 0 RCIM_Cubic>]}
       RedCurve {keys [<at 1 0.2188 RCIM_Cubic>]}
      }})
    R1 Purple:LinearColor=&ColorFromCurve(ColorCurve=DataInterface() {propertyValues {
       AlphaCurve {keys [
         <at 0.2819 7>
         <at 0.4191 0>
         <at 0.6302 0>
         <at 0.7212 1>
         {arriveTangent=-6.7757, interpMode=RCIM_Cubic, leaveTangent=-6.7757, time=0.7822, value=1}
         <at 0.8688 0>
        ]}
       BlueCurve {keys [<at 1 0.375 RCIM_Cubic>]}
       GreenCurve {keys [<at 1 0 RCIM_Cubic>]}
       RedCurve {keys [<at 1 0.1484 RCIM_Cubic>]}
      }})
    R1 Saffron:LinearColor=&ColorFromCurve(ColorCurve=DataInterface() {propertyValues {
       AlphaCurve {keys [
         <at 0.2819 7>
         <at 0.4191 0>
         <at 0.6302 0>
         <at 0.7212 1>
         {arriveTangent=-6.7757, interpMode=RCIM_Cubic, leaveTangent=-6.7757, time=0.7822, value=1}
         <at 0.8688 0>
        ]}
       BlueCurve {keys [<at 1 0 RCIM_Cubic>]}
       GreenCurve {keys [<at 1 0.2395 RCIM_Cubic>]}
       RedCurve {keys [<at 1 0.2812 RCIM_Cubic>]}
      }})
    R1 SkyBlue:LinearColor=&ColorFromCurve(ColorCurve=DataInterface() {propertyValues {
       AlphaCurve {keys [
         <at 0.2819 7>
         <at 0.4191 0>
         <at 0.6302 0>
         <at 0.7212 1>
         {arriveTangent=-6.7757, interpMode=RCIM_Cubic, leaveTangent=-6.7757, time=0.7822, value=1}
         <at 0.8688 0>
        ]}
       BlueCurve {keys [<at 1 0.3802 RCIM_Cubic>]}
       GreenCurve {keys [<at 1 0.1905 RCIM_Cubic>]}
       RedCurve {keys [<at 1 0 RCIM_Cubic>]}
      }})
    R1 TitaniumWhite:LinearColor=&ColorFromCurve(ColorCurve=DataInterface() {propertyValues {
       AlphaCurve {keys [
         <at 0.2819 2>
         <at 0.4191 0>
         <at 0.6302 0>
         <at 0.7212 1>
         {arriveTangent=-6.7757, interpMode=RCIM_Cubic, leaveTangent=-6.7757, time=0.7822, value=1}
         <at 0.8688 0>
        ]}
       BlueCurve {keys [<at 1 0.6406 RCIM_Cubic>]}
       GreenCurve {keys [<at 1 0.6406 RCIM_Cubic>]}
       RedCurve {keys [<at 1 0.6406 RCIM_Cubic>]}
      }})
    R2 Black:LinearColor=&ColorFromCurve(ColorCurve=DataInterface() {propertyValues {
       AlphaCurve {keys [<at 0 7>,<at 0.2232 1>,<at 0.3261 0.05>,<at 0.5495 0.2>,<at 0.7007 0.2>,<at 0.8474 0 RCIM_Cubic>]}
       BlueCurve {keys [<at 0 0.0194 RCIM_Cubic>,<at 0.1492 0.0333>,<at 0.2681 0.1104>,<at 0.4863 0.9274>,<at 0.6088 0.0305>]}
       GreenCurve {keys [<at 0 0.0194 RCIM_Cubic>,<at 0.1492 0.0333>,<at 0.2681 0.1104>,<at 0.4863 0.9274>,<at 0.6088 0.0305>]}
       RedCurve {keys [<at 0 0.0194 RCIM_Cubic>,<at 0.1492 0.0333>,<at 0.2681 0.1104>,<at 0.4863 0.9274>,<at 0.6088 0.0305>]}
      }})
    R2 BurntSienna:LinearColor=&ColorFromCurve(ColorCurve=DataInterface() {propertyValues {
       AlphaCurve {keys [<at 0 4>,<at 0.1907 1>,<at 0.3261 0.05>,<at 0.593 0.2>,<at 0.8474 0 RCIM_Cubic>]}
       BlueCurve {keys [<at 0 0.008 RCIM_Cubic>,<at 0.2839 0.0202>,<at 0.4863 0.2907>,<at 0.5956 0.008>]}
       GreenCurve {keys [<at 0 0.0203 RCIM_Cubic>,<at 0.2839 0.0323>,<at 0.4863 0.4394>,<at 0.5956 0.0203>]}
       RedCurve {keys [<at 0 0.0648 RCIM_Cubic>,<at 0.2839 0.0762>,<at 0.4863 0.9792>,<at 0.5956 0.0648>]}
      }})
    R2 Cobalt:LinearColor=&ColorFromCurve(ColorCurve=DataInterface() {propertyValues {
       AlphaCurve {keys [<at 0 7>,<at 0.1907 1>,<at 0.3261 0.05>,<at 0.5495 0.6>,<at 0.7007 0.2>,<at 0.8474 0 RCIM_Cubic>]}
       BlueCurve {keys [<at 0 0.6458 RCIM_Cubic>,<at 0.2575 0.9985>,<at 0.4863 1>,<at 0.6115 0.8177>]}
       GreenCurve {keys [<at 0 0.0119 RCIM_Cubic>,<at 0.2575 0.0367>,<at 0.4863 0.2997>,<at 0.6115 0.0527>]}
       RedCurve {keys [<at 0 0 RCIM_Cubic>,<at 0.2575 0.0186>,<at 0.4863 0.2865>,<at 0.6115 0.0383>]}
      }})
    R2 Crimson:LinearColor=&ColorFromCurve(ColorCurve=DataInterface() {propertyValues {
       AlphaCurve {keys [<at 0 7>,<at 0.1907 1>,<at 0.3261 0.05>,<at 0.5495 0.2>,<at 0.7007 0.2>,<at 0.8474 0 RCIM_Cubic>]}
       BlueCurve {keys [<at 0 0 RCIM_Cubic>,<at 0.1492 0.0104>,<at 0.4863 0.2995>,<at 0.6458 0.0894>]}
       GreenCurve {keys [<at 0 0 RCIM_Cubic>,<at 0.1492 0.0104>,<at 0.4863 0.2995>,<at 0.6458 0.0894>]}
       RedCurve {keys [<at 0 0.4323 RCIM_Cubic>,<at 0.1492 0.9992>,<at 0.4863 0.9274>,<at 0.6458 0.993>]}
      }})
    R2 DefaultColor:LinearColor=&ColorFromCurve(ColorCurve=DataInterface() {propertyValues {
       AlphaCurve {keys [<at 0 1>,<at 0.0914 0.2857>,<at 0.308 0.05>,<at 0.5592 0.2>,<at 0.8632 0>]}
       BlueCurve {keys [<at 0.0009 0>,<at 0.2668 -0.2829>,<at 0.5114 0.8849>,<at 0.619 1>,<at 0.682 0.383>]}
       GreenCurve {keys [<at 0.0009 0.0767>,<at 0.2742 -0.1994>,<at 0.5114 0.8849>,<at 0.619 1>,<at 0.682 0.2585>]}
       RedCurve {keys [<at 0.0009 3.2194>,<at 0.0016 3.2194>,<at 0.4479 0.8849>,<at 0.5114 0.8849>,<at 0.619 1>]}
      }})
    R2 ForestGreen:LinearColor=&ColorFromCurve(ColorCurve=DataInterface() {propertyValues {
       AlphaCurve {keys [<at 0 6>,<at 0.1907 1>,<at 0.3261 0.05>,<at 0.5495 0.2>,<at 0.7007 0.2>,<at 0.8474 0 RCIM_Cubic>]}
       BlueCurve {keys [<at 0 0.0172 RCIM_Cubic>,<at 0.2179 0.0322>,<at 0.4863 0.4656>,<at 0.6537 0.1027>]}
       GreenCurve {keys [<at 0 0.3438 RCIM_Cubic>,<at 0.2179 0.245>,<at 0.4863 0.875>,<at 0.6537 0.2987>]}
       RedCurve {keys [<at 0 0 RCIM_Cubic>,<at 0.2179 0.021>,<at 0.4863 0.4466>,<at 0.6537 0.0924>]}
      }})
    R2 Grey:LinearColor=&ColorFromCurve(ColorCurve=DataInterface() {propertyValues {
       AlphaCurve {keys [<at 0 4>,<at 0.1907 1>,<at 0.3261 0.05>,<at 0.5495 0.2>,<at 0.7007 0.2>,<at 0.8474 0 RCIM_Cubic>]}
       BlueCurve {keys [<at 0 0.0692 RCIM_Cubic>,<at 0.2285 0.1336>,<at 0.4863 0.9274>,<at 0.6405 0.132>]}
       GreenCurve {keys [<at 0 0.0781 RCIM_Cubic>,<at 0.2285 0.1336>,<at 0.4863 0.9274>,<at 0.6405 0.132>]}
       RedCurve {keys [<at 0 0.0767 RCIM_Cubic>,<at 0.2285 0.1406>,<at 0.4863 0.9274>,<at 0.6405 0.1354>]}
      }})
    R2 Lime:LinearColor=&ColorFromCurve(ColorCurve=DataInterface() {propertyValues {
       AlphaCurve {keys [<at 0 6>,<at 0.1907 1>,<at 0.3261 0.05>,<at 0.5495 0.2>,<at 0.8122 0 RCIM_Cubic>]}
       BlueCurve {keys [<at 0 0 RCIM_Cubic>,<at 0.2179 0.0459>,<at 0.4863 0.9274>,<at 0.6299 0.0724>]}
       GreenCurve {keys [<at 0 0.6875 RCIM_Cubic>,<at 0.2179 0.4635>,<at 0.4863 0.9274>,<at 0.6299 0.7708>]}
       RedCurve {keys [<at 0 0.3447 RCIM_Cubic>,<at 0.2179 0.2553>,<at 0.4863 0.9274>,<at 0.6299 0.4226>]}
      }})
    R2 Orange:LinearColor=&ColorFromCurve(ColorCurve=DataInterface() {propertyValues {
       AlphaCurve {keys [<at 0 7>,<at 0.1907 1>,<at 0.3261 0.05>,<at 0.5495 0.2>,<at 0.7007 0.2>,<at 0.8474 0 RCIM_Cubic>]}
       BlueCurve {keys [<at 0 0 RCIM_Cubic>,<at 0.2364 0.0598>,<at 0.4863 0.4263>,<at 0.6217 0.1191>]}
       GreenCurve {keys [<at 0 0.2834 RCIM_Cubic>,<at 0.2364 0.2792>,<at 0.4863 0.5719>,<at 0.6217 0.3121>]}
       RedCurve {keys [<at 0 0.8125 RCIM_Cubic>,<at 0.2364 0.6889>,<at 0.4863 0.8438>,<at 0.6217 0.6724>]}
      }})
    R2 Pink:LinearColor=&ColorFromCurve(ColorCurve=DataInterface() {propertyValues {
       AlphaCurve {keys [<at 0 7>,<at 0.1907 1>,<at 0.3261 0.05>,<at 0.5495 0.2>,<at 0.7007 0.2>,<at 0.8474 0 RCIM_Cubic>]}
       BlueCurve {keys [<at 0 0.3419 RCIM_Cubic>,<at 0.2707 0.3564>,<at 0.4863 0.7216>,<at 0.6141 0.5634>]}
       GreenCurve {keys [<at 0 0.0612 RCIM_Cubic>,<at 0.2707 0.1489>,<at 0.4863 0.5789>,<at 0.6141 0.2536>]}
       RedCurve {keys [<at 0 0.5029 RCIM_Cubic>,<at 0.2707 0.5614>,<at 0.4863 0.855>,<at 0.6141 0.8542>]}
      }})
    R2 Purple:LinearColor=&ColorFromCurve(ColorCurve=DataInterface() {propertyValues {
       AlphaCurve {keys [<at 0 7>,<at 0.1907 1>,<at 0.3261 0.05>,<at 0.5495 0.2>,<at 0.7007 0.2>,<at 0.8474 0 RCIM_Cubic>]}
       BlueCurve {keys [<at 0 0.8281 RCIM_Cubic>,<at 0.2919 0.6836>,<at 0.4863 0.8957>,<at 0.6141 0.9565>]}
       GreenCurve {keys [<at 0 0 RCIM_Cubic>,<at 0.2919 0.0563>,<at 0.4863 0.2791>,<at 0.6141 0.1097>]}
       RedCurve {keys [<at 0 0.1924 RCIM_Cubic>,<at 0.2919 0.3349>,<at 0.4863 0.5023>,<at 0.6141 0.3436>]}
      }})
    R2 Saffron:LinearColor=&ColorFromCurve(ColorCurve=DataInterface() {propertyValues {
       AlphaCurve {keys [<at 0 4>,<at 0.1907 1>,<at 0.3261 0.05>,<at 0.5495 0.2>,<at 0.8175 0 RCIM_Cubic>]}
       BlueCurve {keys [<at 0 0 RCIM_Cubic>,<at 0.2697 0.0046>,<at 0.4863 0.3188>,<at 0.6748 0.0798>]}
       GreenCurve {keys [<at 0 0.6105 RCIM_Cubic>,<at 0.2697 0.6881>,<at 0.4863 0.9274>,<at 0.6748 0.6052>]}
       RedCurve {keys [<at 0 0.7969 RCIM_Cubic>,<at 0.2697 0.8967>,<at 0.4863 0.8744>,<at 0.6748 0.7656>]}
      }})
    R2 SkyBlue:LinearColor=&ColorFromCurve(ColorCurve=DataInterface() {propertyValues {
       AlphaCurve {keys [<at 0 7>,<at 0.1907 1>,<at 0.3261 0.05>,<at 0.5495 0.2>,<at 0.8096 0 RCIM_Cubic>]}
       BlueCurve {keys [<at 0 0.5333 RCIM_Cubic>,<at 0.2496 0.5525>,<at 0.4863 0.9323>,<at 0.6141 0.9323>]}
       GreenCurve {keys [<at 0 0.3564 RCIM_Cubic>,<at 0.2496 0.3792>,<at 0.4863 0.8083>,<at 0.6141 0.6988>]}
       RedCurve {keys [<at 0 0 RCIM_Cubic>,<at 0.2496 0.0875>,<at 0.4863 0.5584>,<at 0.6141 0.2282>]}
      }})
    R2 TitaniumWhite:LinearColor=&ColorFromCurve(ColorCurve=DataInterface() {propertyValues {
       AlphaCurve {keys [<at 0 4>,<at 0.1907 1>,<at 0.3261 0.05>,<at 0.5495 0.2>,<at 0.7007 0.2>,<at 0.8175 0 RCIM_Cubic>]}
       BlueCurve {keys [<at 0 0.1846 RCIM_Cubic>,<at 0.2285 0.1336>,<at 0.4863 0.9274>,<at 0.6405 0.3248>]}
       GreenCurve {keys [<at 0 0.2083 RCIM_Cubic>,<at 0.2285 0.1336>,<at 0.4863 0.9274>,<at 0.6405 0.3248>]}
       RedCurve {keys [<at 0 0.2046 RCIM_Cubic>,<at 0.2285 0.1406>,<at 0.4863 0.9274>,<at 0.6405 0.3333>]}
      }})
    R3 Black:LinearColor=&ColorFromCurve(ColorCurve=DataInterface() {propertyValues {
       AlphaCurve {keys [<at 0.2665 3>,<at 0.3905 0>,<at 0.5841 0>]}
       BlueCurve {keys [<at 0.6966 0.3043>]}
       GreenCurve {keys [<at 0.6966 0.3043>]}
       RedCurve {keys [<at 0.6966 0.3043>]}
      }})
    R3 BurntSienna:LinearColor=&ColorFromCurve(ColorCurve=DataInterface() {propertyValues {
       AlphaCurve {keys [<at 0.2665 1>,<at 0.3905 0>,<at 0.5841 0>]}
       BlueCurve {keys [<at 0.6966 0.2729>]}
       GreenCurve {keys [<at 0.6966 0.3693>]}
       RedCurve {keys [<at 0.6966 0.6719>]}
      }})
    R3 Cobalt:LinearColor=&ColorFromCurve(ColorCurve=DataInterface() {propertyValues {
       AlphaCurve {keys [<at 0.2665 3>,<at 0.3905 0>,<at 0.5841 0>]}
       BlueCurve {keys [<at 0.6966 1.5>]}
       GreenCurve {keys [<at 0.6966 0.0598>]}
       RedCurve {keys [<at 0.6966 0>]}
      }})
    R3 Crimson:LinearColor=&ColorFromCurve(ColorCurve=DataInterface() {propertyValues {
       AlphaCurve {keys [<at 0.2665 3>,<at 0.3905 0>,<at 0.5841 0>]}
       BlueCurve {keys [<at 0.6966 0>]}
       GreenCurve {keys [<at 0.6966 0>]}
       RedCurve {keys [<at 0.6966 1>]}
      }})
    R3 DefaultColor:LinearColor=&ColorFromCurve(ColorCurve=DataInterface() {propertyValues {
       AlphaCurve {keys [<at 0.2665 3>,<at 0.3905 0>,<at 0.5841 0>]}
       BlueCurve {keys [<at 0.6966 2>]}
       GreenCurve {keys [<at 0.6966 0.5201>]}
       RedCurve {keys [<at 0.6966 0>]}
      }})
    R3 ForestGreen:LinearColor=&ColorFromCurve(ColorCurve=DataInterface() {propertyValues {
       AlphaCurve {keys [<at 0.2665 3>,<at 0.3905 0>,<at 0.5841 0>]}
       BlueCurve {keys [<at 0.6966 0.3387>]}
       GreenCurve {keys [<at 0.6966 1.5>]}
       RedCurve {keys [<at 0.6966 0>]}
      }})
    R3 Grey:LinearColor=&ColorFromCurve(ColorCurve=DataInterface() {propertyValues {
       AlphaCurve {keys [<at 0.2665 3>,<at 0.5658 0>,<at 0.7715 0>]}
       BlueCurve {keys [<at 0.6966 0.5>]}
       GreenCurve {keys [<at 0.6966 0.5>]}
       RedCurve {keys [<at 0.6966 0.5>]}
      }})
    R3 Lime:LinearColor=&ColorFromCurve(ColorCurve=DataInterface() {propertyValues {
       AlphaCurve {keys [<at 0.2665 3>,<at 0.3905 0>,<at 0.5841 0>]}
       BlueCurve {keys [<at 0.6966 0>]}
       GreenCurve {keys [<at 0.6966 1>]}
       RedCurve {keys [<at 0.6966 0.5901>]}
      }})
    R3 Orange:LinearColor=&ColorFromCurve(ColorCurve=DataInterface() {propertyValues {
       AlphaCurve {keys [<at 0.2665 3>,<at 0.3905 0>,<at 0.5841 0>]}
       BlueCurve {keys [<at 0.6966 0.0304>]}
       GreenCurve {keys [<at 0.6966 0.2737>]}
       RedCurve {keys [<at 0.6966 0.7292>]}
      }})
    R3 Pink:LinearColor=&ColorFromCurve(ColorCurve=DataInterface() {propertyValues {
       AlphaCurve {keys [<at 0.2665 3>,<at 0.3905 0>,<at 0.5841 0>]}
       BlueCurve {keys [<at 0.6966 1.2188>]}
       GreenCurve {keys [<at 0.6966 0.2158>]}
       RedCurve {keys [<at 0.6966 1.1488>]}
      }})
    R3 Purple:LinearColor=&ColorFromCurve(ColorCurve=DataInterface() {propertyValues {
       AlphaCurve {keys [<at 0.2665 3>,<at 0.3905 0>,<at 0.5841 0>]}
       BlueCurve {keys [<at 0.6966 1>]}
       GreenCurve {keys [<at 0.6966 0>]}
       RedCurve {keys [<at 0.6966 0.3627>]}
      }})
    R3 Saffron:LinearColor=&ColorFromCurve(ColorCurve=DataInterface() {propertyValues {
       AlphaCurve {keys [<at 0.2665 3>,<at 0.3905 0>,<at 0.5841 0>]}
       BlueCurve {keys [<at 0.6966 0>]}
       GreenCurve {keys [<at 0.6966 1.8157>]}
       RedCurve {keys [<at 0.6966 2>]}
      }})
    R3 SkyBlue:LinearColor=&ColorFromCurve(ColorCurve=DataInterface() {propertyValues {
       AlphaCurve {keys [<at 0.2665 3>,<at 0.3905 0>,<at 0.5841 0>]}
       BlueCurve {keys [<at 0.6966 1.5>]}
       GreenCurve {keys [<at 0.6966 0.3855>]}
       RedCurve {keys [<at 0.6966 0>]}
      }})
    R3 TitaniumWhite:LinearColor=&ColorFromCurve(ColorCurve=DataInterface() {propertyValues {
       AlphaCurve {keys [<at 0.4952 1>,<at 0.9629 0>]}
       BlueCurve {keys [<at 0.6966 2.5>]}
       GreenCurve {keys [<at 0.6966 2.5>]}
       RedCurve {keys [<at 0.6966 2.5>]}
      }})
    R4 BurntSienna:LinearColor=<rgba 0.0312 0.0096 0.0039 0.6>
    R4 Cobalt:LinearColor=<rgba 0 0.0115 0.6198 0.7>
    R4 Crimson:LinearColor=<rgba 0.2656 0 0.0072 0.7>
    R4 ForestGreen:LinearColor=<rgba 0 0.2552 0.0129 0.6>
    R4 Grey:LinearColor=<rgba 0.0417 0.0417 0.0417 0.7>
    R4 Lime:LinearColor=<rgba 0.1875 0.375 0 0.6>
    R4 Orange:LinearColor=<rgba 0.5521 0.1706 0 0.45>
    R4 Pink:LinearColor=<rgba 0.3906 0.0468 0.2656 0.5>
    R4 Purple:LinearColor=<rgba 0.1567 0 0.7552 0.5>
    R4 Saffron:LinearColor=<rgba 0.5573 0.4299 0 0.3>
    R4 SkyBlue:LinearColor=<rgba 0.0137 0.2957 0.4375 0.5>
    R4 TitaniumWhite:LinearColor=<rgba 0.224 0.224 0.224 0.6>
    R5 Black:LinearColor=<rgba 0.0469 0.0469 0.0469 0.6>
    R5 BurntSienna:LinearColor=<rgba 0.1042 0.0322 0.0128 0.7>
    R5 Cobalt:LinearColor=<rgba 0.0312 0.0401 1 0.6>
    R5 Crimson:LinearColor=<rgba 0.4635 0.0266 0.0384 0.6>
    R5 DefaultColor:LinearColor=<rgba 1 1 1 0.5>
    R5 ForestGreen:LinearColor=<rgba 0.0704 0.901 0.1124 0.4>
    R5 Grey:LinearColor=<rgba 0.1667 0.1667 0.1667 0.4>
    R5 Lime:LinearColor=<rgba 0.2938 0.5104 0.0771 0.25>
    R5 Orange:LinearColor=<rgba 0.8261 0.2993 0.0172 0.3>
    R5 Pink:LinearColor=<rgba 0.651 0.1763 0.4784 0.4>
    R5 Purple:LinearColor=<rgba 0.2704 0.1155 0.7917 0.3>
    R5 Saffron:LinearColor=<rgba 0.8906 0.6824 0 0.15>
    R5 SkyBlue:LinearColor=<rgba 0.0754 0.3822 0.5365 0.4>
    R5 TitaniumWhite:LinearColor=<rgba 0.5729 0.5729 0.5729 0.3>
    Use R1 Painted Overrides=true
    Use R2 Painted Overrides=true
    Use R3 Painted Overrides=true
    Use R4 Painted Overrides=true
    Use R5 Painted Overrides=true
   ) {module=VC_PaintedVariantsOverride001}
   &ParticleState()
   &InheritSourceMovement(
    Applied Position Scale=&Multiply_VectorByFloat(
     Vector=<1 1 1>
     Float=&FloatFromCurve(FloatCurve=DataInterface() {propertyValues {Curve {keys [{interpMode=RCIM_Cubic, tangentMode=RCTM_Break, time=0, value=1}, {interpMode=RCIM_Cubic, tangentMode=RCTM_Break, time=1, value=0}]}}})
    )
    ApplyForce=false
    ApplyPosition=true
    ApplyPositionFalloff=false
   ) {module=InheritSourceMovement001}
   &SolveForcesAndVelocity(Clamp Velocity=false, Limit Acceleration=false)
   &VC_UpdateRibbon_v1(Enable Skeletal Position=true)
   /SetVariables(Particles.DotProduct:NiagaraFloat=&DotProduct(
     Vector A=&TransformVectorByMatrix(
      Vector:Vector3f=<1 0 0>
      Transformation Matrix:NiagaraMatrix=$Engine.Owner.SystemLocalToWorld
      Transform Vector as Position:NiagaraBool=false
     )
     Vector B=&NormalizeVector(
      Vector To Normalize=&SubtractVectorFromPosition(
       Position=$Engine.Owner.Position
       Vector=&ReturnCameraPosition(Coordinate Space:ENiagaraCoordinateSpace=::Simulation, Camera Query:NiagaraDataInterfaceCamera=DataInterface())
      )
      Safe Normalize=true
      Safe Fallback Threshold=0
      Safe Fallback Vector=<0 0 0.001>
     )
    )) {isSetParameters=true, module=SetVariables_#3}
   /SetVariables(
    Particles.RibbonBack:LinearColor=&ColorFromCurve(ColorCurve=DataInterface() {propertyValues {
       AlphaCurve {keys [<at 0.5096 0>,<at 0.6 2>,<at 0.6566 0>]}
       BlueCurve {keys [<at 0.9243 0>]}
       GreenCurve {keys [<at 0.9243 0>]}
       RedCurve {keys [<at 0.9243 0>]}
      }})
    Particles.RibbonBase:LinearColor=&ColorFromCurve(ColorCurve=DataInterface() {propertyValues {
       AlphaCurve {keys [<at 0.108 0>,<at 0.1426 1>,<at 0.31 1>,<at 0.4762 0>]}
       BlueCurve {keys [<at 0.9243 0>]}
       GreenCurve {keys [<at 0.9243 0>]}
       RedCurve {keys [<at 0.9243 0>]}
      }})
    Particles.RibbonGlow:LinearColor=&ColorFromCurve(ColorCurve=DataInterface() {propertyValues {
       AlphaCurve {keys [<at 0.2665 3>,<at 0.3905 0>,<at 0.5841 0>]}
       BlueCurve {keys [<at 0.6966 2>]}
       GreenCurve {keys [<at 0.6966 0.5201>]}
       RedCurve {keys [<at 0.6966 0>]}
      }})
    Particles.RibbonGradient:LinearColor=&ColorFromCurve(ColorCurve=DataInterface() {propertyValues {
       AlphaCurve {keys [
         <at 0.2819 7>
         <at 0.4191 0>
         <at 0.6302 0>
         <at 0.7212 1>
         {arriveTangent=-6.7757, interpMode=RCIM_Cubic, leaveTangent=-6.7757, time=0.7822, value=1}
         <at 0.8688 0>
        ]}
       BlueCurve {keys [<at 1 0.1683 RCIM_Cubic>]}
       GreenCurve {keys [<at 1 0 RCIM_Cubic>]}
       RedCurve {keys [<at 1 0.0865 RCIM_Cubic>]}
      }})
   ) {isSetParameters=true, module=SetVariables_#4}
   /SetVariables(
    Particles.R1_Color_Scaled:LinearColor=&MultiplyLinearColor(
     Linear Color=$Particles.R1 Color
     Scale Factor=&MakeVector4FromVectorAndFloat(
      "Vector (RGB)"=&VectorFromFloat(Value=&FloatFromCurve(
        FloatCurve=DataInterface() {propertyValues {Curve {keys [
            {arriveTangent=-13.0287, interpMode=RCIM_Cubic, time=0.2, value=1}
            <at 0.4191 0>
            <at 0.6302 0>
            <at 0.7212 1>
            {arriveTangent=-6.7757, interpMode=RCIM_Cubic, leaveTangent=-6.7757, time=0.7822, value=1}
            <at 0.8688 0>
           ]}}}
        Scale Curve=&MakeCustomFloatFromBool(Bool=$User.bIsPainted, True Float=$User.Power_EmissivePaintedR4, False Float=2)
       ))
      "Float (Alpha)"=&FloatFromCurve(
       FloatCurve=DataInterface() {propertyValues {Curve {keys [
           <at -0.0449 3.0504 RCIM_Cubic>
           {arriveTangent=0.1754, interpMode=RCIM_Cubic, leaveTangent=0.1754, time=0.111, value=0.7392}
           {arriveTangent=-0.2383, interpMode=RCIM_Cubic, leaveTangent=-0.2383, tangentMode=RCTM_User, time=0.4793, value=3.1423}
           {arriveTangent=-43.3151, interpMode=RCIM_Cubic, leaveTangent=-30.5159, tangentMode=RCTM_Break, time=0.675, value=3.1272}
           {arriveTangent=-9.9909, interpMode=RCIM_Cubic, leaveTangent=-9.9909, time=0.8329, value=0.2261}
           <at 1.01 -0.2199 RCIM_Cubic>
          ]}}}
       Scale Curve=&MakeCustomFloatFromBool(Bool=$User.bIsPainted, True Float=$User.Power_AlphaPaintedR4, False Float=0.5)
      )
     )
    )
    Particles.R3_Color_Scaled:LinearColor=&MultiplyLinearColor(
     Linear Color=$Particles.R3 Color
     Scale Factor=&MakeVector4FromVectorAndFloat(
      "Vector (RGB)"=&VectorFromFloat(Value=&FloatFromCurve(
        FloatCurve=DataInterface() {propertyValues {Curve {keys [<at 0.2665 1>,<at 0.3905 0>,<at 0.5841 0>]}}}
        Scale Curve=&MakeCustomFloatFromBool(Bool=$User.bIsPainted, True Float=$User.Power_EmissivePaintedR4, False Float=2)
       ))
      "Float (Alpha)"=&FloatFromCurve(
       FloatCurve=DataInterface() {propertyValues {Curve {keys [<at 0.2665 1>,<at 0.3905 0>,<at 0.5841 0>]}}}
       Scale Curve=&MakeCustomFloatFromBool(Bool=$User.bIsPainted, True Float=$User.Power_AlphaPaintedR4, False Float=0.5)
      )
     )
    )
    Particles.R4_Color_Scaled:LinearColor=&MultiplyLinearColor(
     Linear Color=$Particles.R4 Color
     Scale Factor=&MakeVector4FromVectorAndFloat(
      "Vector (RGB)"=&VectorFromFloat(Value=&FloatFromCurve(
        FloatCurve=DataInterface() {propertyValues {Curve {keys [
            <at -0.0135 1.393 RCIM_Cubic>
            {arriveTangent=0.2114, interpMode=RCIM_Cubic, leaveTangent=0.2114, time=0.1169, value=2.0317}
            {arriveTangent=-2.6168, interpMode=RCIM_Cubic, leaveTangent=-2.6168, tangentMode=RCTM_User, time=0.887, value=1.5834}
            {interpMode=RCIM_Cubic, leaveTangent=-0.8868, time=1.0581, value=1.0126}
           ]}}}
        Scale Curve=&MakeCustomFloatFromBool(Bool=$User.bIsPainted, True Float=$User.Power_EmissivePaintedR4, False Float=0.6)
       ))
      "Float (Alpha)"=&FloatFromCurve(
       FloatCurve=DataInterface() {propertyValues {Curve {keys [
           <at -0.0014 1.7739 RCIM_Cubic>
           {arriveTangent=-1.9854, interpMode=RCIM_Cubic, leaveTangent=-1.9854, time=0.3044, value=2.258}
           {arriveTangent=-0.2383, interpMode=RCIM_Cubic, leaveTangent=-1.8935, tangentMode=RCTM_Break, time=0.7964, value=0.1899}
           <at 1.4148 -1.634 RCIM_Cubic>
          ]}}}
       Scale Curve=&MakeCustomFloatFromBool(Bool=$User.bIsPainted, True Float=$User.Power_AlphaPaintedR4, False Float=1.2)
      )
     )
    )
    Particles.R5_Color_Scaled:LinearColor=&MultiplyLinearColor(
     Linear Color=$Particles.R5 Color
     Scale Factor=&MakeVector4FromVectorAndFloat(
      "Vector (RGB)"=&VectorFromFloat(Value=&FloatFromCurve(
        FloatCurve=DataInterface() {propertyValues {Curve {keys [
            <at -0.17 16.4711 RCIM_Cubic>
            {arriveTangent=-27.6507, interpMode=RCIM_Cubic, leaveTangent=-27.6507, time=0.1072, value=17.8443}
            {arriveTangent=-3.2592, interpMode=RCIM_Cubic, leaveTangent=-3.2592, tangentMode=RCTM_User, time=0.3385, value=2.4113}
            {arriveTangent=2.5648, interpMode=RCIM_Cubic, leaveTangent=2.5648, tangentMode=RCTM_User, time=0.64, value=1.973}
            {arriveTangent=11.7336, interpMode=RCIM_Cubic, leaveTangent=11.7336, time=1.0714, value=20.2083}
            {arriveTangent=-34.897, interpMode=RCIM_Cubic, leaveTangent=-34.897, time=1.3482, value=10.2825}
            {arriveTangent=-8.1043, interpMode=RCIM_Cubic, leaveTangent=-8.1043, time=1.5789, value=2.4972}
            <at 1.7789 6.7918 RCIM_Cubic>
           ]}}}
        Scale Curve=&MakeCustomFloatFromBool(Bool=$User.bIsPainted, True Float=$User.Power_EmissivePaintedR4, False Float=0.7)
       ))
      "Float (Alpha)"=&FloatFromCurve(
       FloatCurve=DataInterface() {propertyValues {Curve {keys [
           <at -0.0556 3.887 RCIM_Cubic>
           {arriveTangent=-1.3925, interpMode=RCIM_Cubic, leaveTangent=-1.3925, time=0.111, value=0.7392}
           {arriveTangent=-0.2383, interpMode=RCIM_Cubic, leaveTangent=-0.2383, tangentMode=RCTM_User, time=0.4793, value=3.1423}
           {arriveTangent=-6.3349, interpMode=RCIM_Cubic, leaveTangent=-6.3349, time=0.8329, value=0.2261}
           <at 1.01 -0.2199 RCIM_Cubic>
          ]}}}
       Scale Curve=&MakeCustomFloatFromBool(Bool=$User.bIsPainted, True Float=$User.Power_AlphaPaintedR4, False Float=0.3)
      )
     )
    )
   ) {isSetParameters=true, module=SetVariables_#8}
   &ScaleRibbonWidth(Ribbon Width Scale=&FloatFromCurve(FloatCurve=DataInterface() {propertyValues {Curve {keys [
         {arriveTangent=-0.2605, interpMode=RCIM_Cubic, leaveTangent=-0.2605, tangentMode=RCTM_User, time=0.0043, value=1.4579}
         {arriveTangent=-1.5709, interpMode=RCIM_Cubic, leaveTangent=-1.5709, tangentMode=RCTM_User, time=0.3137, value=1.1734}
         {arriveTangent=-2.7437, interpMode=RCIM_Cubic, leaveTangent=-2.7437, time=0.5043, value=0.5604}
         {arriveTangent=-1.126, interpMode=RCIM_Cubic, leaveTangent=-1.126, time=0.6133, value=0.3515}
         {arriveTangent=-1.0469, tangentMode=RCTM_User, time=1.002, value=0}
        ]}}}))
   /SetVariables(
    Particles.Ribbon_BlackColor:LinearColor=<rgba 0 0 0 1>
    Particles.WaveScaleCurve:NiagaraFloat=&FloatFromCurve(
     FloatCurve=DataInterface() {propertyValues {Curve {keys [
         {arriveTangent=39.5227, interpMode=RCIM_Cubic, leaveTangent=39.5227, tangentMode=RCTM_User, time=0.0092, value=0}
         {
          arriveTangent=0.7219
          arriveTangentWeight=0.0795
          interpMode=RCIM_Cubic
          leaveTangent=0.1806
          leaveTangentWeight=0.0696
          tangentMode=RCTM_Break
          tangentWeightMode=RCTWM_WeightedBoth
          time=0.125
          value=1.25
         }
         {arriveTangent=0.1988, interpMode=RCIM_Cubic, leaveTangent=0.1988, tangentMode=RCTM_User, time=0.45, value=0.625}
        ]}}}
     Scale Curve=10
    )
   ) {isSetParameters=true, module=SetVariables_#5}
   /SetVariables(
    Particles.CustomPosition_R1:NiagaraPosition=&AddVectorToPosition(
     Position=$Particles.Position
     Vector=&Multiply_VectorByFloat(
      Vector=&MakeVector(
       X=&Sine(
        Normalized Angle=&Add_Float(A=0, B=&Subtract_Float(A=$Emitter.NormalizedLoopAge, B=$Particles.NormalizedAge))
        Period=$User.WavePeriod
        Scale=&Multiply_Float(
         A=&SelectFloatFromArray(Direct Array Index=$Particles.SpawnGroup, Float Selection Array=DataInterface() {propertyValues {FloatData [1,-1]}})
         B=$User.WaveMaxScale
        )
        Bias=0
       )
       Y=&Sine(
        Normalized Angle=&Add_Float(A=0, B=&Subtract_Float(A=$Emitter.NormalizedLoopAge, B=$Particles.NormalizedAge))
        Period=$User.WavePeriod
        Scale=&Multiply_Float(
         A=&SelectFloatFromArray(Direct Array Index=$Particles.SpawnGroup, Float Selection Array=DataInterface() {propertyValues {FloatData [1,-1]}})
         B=$User.WaveMaxScale
        )
        Bias=0
       )
       Z=&Cosine(
        Normalized Angle=&Add_Float(A=0, B=&Subtract_Float(A=$Emitter.NormalizedLoopAge, B=$Particles.NormalizedAge))
        Period=$User.WavePeriod
        Scale=&Multiply_Float(
         A=&SelectFloatFromArray(Direct Array Index=$Particles.SpawnGroup, Float Selection Array=DataInterface() {propertyValues {FloatData [1,-1]}})
         B=$User.WaveMaxScale
        )
        Bias=0
       )
      )
      Float=$Particles.WaveScaleCurve
     )
    )
    Particles.CustomPosition_R2:NiagaraPosition=&AddVectorToPosition(
     Position=$Particles.Position
     Vector=&Multiply_VectorByFloat(
      Vector=&MakeVector(
       X=&Sine(
        Normalized Angle=&Add_Float(
         A=&Multiply_Float(A=$User.WavePeriod, B=0.5)
         B=&Subtract_Float(A=$Emitter.NormalizedLoopAge, B=$Particles.NormalizedAge)
        )
        Period=$User.WavePeriod
        Scale=&Multiply_Float(
         A=&SelectFloatFromArray(Direct Array Index=$Particles.SpawnGroup, Float Selection Array=DataInterface() {propertyValues {FloatData [1,1]}})
         B=$User.WaveMaxScale
        )
        Bias=0
       )
       Y=&Sine(
        Normalized Angle=&Add_Float(
         A=&Multiply_Float(A=$User.WavePeriod, B=0.5)
         B=&Subtract_Float(A=$Emitter.NormalizedLoopAge, B=$Particles.NormalizedAge)
        )
        Period=$User.WavePeriod
        Scale=&Multiply_Float(
         A=&SelectFloatFromArray(Direct Array Index=$Particles.SpawnGroup, Float Selection Array=DataInterface() {propertyValues {FloatData [1,1]}})
         B=$User.WaveMaxScale
        )
        Bias=0
       )
       Z=&Cosine(
        Normalized Angle=&Add_Float(
         A=&Multiply_Float(A=$User.WavePeriod, B=0.5)
         B=&Subtract_Float(A=$Emitter.NormalizedLoopAge, B=$Particles.NormalizedAge)
        )
        Period=$User.WavePeriod
        Scale=&Multiply_Float(
         A=&SelectFloatFromArray(Direct Array Index=$Particles.SpawnGroup, Float Selection Array=DataInterface() {propertyValues {FloatData [-1,1]}})
         B=$User.WaveMaxScale
        )
        Bias=0
       )
      )
      Float=$Particles.WaveScaleCurve
     )
    )
   ) {isSetParameters=true, module=SetVariables_#6}
   /SetVariables(Particles.DynamicMP_R2:Vector4f=&MakeVector4(
     X=&FloatFromCurve(FloatCurve=DataInterface() {propertyValues {Curve {
         defaultValue=1
         keys [
          {tangentMode=RCTM_User, time=0, value=0.5}
          {arriveTangent=-0.0633, interpMode=RCIM_Cubic, leaveTangent=-0.0633, tangentMode=RCTM_User, time=0.65, value=0.5}
          {arriveTangent=-0.0699, interpMode=RCIM_Cubic, leaveTangent=-0.0699, tangentMode=RCTM_User, time=0.75, value=-2}
         ]
        }}})
     Y=0
     Z=0
     W=0
    )) {isSetParameters=true, module=SetVariables_#7}
   &DynamicMaterialParameters(
    Index 0 Param 1=&FloatFromCurve(FloatCurve=DataInterface() {propertyValues {Curve {
        defaultValue=1
        keys [
         {
          arriveTangent=0.0405
          arriveTangentWeight=0.1835
          interpMode=RCIM_Cubic
          leaveTangent=0.0405
          leaveTangentWeight=0.1835
          tangentMode=RCTM_User
          time=0
          value=0
         }
         {
          arriveTangent=0.0201
          arriveTangentWeight=0.1834
          interpMode=RCIM_Cubic
          leaveTangent=0.0201
          leaveTangentWeight=0.1834
          tangentMode=RCTM_User
          time=0.5
          value=1
         }
        ]
       }}})
    Module.Index0Float0Write=true
    Module.Index0Float1Write=true
    Module.Index0Float2Write=true
    Module.Index0Float3Write=true
   )
  ]
  rendererClasses [&NiagaraRibbonRendererProperties]
  renderers [
   {rendererClass=&NiagaraRibbonRendererProperties}
   {rendererClass=&NiagaraRibbonRendererProperties, rendererIndex=1}
   {rendererClass=&NiagaraRibbonRendererProperties, rendererIndex=2}
   {rendererClass=&NiagaraRibbonRendererProperties, rendererIndex=3}
   {rendererClass=&NiagaraRibbonRendererProperties, rendererIndex=4}
  ]
 }
 NE_BlackFlame_Base1003 {
  emitterUpdateScript [
   &EmitterState(Life Cycle Mode=::NewEnumerator1, Loop Duration=1, UseLoopDelay=false)
   &VC_FilteredSocketCount()
   &VC_RateQualityHandler(Epic Spawn Rate=30, High Spawn Rate=30, Low Spawn Rate=30, Use Custom Rate?=true)
   &VC_SpawnRibbon_v1(Number of Ribbons=$Emitter.PortCount, Spawn Time=-1, SpawnRate=$Emitter.Spawn Rate)
  ]
  particleSpawnScript [
   &VC_PaintedVariantsOverride() {enabled=false}
   &VC_InitializeRibbon_v1(
    Color=$Particles.Color
    Custom Offset=true
    Enable Skeletal Position=true
    Lifetime=&MergeHelper_FloatSwitch(0=1, 1=1, 2=1, 3=?, Index=$User.GameModeID, Num=::NewEnumerator1)
    Local Vector Offset=&AddVectorToPosition(
     Position=&MergeHelper_VectorSwitch(0=<30 0 0>, 1=<200 0 0>, 2=<200 0 0>, 3=?, Index=$System.GameModeID, Num=::NewEnumerator1)
     Vector=$User.DTCarOffset
    )
    Offset Amount=1
    Ribbon Width=55
   )
   &AddVelocity(Velocity=<-2300 0 0>) {module=AddVelocity001}
  ]
  particleUpdateScript [
   &VC_PaintedVariantsOverride(
    R1 Black:LinearColor=&ColorFromCurve(ColorCurve=DataInterface() {propertyValues {
       AlphaCurve {
        defaultValue=7
        keys [
         <at 0.2819 7>
         <at 0.6302 1>
         <at 0.7212 1>
         {arriveTangent=-6.7193, interpMode=RCIM_Cubic, leaveTangent=-6.7193, time=0.7822, value=1}
         <at 0.87 0>
        ]
       }
       BlueCurve {keys [<at 1 0.1094 RCIM_Cubic>]}
       GreenCurve {keys [<at 1 0.1094 RCIM_Cubic>]}
       RedCurve {keys [<at 1 0.1094 RCIM_Cubic>]}
      }})
    R1 BurntSienna:LinearColor=&ColorFromCurve(ColorCurve=DataInterface() {propertyValues {
       AlphaCurve {
        defaultValue=7
        keys [
         <at 0.2819 7>
         <at 0.6302 1>
         <at 0.7212 1>
         {arriveTangent=-6.7193, interpMode=RCIM_Cubic, leaveTangent=-6.7193, time=0.7822, value=1}
         <at 0.87 0>
        ]
       }
       BlueCurve {keys [<at 1 0.0825 RCIM_Cubic>]}
       GreenCurve {keys [<at 1 0.0972 RCIM_Cubic>]}
       RedCurve {keys [<at 1 0.1667 RCIM_Cubic>]}
      }})
    R1 Cobalt:LinearColor=&ColorFromCurve(ColorCurve=DataInterface() {propertyValues {
       AlphaCurve {
        defaultValue=7
        keys [
         <at 0.2819 7>
         <at 0.6302 1>
         <at 0.7212 1>
         {arriveTangent=-6.7193, interpMode=RCIM_Cubic, leaveTangent=-6.7193, time=0.7822, value=1}
         <at 0.87 0>
        ]
       }
       BlueCurve {keys [<at 1 0.2344 RCIM_Cubic>]}
       GreenCurve {keys [<at 1 0.0574 RCIM_Cubic>]}
       RedCurve {keys [<at 1 0.0574 RCIM_Cubic>]}
      }})
    R1 Crimson:LinearColor=&ColorFromCurve(ColorCurve=DataInterface() {propertyValues {
       AlphaCurve {
        defaultValue=7
        keys [
         <at 0.2819 7>
         <at 0.6302 1>
         <at 0.7212 1>
         {arriveTangent=-6.7193, interpMode=RCIM_Cubic, leaveTangent=-6.7193, time=0.7822, value=1}
         <at 0.87 0>
        ]
       }
       BlueCurve {keys [<at 1 0.0131 RCIM_Cubic>]}
       GreenCurve {keys [<at 1 0 RCIM_Cubic>]}
       RedCurve {keys [<at 1 0.2348 RCIM_Cubic>]}
      }})
    R1 DefaultColor=&ColorFromCurve(ColorCurve=DataInterface() {propertyValues {
       AlphaCurve {keys [<at 0.2819 7>, <at 0.6302 1>, <at 0.7212 1>, {interpMode=RCIM_Cubic, leaveTangent=-6.7757, time=0.7822, value=1}, <at 0.87 0>]}
       BlueCurve {keys [<at 1 0.125 RCIM_Cubic>]}
       GreenCurve {keys [<at 1 0.0003 RCIM_Cubic>]}
       RedCurve {keys [<at 1 0.0441 RCIM_Cubic>]}
      }})
    R1 ForestGreen:LinearColor=&ColorFromCurve(ColorCurve=DataInterface() {propertyValues {
       AlphaCurve {
        defaultValue=7
        keys [
         <at 0.2819 7>
         <at 0.6302 1>
         <at 0.7212 1>
         {arriveTangent=-6.7193, interpMode=RCIM_Cubic, leaveTangent=-6.7193, time=0.7822, value=1}
         <at 0.87 0>
        ]
       }
       BlueCurve {keys [<at 0.6169 0.0198 RCIM_Cubic>]}
       GreenCurve {keys [<at 0.6169 0.2031 RCIM_Cubic>]}
       RedCurve {keys [<at 0.6169 0 RCIM_Cubic>]}
      }})
    R1 Grey:LinearColor=&ColorFromCurve(ColorCurve=DataInterface() {propertyValues {
       AlphaCurve {
        defaultValue=1
        keys [
         <at 0.2819 7>
         <at 0.6302 1>
         <at 0.7212 1>
         {arriveTangent=-6.7193, interpMode=RCIM_Cubic, leaveTangent=-6.7193, time=0.7822, value=1}
         <at 0.87 0>
        ]
       }
       BlueCurve {keys [<at 1 0.1979 RCIM_Cubic>]}
       GreenCurve {keys [<at 1 0.1979 RCIM_Cubic>]}
       RedCurve {keys [<at 1 0.1979 RCIM_Cubic>]}
      }})
    R1 Lime:LinearColor=&ColorFromCurve(ColorCurve=DataInterface() {propertyValues {
       AlphaCurve {
        defaultValue=7
        keys [
         <at 0.2819 7>
         <at 0.6302 1>
         <at 0.7212 1>
         {arriveTangent=-6.7193, interpMode=RCIM_Cubic, leaveTangent=-6.7193, time=0.7822, value=1}
         <at 0.87 0>
        ]
       }
       BlueCurve {keys [<at 1 0.0865 RCIM_Cubic>]}
       GreenCurve {keys [<at 1 0.2865 RCIM_Cubic>]}
       RedCurve {keys [<at 1 0.1697 RCIM_Cubic>]}
      }})
    R1 Orange:LinearColor=&ColorFromCurve(ColorCurve=DataInterface() {propertyValues {
       AlphaCurve {
        defaultValue=7
        keys [
         <at 0.2819 7>
         <at 0.6302 1>
         <at 0.7212 1>
         {arriveTangent=-6.7193, interpMode=RCIM_Cubic, leaveTangent=-6.7193, time=0.7822, value=1}
         <at 0.87 0>
        ]
       }
       BlueCurve {keys [<at 1 0.0622 RCIM_Cubic>]}
       GreenCurve {keys [<at 1 0.1169 RCIM_Cubic>]}
       RedCurve {keys [<at 1 0.3073 RCIM_Cubic>]}
      }})
    R1 Pink:LinearColor=&ColorFromCurve(ColorCurve=DataInterface() {propertyValues {
       AlphaCurve {
        defaultValue=7
        keys [
         <at 0.2819 7>
         <at 0.6302 1>
         <at 0.7212 1>
         {arriveTangent=-6.7193, interpMode=RCIM_Cubic, leaveTangent=-6.7193, time=0.7822, value=1}
         <at 0.87 0>
        ]
       }
       BlueCurve {keys [<at 1 0.2004 RCIM_Cubic>]}
       GreenCurve {keys [<at 1 0 RCIM_Cubic>]}
       RedCurve {keys [<at 1 0.2188 RCIM_Cubic>]}
      }})
    R1 Purple:LinearColor=&ColorFromCurve(ColorCurve=DataInterface() {propertyValues {
       AlphaCurve {
        defaultValue=7
        keys [
         <at 0.2819 7>
         <at 0.6302 1>
         <at 0.7212 1>
         {arriveTangent=-6.7193, interpMode=RCIM_Cubic, leaveTangent=-6.7193, time=0.7822, value=1}
         <at 0.87 0>
        ]
       }
       BlueCurve {keys [<at 1 0.375 RCIM_Cubic>]}
       GreenCurve {keys [<at 1 0 RCIM_Cubic>]}
       RedCurve {keys [<at 1 0.1484 RCIM_Cubic>]}
      }})
    R1 Saffron:LinearColor=&ColorFromCurve(ColorCurve=DataInterface() {propertyValues {
       AlphaCurve {
        defaultValue=7
        keys [
         <at 0.2819 7>
         <at 0.6302 1>
         <at 0.7212 1>
         {arriveTangent=-6.7193, interpMode=RCIM_Cubic, leaveTangent=-6.7193, time=0.7822, value=1}
         <at 0.87 0>
        ]
       }
       BlueCurve {keys [<at 1 0.0278 RCIM_Cubic>]}
       GreenCurve {keys [<at 1 0.2596 RCIM_Cubic>]}
       RedCurve {keys [<at 1 0.2812 RCIM_Cubic>]}
      }})
    R1 SkyBlue:LinearColor=&ColorFromCurve(ColorCurve=DataInterface() {propertyValues {
       AlphaCurve {
        defaultValue=7
        keys [
         <at 0.2819 7>
         <at 0.6302 1>
         <at 0.7212 1>
         {arriveTangent=-6.7193, interpMode=RCIM_Cubic, leaveTangent=-6.7193, time=0.7822, value=1}
         <at 0.87 0>
        ]
       }
       BlueCurve {keys [<at 1 0.375 RCIM_Cubic>]}
       GreenCurve {keys [<at 1 0.1761 RCIM_Cubic>]}
       RedCurve {keys [<at 1 0.0547 RCIM_Cubic>]}
      }})
    R1 TitaniumWhite:LinearColor=&ColorFromCurve(ColorCurve=DataInterface() {propertyValues {
       AlphaCurve {
        defaultValue=2
        keys [
         <at 0.2819 7>
         <at 0.6302 1>
         <at 0.7212 1>
         {arriveTangent=-6.7193, interpMode=RCIM_Cubic, leaveTangent=-6.7193, time=0.7822, value=1}
         <at 0.87 0>
        ]
       }
       BlueCurve {keys [<at 0.6118 0.3802 RCIM_Cubic>,<at 1 0.0677>]}
       GreenCurve {keys [<at 0.6118 0.3802 RCIM_Cubic>,<at 1 0.0677>]}
       RedCurve {keys [<at 0.6118 0.3802 RCIM_Cubic>,<at 1 0.0677>]}
      }})
    R2 Black:LinearColor=&ColorFromCurve(ColorCurve=DataInterface() {propertyValues {
       AlphaCurve {keys [<at 0 7>,<at 0.2232 1>,<at 0.3261 0.05>,<at 0.5495 0.2>,<at 0.7007 0.2>,<at 0.8474 0 RCIM_Cubic>]}
       BlueCurve {keys [<at 0 0.0194 RCIM_Cubic>,<at 0.1492 0.0333>,<at 0.2681 0.1104>,<at 0.4863 0.9274>,<at 0.6088 0.0305>]}
       GreenCurve {keys [<at 0 0.0194 RCIM_Cubic>,<at 0.1492 0.0333>,<at 0.2681 0.1104>,<at 0.4863 0.9274>,<at 0.6088 0.0305>]}
       RedCurve {keys [<at 0 0.0194 RCIM_Cubic>,<at 0.1492 0.0333>,<at 0.2681 0.1104>,<at 0.4863 0.9274>,<at 0.6088 0.0305>]}
      }})
    R2 BurntSienna:LinearColor=&ColorFromCurve(ColorCurve=DataInterface() {propertyValues {
       AlphaCurve {keys [<at 0 4>,<at 0.1907 1>,<at 0.3261 0.05>,<at 0.593 0.2>,<at 0.8474 0 RCIM_Cubic>]}
       BlueCurve {keys [<at 0 0.008 RCIM_Cubic>,<at 0.2839 0.0202>,<at 0.4863 0.2907>,<at 0.5956 0.008>]}
       GreenCurve {keys [<at 0 0.0203 RCIM_Cubic>,<at 0.2839 0.0323>,<at 0.4863 0.4394>,<at 0.5956 0.0203>]}
       RedCurve {keys [<at 0 0.0648 RCIM_Cubic>,<at 0.2839 0.0762>,<at 0.4863 0.9792>,<at 0.5956 0.0648>]}
      }})
    R2 Cobalt:LinearColor=&ColorFromCurve(ColorCurve=DataInterface() {propertyValues {
       AlphaCurve {keys [<at 0 7>,<at 0.1907 1>,<at 0.3261 0.05>,<at 0.5495 0.6>,<at 0.7007 0.2>,<at 0.8474 0 RCIM_Cubic>]}
       BlueCurve {keys [<at 0 0.6458 RCIM_Cubic>,<at 0.2575 0.9985>,<at 0.4863 1>,<at 0.6115 0.8177>]}
       GreenCurve {keys [<at 0 0.0119 RCIM_Cubic>,<at 0.2575 0.0367>,<at 0.4863 0.2997>,<at 0.6115 0.0527>]}
       RedCurve {keys [<at 0 0 RCIM_Cubic>,<at 0.2575 0.0186>,<at 0.4863 0.2865>,<at 0.6115 0.0383>]}
      }})
    R2 Crimson:LinearColor=&ColorFromCurve(ColorCurve=DataInterface() {propertyValues {
       AlphaCurve {keys [<at 0 7>,<at 0.1907 1>,<at 0.3261 0.05>,<at 0.5495 0.2>,<at 0.7007 0.2>,<at 0.8474 0 RCIM_Cubic>]}
       BlueCurve {keys [<at 0 0 RCIM_Cubic>,<at 0.1492 0.0104>,<at 0.4863 0.2995>,<at 0.6458 0.0894>]}
       GreenCurve {keys [<at 0 0 RCIM_Cubic>,<at 0.1492 0.0104>,<at 0.4863 0.2995>,<at 0.6458 0.0894>]}
       RedCurve {keys [<at 0 0.4323 RCIM_Cubic>,<at 0.1492 0.9992>,<at 0.4863 0.9274>,<at 0.6458 0.993>]}
      }})
    R2 DefaultColor:LinearColor=&ColorFromCurve(ColorCurve=DataInterface() {propertyValues {
       AlphaCurve {keys [<at 0 1>,<at 0.0914 0.2857>,<at 0.308 0.05>,<at 0.5592 0.2>,<at 0.8632 0>]}
       BlueCurve {keys [<at 0.0009 0>,<at 0.2668 -0.2829>,<at 0.5114 0.8849>,<at 0.619 1>,<at 0.682 0.383>]}
       GreenCurve {keys [<at 0.0009 0.0767>,<at 0.2742 -0.1994>,<at 0.5114 0.8849>,<at 0.619 1>,<at 0.682 0.2585>]}
       RedCurve {keys [<at 0.0009 3.2194>,<at 0.0016 3.2194>,<at 0.4479 0.8849>,<at 0.5114 0.8849>,<at 0.619 1>]}
      }})
    R2 ForestGreen:LinearColor=&ColorFromCurve(ColorCurve=DataInterface() {propertyValues {
       AlphaCurve {keys [<at 0 6>,<at 0.1907 1>,<at 0.3261 0.05>,<at 0.5495 0.2>,<at 0.7007 0.2>,<at 0.8474 0 RCIM_Cubic>]}
       BlueCurve {keys [<at 0 0.0172 RCIM_Cubic>,<at 0.2179 0.0322>,<at 0.4863 0.4656>,<at 0.6537 0.1027>]}
       GreenCurve {keys [<at 0 0.3438 RCIM_Cubic>,<at 0.2179 0.245>,<at 0.4863 0.875>,<at 0.6537 0.2987>]}
       RedCurve {keys [<at 0 0 RCIM_Cubic>,<at 0.2179 0.021>,<at 0.4863 0.4466>,<at 0.6537 0.0924>]}
      }})
    R2 Grey:LinearColor=&ColorFromCurve(ColorCurve=DataInterface() {propertyValues {
       AlphaCurve {keys [<at 0 4>,<at 0.1907 1>,<at 0.3261 0.05>,<at 0.5495 0.2>,<at 0.7007 0.2>,<at 0.8474 0 RCIM_Cubic>]}
       BlueCurve {keys [<at 0 0.0692 RCIM_Cubic>,<at 0.2285 0.1336>,<at 0.4863 0.9274>,<at 0.6405 0.132>]}
       GreenCurve {keys [<at 0 0.0781 RCIM_Cubic>,<at 0.2285 0.1336>,<at 0.4863 0.9274>,<at 0.6405 0.132>]}
       RedCurve {keys [<at 0 0.0767 RCIM_Cubic>,<at 0.2285 0.1406>,<at 0.4863 0.9274>,<at 0.6405 0.1354>]}
      }})
    R2 Lime:LinearColor=&ColorFromCurve(ColorCurve=DataInterface() {propertyValues {
       AlphaCurve {keys [<at 0 6>,<at 0.1907 1>,<at 0.3261 0.05>,<at 0.5495 0.2>,<at 0.8122 0 RCIM_Cubic>]}
       BlueCurve {keys [<at 0 0 RCIM_Cubic>,<at 0.2179 0.0459>,<at 0.4863 0.9274>,<at 0.6299 0.0724>]}
       GreenCurve {keys [<at 0 0.6875 RCIM_Cubic>,<at 0.2179 0.4635>,<at 0.4863 0.9274>,<at 0.6299 0.7708>]}
       RedCurve {keys [<at 0 0.3447 RCIM_Cubic>,<at 0.2179 0.2553>,<at 0.4863 0.9274>,<at 0.6299 0.4226>]}
      }})
    R2 Orange:LinearColor=&ColorFromCurve(ColorCurve=DataInterface() {propertyValues {
       AlphaCurve {keys [<at 0 7>,<at 0.1907 1>,<at 0.3261 0.05>,<at 0.5495 0.2>,<at 0.7007 0.2>,<at 0.8474 0 RCIM_Cubic>]}
       BlueCurve {keys [<at 0 0 RCIM_Cubic>,<at 0.2364 0.0598>,<at 0.4863 0.4263>,<at 0.6217 0.1191>]}
       GreenCurve {keys [<at 0 0.2834 RCIM_Cubic>,<at 0.2364 0.2792>,<at 0.4863 0.5719>,<at 0.6217 0.3121>]}
       RedCurve {keys [<at 0 0.8125 RCIM_Cubic>,<at 0.2364 0.6889>,<at 0.4863 0.8438>,<at 0.6217 0.6724>]}
      }})
    R2 Pink:LinearColor=&ColorFromCurve(ColorCurve=DataInterface() {propertyValues {
       AlphaCurve {keys [<at 0 7>,<at 0.1907 1>,<at 0.3261 0.05>,<at 0.5495 0.2>,<at 0.7007 0.2>,<at 0.8474 0 RCIM_Cubic>]}
       BlueCurve {keys [<at 0 0.3419 RCIM_Cubic>,<at 0.2707 0.3564>,<at 0.4863 0.7216>,<at 0.6141 0.5634>]}
       GreenCurve {keys [<at 0 0.0612 RCIM_Cubic>,<at 0.2707 0.1489>,<at 0.4863 0.5789>,<at 0.6141 0.2536>]}
       RedCurve {keys [<at 0 0.5029 RCIM_Cubic>,<at 0.2707 0.5614>,<at 0.4863 0.855>,<at 0.6141 0.8542>]}
      }})
    R2 Purple:LinearColor=&ColorFromCurve(ColorCurve=DataInterface() {propertyValues {
       AlphaCurve {keys [<at 0 7>,<at 0.1907 1>,<at 0.3261 0.05>,<at 0.5495 0.2>,<at 0.7007 0.2>,<at 0.8474 0 RCIM_Cubic>]}
       BlueCurve {keys [<at 0 0.8281 RCIM_Cubic>,<at 0.2919 0.6836>,<at 0.4863 0.8957>,<at 0.6141 0.9565>]}
       GreenCurve {keys [<at 0 0 RCIM_Cubic>,<at 0.2919 0.0563>,<at 0.4863 0.2791>,<at 0.6141 0.1097>]}
       RedCurve {keys [<at 0 0.1924 RCIM_Cubic>,<at 0.2919 0.3349>,<at 0.4863 0.5023>,<at 0.6141 0.3436>]}
      }})
    R2 Saffron:LinearColor=&ColorFromCurve(ColorCurve=DataInterface() {propertyValues {
       AlphaCurve {keys [<at 0 4>,<at 0.1907 1>,<at 0.3261 0.05>,<at 0.5495 0.2>,<at 0.8175 0 RCIM_Cubic>]}
       BlueCurve {keys [<at 0 0 RCIM_Cubic>,<at 0.2697 0.0046>,<at 0.4863 0.3188>,<at 0.6748 0.0798>]}
       GreenCurve {keys [<at 0 0.6105 RCIM_Cubic>,<at 0.2697 0.6881>,<at 0.4863 0.9274>,<at 0.6748 0.6052>]}
       RedCurve {keys [<at 0 0.7969 RCIM_Cubic>,<at 0.2697 0.8967>,<at 0.4863 0.8744>,<at 0.6748 0.7656>]}
      }})
    R2 SkyBlue:LinearColor=&ColorFromCurve(ColorCurve=DataInterface() {propertyValues {
       AlphaCurve {keys [<at 0 7>,<at 0.1907 1>,<at 0.3261 0.05>,<at 0.5495 0.2>,<at 0.8096 0 RCIM_Cubic>]}
       BlueCurve {keys [<at 0 0.5333 RCIM_Cubic>,<at 0.2496 0.5525>,<at 0.4863 0.9323>,<at 0.6141 0.9323>]}
       GreenCurve {keys [<at 0 0.3564 RCIM_Cubic>,<at 0.2496 0.3792>,<at 0.4863 0.8083>,<at 0.6141 0.6988>]}
       RedCurve {keys [<at 0 0 RCIM_Cubic>,<at 0.2496 0.0875>,<at 0.4863 0.5584>,<at 0.6141 0.2282>]}
      }})
    R2 TitaniumWhite:LinearColor=&ColorFromCurve(ColorCurve=DataInterface() {propertyValues {
       AlphaCurve {keys [<at 0 4>,<at 0.1907 1>,<at 0.3261 0.05>,<at 0.5495 0.2>,<at 0.7007 0.2>,<at 0.8175 0 RCIM_Cubic>]}
       BlueCurve {keys [<at 0 0.1846 RCIM_Cubic>,<at 0.2285 0.1336>,<at 0.4863 0.9274>,<at 0.6405 0.3248>]}
       GreenCurve {keys [<at 0 0.2083 RCIM_Cubic>,<at 0.2285 0.1336>,<at 0.4863 0.9274>,<at 0.6405 0.3248>]}
       RedCurve {keys [<at 0 0.2046 RCIM_Cubic>,<at 0.2285 0.1406>,<at 0.4863 0.9274>,<at 0.6405 0.3333>]}
      }})
    R3 Black:LinearColor=&ColorFromCurve(ColorCurve=DataInterface() {propertyValues {
       AlphaCurve {keys [{tangentMode=RCTM_User, time=0.8086, value=3}, {tangentMode=RCTM_User, time=0.8837, value=0}, {tangentMode=RCTM_User, time=1, value=0}]}
       BlueCurve {keys [<at 0.6966 0.25>]}
       GreenCurve {keys [<at 0.6966 0.25>]}
       RedCurve {keys [<at 0.6966 0.25>]}
      }})
    R3 BurntSienna:LinearColor=&ColorFromCurve(ColorCurve=DataInterface() {propertyValues {
       AlphaCurve {
        defaultValue=1
        keys [{tangentMode=RCTM_User, time=0.8086, value=3}, {tangentMode=RCTM_User, time=0.8837, value=0}, {tangentMode=RCTM_User, time=1, value=0}]
       }
       BlueCurve {keys [<at 0.6966 0.147>]}
       GreenCurve {keys [<at 0.6966 0.2197>]}
       RedCurve {keys [<at 0.6966 0.4479>]}
      }})
    R3 Cobalt:LinearColor=&ColorFromCurve(ColorCurve=DataInterface() {propertyValues {
       AlphaCurve {
        defaultValue=3
        keys [{tangentMode=RCTM_User, time=0.8086, value=3}, {tangentMode=RCTM_User, time=0.8837, value=0}, {tangentMode=RCTM_User, time=1, value=0}]
       }
       BlueCurve {keys [<at 0.6966 0.9688>]}
       GreenCurve {keys [<at 0.6966 0>]}
       RedCurve {keys [<at 0.6966 0>]}
      }})
    R3 Crimson:LinearColor=&ColorFromCurve(ColorCurve=DataInterface() {propertyValues {
       AlphaCurve {
        defaultValue=3
        keys [{tangentMode=RCTM_User, time=0.8086, value=3}, {tangentMode=RCTM_User, time=0.8837, value=0}, {tangentMode=RCTM_User, time=1, value=0}]
       }
       BlueCurve {keys [<at 0.6966 0>]}
       GreenCurve {keys [<at 0.6966 0>]}
       RedCurve {keys [<at 0.6966 1>]}
      }})
    R3 DefaultColor:LinearColor=&ColorFromCurve(ColorCurve=DataInterface() {propertyValues {
       AlphaCurve {keys [<at 0.8086 3>,<at 0.8837 0>,<at 1 0>]}
       BlueCurve {keys [<at 0.6966 1.25>]}
       GreenCurve {keys [<at 0.6966 0.3251>]}
       RedCurve {keys [<at 0.6966 0>]}
      }})
    R3 ForestGreen:LinearColor=&ColorFromCurve(ColorCurve=DataInterface() {propertyValues {
       AlphaCurve {
        defaultValue=3
        keys [{tangentMode=RCTM_User, time=0.8086, value=3}, {tangentMode=RCTM_User, time=0.8837, value=0}, {tangentMode=RCTM_User, time=1, value=0}]
       }
       BlueCurve {keys [<at 0.6966 0.2593>]}
       GreenCurve {keys [<at 0.6966 0.6562>]}
       RedCurve {keys [<at 0.6966 0.1436>]}
      }})
    R3 Grey:LinearColor=&ColorFromCurve(ColorCurve=DataInterface() {propertyValues {
       AlphaCurve {
        defaultValue=3
        keys [{tangentMode=RCTM_User, time=0.8086, value=3}, {tangentMode=RCTM_User, time=0.8837, value=0}, {tangentMode=RCTM_User, time=1, value=0}]
       }
       BlueCurve {keys [<at 0.6966 0.3646>]}
       GreenCurve {keys [<at 0.6966 0.3646>]}
       RedCurve {keys [<at 0.6966 0.3646>]}
      }})
    R3 Lime:LinearColor=&ColorFromCurve(ColorCurve=DataInterface() {propertyValues {
       AlphaCurve {
        defaultValue=3
        keys [{tangentMode=RCTM_User, time=0.8086, value=3}, {tangentMode=RCTM_User, time=0.8837, value=0}, {tangentMode=RCTM_User, time=1, value=0}]
       }
       BlueCurve {keys [<at 0.6966 0>]}
       GreenCurve {keys [<at 0.6966 1>]}
       RedCurve {keys [<at 0.6966 0.5901>]}
      }})
    R3 Orange:LinearColor=&ColorFromCurve(ColorCurve=DataInterface() {propertyValues {
       AlphaCurve {
        defaultValue=3
        keys [{tangentMode=RCTM_User, time=0.8086, value=3}, {tangentMode=RCTM_User, time=0.8837, value=0}, {tangentMode=RCTM_User, time=1, value=0}]
       }
       BlueCurve {keys [<at 0.6966 0.1367>]}
       GreenCurve {keys [<at 0.6966 0.2734>]}
       RedCurve {keys [<at 0.6966 0.7292>]}
      }})
    R3 Pink:LinearColor=&ColorFromCurve(ColorCurve=DataInterface() {propertyValues {
       AlphaCurve {
        defaultValue=3
        keys [{tangentMode=RCTM_User, time=0.8086, value=3}, {tangentMode=RCTM_User, time=0.8837, value=0}, {tangentMode=RCTM_User, time=1, value=0}]
       }
       BlueCurve {keys [<at 0.6966 0.6855>]}
       GreenCurve {keys [<at 0.6966 0.1214>]}
       RedCurve {keys [<at 0.6966 0.6462>]}
      }})
    R3 Purple:LinearColor=&ColorFromCurve(ColorCurve=DataInterface() {propertyValues {
       AlphaCurve {
        defaultValue=3
        keys [{tangentMode=RCTM_User, time=0.8086, value=3}, {tangentMode=RCTM_User, time=0.8837, value=0}, {tangentMode=RCTM_User, time=1, value=0}]
       }
       BlueCurve {keys [<at 0.6966 0.8906>]}
       GreenCurve {keys [<at 0.6966 0>]}
       RedCurve {keys [<at 0.6966 0.2841>]}
      }})
    R3 Saffron:LinearColor=&ColorFromCurve(ColorCurve=DataInterface() {propertyValues {
       AlphaCurve {
        defaultValue=3
        keys [{tangentMode=RCTM_User, time=0.8086, value=3}, {tangentMode=RCTM_User, time=0.8837, value=0}, {tangentMode=RCTM_User, time=1, value=0}]
       }
       BlueCurve {keys [<at 0.6966 0.199>]}
       GreenCurve {keys [<at 0.6966 1.2571>]}
       RedCurve {keys [<at 0.6966 1.3646>]}
      }})
    R3 SkyBlue:LinearColor=&ColorFromCurve(ColorCurve=DataInterface() {propertyValues {
       AlphaCurve {
        defaultValue=3
        keys [{tangentMode=RCTM_User, time=0.8086, value=3}, {tangentMode=RCTM_User, time=0.8837, value=0}, {tangentMode=RCTM_User, time=1, value=0}]
       }
       BlueCurve {keys [<at 0.6966 0.9375>]}
       GreenCurve {keys [<at 0.6966 0.4441>]}
       RedCurve {keys [<at 0.6966 0>]}
      }})
    R3 TitaniumWhite:LinearColor=&ColorFromCurve(ColorCurve=DataInterface() {propertyValues {
       AlphaCurve {keys [<at 0.4952 1>,<at 0.9629 0>]}
       BlueCurve {keys [<at 0.6966 2.5>]}
       GreenCurve {keys [<at 0.6966 2.5>]}
       RedCurve {keys [<at 0.6966 2.5>]}
      }})
    R4 Black:LinearColor=&ColorFromCurve(ColorCurve=DataInterface() {propertyValues {
       AlphaCurve {keys [<at 0.0115 0.9906>,<at 0.3154 1>,<at 0.5062 0>]}
       BlueCurve {keys [<at 0.3014 0>]}
       GreenCurve {keys [<at 0.3014 0>]}
       RedCurve {keys [<at 0.3014 0>]}
      }})
    R4 BurntSienna:LinearColor=&ColorFromCurve(ColorCurve=DataInterface() {propertyValues {
       AlphaCurve {keys [<at 0.0115 0.9906>,<at 0.3154 1>,<at 0.5062 0>]}
       BlueCurve {keys [<at 0.3014 0>]}
       GreenCurve {keys [<at 0.3014 0>]}
       RedCurve {keys [<at 0.3014 0>]}
      }})
    R4 Cobalt:LinearColor=&ColorFromCurve(ColorCurve=DataInterface() {propertyValues {
       AlphaCurve {keys [<at 0.0115 0.9906>,<at 0.3154 1>,<at 0.5062 0>]}
       BlueCurve {keys [<at 0.3014 0>]}
       GreenCurve {keys [<at 0.3014 0>]}
       RedCurve {keys [<at 0.3014 0>]}
      }})
    R4 Crimson:LinearColor=&ColorFromCurve(ColorCurve=DataInterface() {propertyValues {
       AlphaCurve {keys [<at 0.0115 0.9906>,<at 0.3154 1>,<at 0.5062 0>]}
       BlueCurve {keys [<at 0.0688 0.069>,<at 0.2065 0.0676>,<at 0.3014 0>]}
       GreenCurve {keys [<at 0.0688 0.069>,<at 0.2065 0.0676>,<at 0.3014 0>]}
       RedCurve {keys [<at 0.0688 0.25>,<at 0.2065 0.2448>,<at 0.3014 0>]}
      }})
    R4 DefaultColor:LinearColor=&ColorFromCurve(ColorCurve=DataInterface() {propertyValues {
       AlphaCurve {keys [<at 0.0115 0.9906>,<at 0.3154 1>,<at 0.5062 0>]}
       BlueCurve {keys [<at 0.1614 0.2917>,<at 0.3014 0>]}
       GreenCurve {keys [<at 0.1614 0.0182>,<at 0.3014 0>]}
       RedCurve {keys [<at 0.1614 0.1445>,<at 0.3014 0>]}
      }})
    R4 ForestGreen:LinearColor=&ColorFromCurve(ColorCurve=DataInterface() {propertyValues {
       AlphaCurve {keys [<at 0.0115 0.9906>,<at 0.3154 1>,<at 0.5062 0>]}
       BlueCurve {keys [<at 0.229 0.0107>,<at 0.4942 0>]}
       GreenCurve {keys [<at 0.229 0.2031>,<at 0.4942 0>]}
       RedCurve {keys [<at 0.229 0>,<at 0.4942 0>]}
      }})
    R4 Grey:LinearColor=&ColorFromCurve(ColorCurve=DataInterface() {propertyValues {
       AlphaCurve {keys [<at 0.0115 0.9906>,<at 0.3154 1>,<at 0.5062 0>]}
       BlueCurve {keys [<at 0.1914 0.1875>,<at 0.3014 0>]}
       GreenCurve {keys [<at 0.1914 0.1875>,<at 0.3014 0>]}
       RedCurve {keys [<at 0.1914 0.1875>,<at 0.3014 0>]}
      }})
    R4 Lime:LinearColor=&ColorFromCurve(ColorCurve=DataInterface() {propertyValues {
       AlphaCurve {keys [<at 0.0115 0.9906>,<at 0.3154 1>,<at 0.5062 0>]}
       BlueCurve {keys [<at 0.3014 0.0328>,<at 0.4842 0.0068>]}
       GreenCurve {keys [<at 0.3014 0.151>,<at 0.4842 0.0312>]}
       RedCurve {keys [<at 0.3014 0.0925>,<at 0.4842 0.0191>]}
      }})
    R4 Orange:LinearColor=&ColorFromCurve(ColorCurve=DataInterface() {propertyValues {
       AlphaCurve {keys [<at 0.0115 0.9906>,<at 0.3154 1>,<at 0.5062 0>]}
       BlueCurve {keys [<at 0.2565 0.0204>,<at 0.4842 0>]}
       GreenCurve {keys [<at 0.2565 0.0617>,<at 0.4842 0>]}
       RedCurve {keys [<at 0.2565 0.276>,<at 0.4842 0>]}
      }})
    R4 Pink:LinearColor=&ColorFromCurve(ColorCurve=DataInterface() {propertyValues {
       AlphaCurve {keys [<at 0.0115 0.9906>,<at 0.3154 1>,<at 0.5062 0>]}
       BlueCurve {keys [<at 0.279 0.6169>,<at 0.4817 0>]}
       GreenCurve {keys [<at 0.279 0.087>,<at 0.4817 0>]}
       RedCurve {keys [<at 0.279 0.6198>,<at 0.4817 0>]}
      }})
    R4 Purple:LinearColor=&ColorFromCurve(ColorCurve=DataInterface() {propertyValues {
       AlphaCurve {keys [<at 0.0115 0.9906>,<at 0.3154 1>,<at 0.5062 0>]}
       BlueCurve {keys [<at 0.3014 0>]}
       GreenCurve {keys [<at 0.3014 0>]}
       RedCurve {keys [<at 0.3014 0>]}
      }})
    R4 Saffron:LinearColor=&ColorFromCurve(ColorCurve=DataInterface() {propertyValues {
       AlphaCurve {keys [<at 0.0115 0.9906>,<at 0.3154 1>,<at 0.5062 0>]}
       BlueCurve {keys [<at 0.3014 0>]}
       GreenCurve {keys [<at 0.3014 0>]}
       RedCurve {keys [<at 0.3014 0>]}
      }})
    R4 SkyBlue:LinearColor=&ColorFromCurve(ColorCurve=DataInterface() {propertyValues {
       AlphaCurve {keys [<at 0.0115 0.9906>,<at 0.3154 1>,<at 0.5062 0>]}
       BlueCurve {keys [<at 0.3014 0>]}
       GreenCurve {keys [<at 0.3014 0>]}
       RedCurve {keys [<at 0.3014 0>]}
      }})
    R4 TitaniumWhite:LinearColor=&ColorFromCurve(ColorCurve=DataInterface() {propertyValues {
       AlphaCurve {keys [<at 0.0115 0.9906>,<at 0.1682 1>,<at 0.3697 0>]}
       BlueCurve {keys [<at 0.3509 0.2708>,<at 0.4792 0.0521>]}
       GreenCurve {keys [<at 0.3509 0.2708>,<at 0.4792 0.0521>]}
       RedCurve {keys [<at 0.3509 0.2708>,<at 0.4792 0.0521>]}
      }})
    R5 Black:LinearColor=<rgba 0.0469 0.0469 0.0469 0.6>
    R5 BurntSienna:LinearColor=<rgba 0.1042 0.0322 0.0128 0.7>
    R5 Cobalt:LinearColor=<rgba 0.0312 0.0401 1 0.6>
    R5 Crimson:LinearColor=<rgba 0.4635 0.0266 0.0384 0.6>
    R5 DefaultColor:LinearColor=<rgba 1 1 1 0.5>
    R5 ForestGreen:LinearColor=<rgba 0.0704 0.901 0.1124 0.4>
    R5 Grey:LinearColor=<rgba 0.1667 0.1667 0.1667 0.4>
    R5 Lime:LinearColor=<rgba 0.2938 0.5104 0.0771 0.25>
    R5 Orange:LinearColor=<rgba 0.8261 0.2993 0.0172 0.3>
    R5 Pink:LinearColor=<rgba 0.651 0.1763 0.4784 0.4>
    R5 Purple:LinearColor=<rgba 0.2704 0.1155 0.7917 0.3>
    R5 Saffron:LinearColor=<rgba 0.8906 0.6824 0 0.15>
    R5 SkyBlue:LinearColor=<rgba 0.0754 0.3822 0.5365 0.4>
    R5 TitaniumWhite:LinearColor=<rgba 0.5729 0.5729 0.5729 0.3>
    Use R1 Painted Overrides=true
    Use R2 Painted Overrides=true
    Use R3 Painted Overrides=true
    Use R4 Painted Overrides=true
    Use R5 Painted Overrides=true
   ) {module=VC_PaintedVariantsOverride001}
   &ParticleState()
   &InheritSourceMovement(
    Applied Position Scale=&Multiply_VectorByFloat(
     Vector=<1 1 1>
     Float=&FloatFromCurve(FloatCurve=DataInterface() {propertyValues {Curve {keys [{interpMode=RCIM_Cubic, tangentMode=RCTM_Break, time=0, value=1}, {interpMode=RCIM_Cubic, tangentMode=RCTM_Break, time=1, value=0}]}}})
    )
    ApplyForce=false
    ApplyPosition=true
    ApplyPositionFalloff=false
   ) {module=InheritSourceMovement001}
   &SolveForcesAndVelocity(Clamp Velocity=false, Limit Acceleration=false)
   &VC_UpdateRibbon_v1(Enable Skeletal Position=true)
   /SetVariables(Particles.DotProduct:NiagaraFloat=&DotProduct(
     Vector A=&TransformVectorByMatrix(
      Vector:Vector3f=<1 0 0>
      Transformation Matrix:NiagaraMatrix=$Engine.Owner.SystemLocalToWorld
      Transform Vector as Position:NiagaraBool=false
     )
     Vector B=&NormalizeVector(
      Vector To Normalize=&SubtractVectorFromPosition(
       Position=$Engine.Owner.Position
       Vector=&ReturnCameraPosition(Coordinate Space:ENiagaraCoordinateSpace=::Simulation, Camera Query:NiagaraDataInterfaceCamera=DataInterface())
      )
      Safe Normalize=true
      Safe Fallback Threshold=0
      Safe Fallback Vector=<0 0 0.001>
     )
    )) {isSetParameters=true, module=SetVariables_#3}
   /SetVariables(
    Particles.RibbonBack:LinearColor=&ColorFromCurve(ColorCurve=DataInterface() {propertyValues {
       AlphaCurve {keys [<at 0.5096 0>,<at 0.6 2>,<at 0.6566 0>]}
       BlueCurve {keys [<at 0.9243 0>]}
       GreenCurve {keys [<at 0.9243 0>]}
       RedCurve {keys [<at 0.9243 0>]}
      }})
    Particles.RibbonBase:LinearColor=&ColorFromCurve(ColorCurve=DataInterface() {propertyValues {
       AlphaCurve {keys [<at 0.1012 0>,<at 0.1163 1>,<at 0.2485 0.92>,<at 0.3857 1>,<at 0.4762 0>]}
       BlueCurve {keys [<at 0.9243 0>]}
       GreenCurve {keys [<at 0.9243 0>]}
       RedCurve {keys [<at 0.9243 0>]}
      }})
   ) {isSetParameters=true, module=SetVariables_#4}
   /SetVariables(
    Particles.R1_Color_Scaled:LinearColor=&MultiplyLinearColor(
     Linear Color=$Particles.R1 Color
     Scale Factor=&MakeVector4FromVectorAndFloat(
      "Vector (RGB)"=&VectorFromFloat(Value=&FloatFromCurve(
        FloatCurve=DataInterface() {propertyValues {Curve {keys [
            {arriveTangent=-13.0287, interpMode=RCIM_Cubic, time=0.2667, value=0.9059}
            {arriveTangent=-1.5185, tangentMode=RCTM_User, time=0.4953, value=0.8118}
            <at 0.6886 0>
            <at 0.7212 1>
            {arriveTangent=-6.7757, interpMode=RCIM_Cubic, leaveTangent=-6.7757, time=0.7822, value=1}
            <at 0.8688 0>
           ]}}}
        Scale Curve=&MakeCustomFloatFromBool(Bool=$User.bIsPainted, True Float=$User.Power_EmissivePaintedR4, False Float=2)
       ))
      "Float (Alpha)"=&FloatFromCurve(
       FloatCurve=DataInterface() {propertyValues {Curve {keys [
           <at -0.0383 3.4294 RCIM_Cubic>
           {arriveTangent=-0.5546, interpMode=RCIM_Cubic, leaveTangent=-0.5546, time=0.1518, value=2.4766}
           {arriveTangent=-0.2383, interpMode=RCIM_Cubic, leaveTangent=-0.2383, tangentMode=RCTM_User, time=0.4793, value=3.1423}
           {arriveTangent=-43.3151, interpMode=RCIM_Cubic, leaveTangent=-30.5159, tangentMode=RCTM_Break, time=0.675, value=3.1272}
           {arriveTangent=-9.9908, interpMode=RCIM_Cubic, leaveTangent=-9.9908, time=0.8263, value=1.0954}
           <at 1.01 -0.2199 RCIM_Cubic>
          ]}}}
       Scale Curve=&MakeCustomFloatFromBool(Bool=$User.bIsPainted, True Float=$User.Power_AlphaPaintedR4, False Float=0.5)
      )
     )
    )
    Particles.R3_Color_Scaled:LinearColor=&MultiplyLinearColor(
     Linear Color=$Particles.R3 Color
     Scale Factor=&MakeVector4FromVectorAndFloat(
      "Vector (RGB)"=&VectorFromFloat(Value=&FloatFromCurve(
        FloatCurve=DataInterface() {propertyValues {Curve {keys [<at 0.3572 0.8681>,<at 0.49 0.675>,<at 0.6121 0.1761>,<at 0.7769 0>]}}}
        Scale Curve=&MakeCustomFloatFromBool(Bool=$User.bIsPainted, True Float=$User.Power_EmissivePaintedR4, False Float=2)
       ))
      "Float (Alpha)"=&FloatFromCurve(
       FloatCurve=DataInterface() {propertyValues {Curve {keys [<at 0.2665 1>,<at 0.4676 0.5559>,<at 0.5805 0.7676>,<at 0.7 0>]}}}
       Scale Curve=&MakeCustomFloatFromBool(Bool=$User.bIsPainted, True Float=$User.Power_AlphaPaintedR4, False Float=0.5)
      )
     )
    )
    Particles.R4_Color_Scaled:LinearColor=&MultiplyLinearColor(
     Linear Color=$Particles.R4 Color
     Scale Factor=&MakeVector4FromVectorAndFloat(
      "Vector (RGB)"=&VectorFromFloat(Value=&FloatFromCurve(
        FloatCurve=DataInterface() {propertyValues {Curve {keys [<at 0.0107 1>,<at 0.0753 1>,<at 0.3572 0.8681>,<at 0.49 1>,<at 0.6121 1>,<at 0.7769 0>]}}}
        Scale Curve=&MakeCustomFloatFromBool(Bool=$User.bIsPainted, True Float=$User.Power_EmissivePaintedR4, False Float=0.6)
       ))
      "Float (Alpha)"=&FloatFromCurve(
       FloatCurve=DataInterface() {propertyValues {Curve {keys [<at 0.106 0>,<at 0.1223 1.0029>,<at 0.2665 1>,<at 0.4676 0.5559>,<at 0.5805 0.7676>,<at 0.7 0>]}}}
       Scale Curve=&MakeCustomFloatFromBool(Bool=$User.bIsPainted, True Float=$User.Power_AlphaPaintedR4, False Float=1.2)
      )
     )
    )
    Particles.R5_Color_Scaled:LinearColor=&MultiplyLinearColor(
     Linear Color=$Particles.R5 Color
     Scale Factor=&MakeVector4FromVectorAndFloat(
      "Vector (RGB)"=&VectorFromFloat(Value=&FloatFromCurve(
        FloatCurve=DataInterface() {propertyValues {Curve {keys [
            <at -0.17 16.4711 RCIM_Cubic>
            {arriveTangent=-27.6507, interpMode=RCIM_Cubic, leaveTangent=-27.6507, time=0.1072, value=17.8443}
            {arriveTangent=-3.2592, interpMode=RCIM_Cubic, leaveTangent=-3.2592, tangentMode=RCTM_User, time=0.3385, value=2.4113}
            {arriveTangent=2.5648, interpMode=RCIM_Cubic, leaveTangent=2.5648, tangentMode=RCTM_User, time=0.64, value=1.973}
            {arriveTangent=11.7336, interpMode=RCIM_Cubic, leaveTangent=11.7336, time=1.0714, value=20.2083}
            {arriveTangent=-34.897, interpMode=RCIM_Cubic, leaveTangent=-34.897, time=1.3482, value=10.2825}
            {arriveTangent=-8.1043, interpMode=RCIM_Cubic, leaveTangent=-8.1043, time=1.5789, value=2.4972}
            <at 1.7789 6.7918 RCIM_Cubic>
           ]}}}
        Scale Curve=&MakeCustomFloatFromBool(Bool=$User.bIsPainted, True Float=$User.Power_EmissivePaintedR4, False Float=0.7)
       ))
      "Float (Alpha)"=&FloatFromCurve(
       FloatCurve=DataInterface() {propertyValues {Curve {keys [
           <at -0.0556 3.887 RCIM_Cubic>
           {arriveTangent=-1.3925, interpMode=RCIM_Cubic, leaveTangent=-1.3925, time=0.111, value=0.7392}
           {arriveTangent=-0.2383, interpMode=RCIM_Cubic, leaveTangent=-0.2383, tangentMode=RCTM_User, time=0.4793, value=3.1423}
           {arriveTangent=-6.3349, interpMode=RCIM_Cubic, leaveTangent=-6.3349, time=0.8329, value=0.2261}
           <at 1.01 -0.2199 RCIM_Cubic>
          ]}}}
       Scale Curve=&MakeCustomFloatFromBool(Bool=$User.bIsPainted, True Float=$User.Power_AlphaPaintedR4, False Float=0.3)
      )
     )
    )
   ) {isSetParameters=true, module=SetVariables_#8}
   &ScaleRibbonWidth(Ribbon Width Scale=&FloatFromCurve(FloatCurve=DataInterface() {propertyValues {Curve {keys [
         {arriveTangent=-0.2605, interpMode=RCIM_Cubic, leaveTangent=-0.2605, tangentMode=RCTM_User, time=0.0016, value=1.2864}
         {arriveTangent=-0.5099, interpMode=RCIM_Cubic, leaveTangent=-0.5099, tangentMode=RCTM_User, time=0.3641, value=1.3272}
         {arriveTangent=-5.6426, interpMode=RCIM_Cubic, leaveTangent=-5.6426, tangentMode=RCTM_User, time=0.4564, value=0.8318}
         {arriveTangent=-1.5491, interpMode=RCIM_Cubic, leaveTangent=-1.5491, tangentMode=RCTM_User, time=0.5261, value=0.6092}
         {arriveTangent=-1.2369, interpMode=RCIM_Cubic, leaveTangent=-1.2369, tangentMode=RCTM_User, time=0.71, value=0.4133}
         {arriveTangent=-1.0469, tangentMode=RCTM_User, time=1.002, value=0}
        ]}}}))
   /SetVariables(Particles.R1_Width:NiagaraFloat=&FloatFromCurve(FloatCurve=DataInterface() {propertyValues {Curve {keys [<at 0 75>,<at 0.274 75>,<at 0.3773 54.9632>,<at 0.4815 48.6724>,<at 0.5921 55.788>,<at 0.7677 41.4558>]}}})) {isSetParameters=true}
   /SetVariables(
    Particles.Ribbon_BlackColor:LinearColor=<rgba 0 0 0 1>
    Particles.WaveScaleCurve:NiagaraFloat=&FloatFromCurve(
     FloatCurve=DataInterface() {propertyValues {Curve {keys [
         {arriveTangent=39.5227, interpMode=RCIM_Cubic, leaveTangent=39.5227, tangentMode=RCTM_User, time=0.0092, value=0}
         {
          arriveTangent=0.7219
          arriveTangentWeight=0.0795
          interpMode=RCIM_Cubic
          leaveTangent=0.1806
          leaveTangentWeight=0.0696
          tangentMode=RCTM_Break
          tangentWeightMode=RCTWM_WeightedBoth
          time=0.125
          value=1.25
         }
         {arriveTangent=0.1988, interpMode=RCIM_Cubic, leaveTangent=0.1988, tangentMode=RCTM_User, time=0.45, value=0.625}
        ]}}}
     Scale Curve=10
    )
   ) {isSetParameters=true, module=SetVariables_#5}
   /SetVariables(
    Particles.CustomPosition_R1:NiagaraPosition=&AddVectorToPosition(
     Position=$Particles.Position
     Vector=&Multiply_VectorByFloat(
      Vector=&MakeVector(
       X=&Sine(
        Normalized Angle=&Add_Float(A=0, B=&Subtract_Float(A=$Emitter.NormalizedLoopAge, B=$Particles.NormalizedAge))
        Period=$User.WavePeriod
        Scale=&Multiply_Float(
         A=&SelectFloatFromArray(Direct Array Index=$Particles.SpawnGroup, Float Selection Array=DataInterface() {propertyValues {FloatData [1,-1]}})
         B=$User.WaveMaxScale
        )
        Bias=0
       )
       Y=&Sine(
        Normalized Angle=&Add_Float(A=0, B=&Subtract_Float(A=$Emitter.NormalizedLoopAge, B=$Particles.NormalizedAge))
        Period=$User.WavePeriod
        Scale=&Multiply_Float(
         A=&SelectFloatFromArray(Direct Array Index=$Particles.SpawnGroup, Float Selection Array=DataInterface() {propertyValues {FloatData [1,-1]}})
         B=$User.WaveMaxScale
        )
        Bias=0
       )
       Z=&Cosine(
        Normalized Angle=&Add_Float(A=0, B=&Subtract_Float(A=$Emitter.NormalizedLoopAge, B=$Particles.NormalizedAge))
        Period=$User.WavePeriod
        Scale=&Multiply_Float(
         A=&SelectFloatFromArray(Direct Array Index=$Particles.SpawnGroup, Float Selection Array=DataInterface() {propertyValues {FloatData [1,-1]}})
         B=$User.WaveMaxScale
        )
        Bias=0
       )
      )
      Float=$Particles.WaveScaleCurve
     )
    )
    Particles.CustomPosition_R2:NiagaraPosition=&AddVectorToPosition(
     Position=$Particles.Position
     Vector=&Multiply_VectorByFloat(
      Vector=&MakeVector(
       X=&Sine(
        Normalized Angle=&Add_Float(
         A=&Multiply_Float(A=$User.WavePeriod, B=0.5)
         B=&Subtract_Float(A=$Emitter.NormalizedLoopAge, B=$Particles.NormalizedAge)
        )
        Period=$User.WavePeriod
        Scale=&Multiply_Float(
         A=&SelectFloatFromArray(Direct Array Index=$Particles.SpawnGroup, Float Selection Array=DataInterface() {propertyValues {FloatData [1,1]}})
         B=$User.WaveMaxScale
        )
        Bias=0
       )
       Y=&Sine(
        Normalized Angle=&Add_Float(
         A=&Multiply_Float(A=$User.WavePeriod, B=0.5)
         B=&Subtract_Float(A=$Emitter.NormalizedLoopAge, B=$Particles.NormalizedAge)
        )
        Period=$User.WavePeriod
        Scale=&Multiply_Float(
         A=&SelectFloatFromArray(Direct Array Index=$Particles.SpawnGroup, Float Selection Array=DataInterface() {propertyValues {FloatData [1,1]}})
         B=$User.WaveMaxScale
        )
        Bias=0
       )
       Z=&Cosine(
        Normalized Angle=&Add_Float(
         A=&Multiply_Float(A=$User.WavePeriod, B=0.5)
         B=&Subtract_Float(A=$Emitter.NormalizedLoopAge, B=$Particles.NormalizedAge)
        )
        Period=$User.WavePeriod
        Scale=&Multiply_Float(
         A=&SelectFloatFromArray(Direct Array Index=$Particles.SpawnGroup, Float Selection Array=DataInterface() {propertyValues {FloatData [-1,1]}})
         B=$User.WaveMaxScale
        )
        Bias=0
       )
      )
      Float=$Particles.WaveScaleCurve
     )
    )
   ) {isSetParameters=true, module=SetVariables_#6}
   /SetVariables(
    Particles.CustomPosition_R1:NiagaraPosition=&AddVectorToPosition(
     Position=$Particles.Position
     Vector=&Multiply_VectorByFloat(
      Vector=&MakeVector(
       X=&Sine(
        Normalized Angle=&Add_Float(A=0, B=&Subtract_Float(A=$Emitter.NormalizedLoopAge, B=$Particles.NormalizedAge))
        Period=$User.WavePeriod
        Scale=&Multiply_Float(
         A=&SelectFloatFromArray(Direct Array Index=$Particles.SpawnGroup, Float Selection Array=DataInterface() {propertyValues {FloatData [1,-1]}})
         B=$User.WaveMaxScale
        )
        Bias=0
       )
       Y=&Sine(
        Normalized Angle=&Add_Float(A=0, B=&Subtract_Float(A=$Emitter.NormalizedLoopAge, B=$Particles.NormalizedAge))
        Period=$User.WavePeriod
        Scale=&Multiply_Float(
         A=&SelectFloatFromArray(Direct Array Index=$Particles.SpawnGroup, Float Selection Array=DataInterface() {propertyValues {FloatData [1,-1]}})
         B=$User.WaveMaxScale
        )
        Bias=0
       )
       Z=&Cosine(
        Normalized Angle=&Add_Float(A=0, B=&Subtract_Float(A=$Emitter.NormalizedLoopAge, B=$Particles.NormalizedAge))
        Period=$User.WavePeriod
        Scale=&Multiply_Float(
         A=&SelectFloatFromArray(Direct Array Index=$Particles.SpawnGroup, Float Selection Array=DataInterface() {propertyValues {FloatData [1,-1]}})
         B=$User.WaveMaxScale
        )
        Bias=0
       )
      )
      Float=$Particles.WaveScaleCurve
     )
    )
    Particles.CustomPosition_R2:NiagaraPosition=&AddVectorToPosition(
     Position=$Particles.Position
     Vector=&Multiply_VectorByFloat(
      Vector=&MakeVector(
       X=&Sine(
        Normalized Angle=&Add_Float(
         A=&Multiply_Float(A=$User.WavePeriod, B=0)
         B=&Subtract_Float(A=$Emitter.NormalizedLoopAge, B=$Particles.NormalizedAge)
        )
        Period=$User.WavePeriod
        Scale=&Multiply_Float(
         A=&SelectFloatFromArray(Direct Array Index=$Particles.SpawnGroup, Float Selection Array=DataInterface() {propertyValues {FloatData [0,0]}})
         B=$User.WaveMaxScale
        )
        Bias=0
       )
       Y=&Sine(
        Normalized Angle=&Add_Float(
         A=&Multiply_Float(A=$User.WavePeriod, B=0)
         B=&Subtract_Float(A=$Emitter.NormalizedLoopAge, B=$Particles.NormalizedAge)
        )
        Period=$User.WavePeriod
        Scale=&Multiply_Float(
         A=&SelectFloatFromArray(Direct Array Index=$Particles.SpawnGroup, Float Selection Array=DataInterface() {propertyValues {FloatData [0,0]}})
         B=$User.WaveMaxScale
        )
        Bias=0
       )
       Z=&Cosine(
        Normalized Angle=&Add_Float(
         A=&Multiply_Float(A=$User.WavePeriod, B=0)
         B=&Subtract_Float(A=$Emitter.NormalizedLoopAge, B=$Particles.NormalizedAge)
        )
        Period=$User.WavePeriod
        Scale=&Multiply_Float(
         A=&SelectFloatFromArray(Direct Array Index=$Particles.SpawnGroup, Float Selection Array=DataInterface() {propertyValues {FloatData [0,0]}})
         B=$User.WaveMaxScale
        )
        Bias=0
       )
      )
      Float=$Particles.WaveScaleCurve
     )
    )
   ) {enabled=false, isSetParameters=true}
   /SetVariables(Particles.DynamicMP_R2:Vector4f=&MakeVector4(
     X=&FloatFromCurve(FloatCurve=DataInterface() {propertyValues {Curve {
         defaultValue=1
         keys [
          {tangentMode=RCTM_User, time=0, value=0.5}
          {arriveTangent=-0.0633, interpMode=RCIM_Cubic, leaveTangent=-0.0633, tangentMode=RCTM_User, time=0.65, value=0.5}
          {arriveTangent=-0.0699, interpMode=RCIM_Cubic, leaveTangent=-0.0699, tangentMode=RCTM_User, time=0.75, value=-2}
         ]
        }}})
     Y=0
     Z=0
     W=0
    )) {isSetParameters=true, module=SetVariables_#7}
   &DynamicMaterialParameters(
    Index 0 Param 1=&FloatFromCurve(FloatCurve=DataInterface() {propertyValues {Curve {
        defaultValue=1
        keys [
         {
          arriveTangent=0.0405
          arriveTangentWeight=0.1835
          interpMode=RCIM_Cubic
          leaveTangent=0.0405
          leaveTangentWeight=0.1835
          tangentMode=RCTM_User
          time=0
          value=0
         }
         {arriveTangent=2.8994, interpMode=RCIM_Cubic, leaveTangent=2.8994, time=0.2297, value=0.2011}
         {arriveTangent=3.1289, interpMode=RCIM_Cubic, leaveTangent=3.1289, tangentMode=RCTM_User, time=0.2915, value=0.8451}
         {
          arriveTangent=0.0201
          arriveTangentWeight=0.1834
          interpMode=RCIM_Cubic
          leaveTangent=0.0201
          leaveTangentWeight=0.1834
          tangentMode=RCTM_User
          time=0.5
          value=1
         }
        ]
       }}})
    Module.Index0Float0Write=true
    Module.Index0Float1Write=true
    Module.Index0Float2Write=true
    Module.Index0Float3Write=true
   )
  ]
  rendererClasses [&NiagaraRibbonRendererProperties]
  renderers [
   {rendererClass=&NiagaraRibbonRendererProperties}
   {rendererClass=&NiagaraRibbonRendererProperties, rendererIndex=1}
   {rendererClass=&NiagaraRibbonRendererProperties, rendererIndex=2}
   {rendererClass=&NiagaraRibbonRendererProperties, rendererIndex=3}
  ]
 }
 NE_BlackFlame_Base2 {
  bEnabled=false
  emitterUpdateScript [
   &EmitterState(Life Cycle Mode=::NewEnumerator1, Loop Duration=1, UseLoopDelay=false)
   &VC_FilteredSocketCount()
   &VC_RateQualityHandler(Epic Spawn Rate=30, High Spawn Rate=30, Low Spawn Rate=30, Use Custom Rate?=true)
   &VC_SpawnRibbon_v1(Number of Ribbons=$Emitter.PortCount, Spawn Time=-1, SpawnRate=$Emitter.Spawn Rate)
  ]
  particleSpawnScript [
   &VC_PaintedVariantsOverride() {enabled=false}
   &VC_InitializeRibbon_v1(
    Color=$Particles.Color
    Custom Offset=true
    Enable Skeletal Position=true
    Lifetime=&MergeHelper_FloatSwitch(0=1, 1=1, 2=1, 3=?, Index=$User.GameModeID, Num=::NewEnumerator1)
    Local Vector Offset=&AddVectorToPosition(Position=<30 0 2.5>, Vector=<0 0 0>)
    Offset Amount=1
    Ribbon Width=-55
   )
   &AddVelocity(Velocity=<-2000 0 0>) {module=AddVelocity001}
   /SetVariables(
    Particles.R1_Width:NiagaraFloat=&SelectFloatFromArray(Direct Array Index=&SpawnGroup(), Float Selection Array=DataInterface() {propertyValues {FloatData [75,0]}})
    Particles.R2_Width:NiagaraFloat=&SelectFloatFromArray(Direct Array Index=&SpawnGroup(), Float Selection Array=DataInterface() {propertyValues {FloatData [0,-105]}})
    Particles.R3_Color:LinearColor=<rgba 0.2241 0.5211 0.6609 3>
   ) {enabled=false, isSetParameters=true, module=SetVariables_#2}
  ]
  particleUpdateScript [
   &VC_PaintedVariantsOverride(
    R1 Black:LinearColor=&ColorFromCurve(ColorCurve=DataInterface() {propertyValues {
       AlphaCurve {keys [
         <at 0.2819 7>
         <at 0.4191 0>
         <at 0.6302 0>
         <at 0.7212 1>
         {arriveTangent=-6.7757, interpMode=RCIM_Cubic, leaveTangent=-6.7757, time=0.7822, value=1}
         <at 0.8688 0>
        ]}
       BlueCurve {keys [<at 1 0.1094 RCIM_Cubic>]}
       GreenCurve {keys [<at 1 0.1094 RCIM_Cubic>]}
       RedCurve {keys [<at 1 0.1094 RCIM_Cubic>]}
      }})
    R1 BurntSienna:LinearColor=&ColorFromCurve(ColorCurve=DataInterface() {propertyValues {
       AlphaCurve {keys [
         <at 0.2819 7>
         <at 0.4191 0>
         <at 0.6302 0>
         <at 0.7212 1>
         {arriveTangent=-6.7757, interpMode=RCIM_Cubic, leaveTangent=-6.7757, time=0.7822, value=1}
         <at 0.8688 0>
        ]}
       BlueCurve {keys [<at 1 0.0686 RCIM_Cubic>]}
       GreenCurve {keys [<at 1 0.0858 RCIM_Cubic>]}
       RedCurve {keys [<at 1 0.1667 RCIM_Cubic>]}
      }})
    R1 Cobalt:LinearColor=&ColorFromCurve(ColorCurve=DataInterface() {propertyValues {
       AlphaCurve {keys [
         <at 0.2819 7>
         <at 0.4191 0>
         <at 0.6302 0>
         <at 0.7212 1>
         {arriveTangent=-6.7757, interpMode=RCIM_Cubic, leaveTangent=-6.7757, time=0.7822, value=1}
         <at 0.8688 0>
        ]}
       BlueCurve {keys [<at 1 0.3802 RCIM_Cubic>]}
       GreenCurve {keys [<at 1 0.0931 RCIM_Cubic>]}
       RedCurve {keys [<at 1 0.0931 RCIM_Cubic>]}
      }})
    R1 Crimson:LinearColor=&ColorFromCurve(ColorCurve=DataInterface() {propertyValues {
       AlphaCurve {keys [
         <at 0.2819 7>
         <at 0.4191 0>
         <at 0.6302 0>
         <at 0.7212 1>
         {arriveTangent=-6.7757, interpMode=RCIM_Cubic, leaveTangent=-6.7757, time=0.7822, value=1}
         <at 0.8688 0>
        ]}
       BlueCurve {keys [<at 1 0.0131 RCIM_Cubic>]}
       GreenCurve {keys [<at 1 0 RCIM_Cubic>]}
       RedCurve {keys [<at 1 0.2348 RCIM_Cubic>]}
      }})
    R1 DefaultColor=&ColorFromCurve(ColorCurve=DataInterface() {propertyValues {
       AlphaCurve {keys [
         <at 0.2819 7>
         <at 0.434 0>
         <at 0.6302 0>
         <at 0.7212 1>
         {arriveTangent=-6.7757, interpMode=RCIM_Cubic, leaveTangent=-6.7757, time=0.7822, value=1}
         <at 0.8688 0>
        ]}
       BlueCurve {keys [<at 1 0.1195 RCIM_Cubic>]}
       GreenCurve {keys [<at 1 0 RCIM_Cubic>]}
       RedCurve {keys [<at 1 0.063 RCIM_Cubic>]}
      }})
    R1 ForestGreen:LinearColor=&ColorFromCurve(ColorCurve=DataInterface() {propertyValues {
       AlphaCurve {keys [
         <at 0.2819 7>
         <at 0.4191 0>
         <at 0.6302 0>
         <at 0.7212 1>
         {arriveTangent=-6.7757, interpMode=RCIM_Cubic, leaveTangent=-6.7757, time=0.7822, value=1}
         <at 0.8688 0>
        ]}
       BlueCurve {keys [<at 1 0.0198 RCIM_Cubic>]}
       GreenCurve {keys [<at 1 0.2031 RCIM_Cubic>]}
       RedCurve {keys [<at 1 0 RCIM_Cubic>]}
      }})
    R1 Grey:LinearColor=&ColorFromCurve(ColorCurve=DataInterface() {propertyValues {
       AlphaCurve {keys [
         <at 0.2819 1>
         <at 0.4191 0>
         <at 0.6302 0>
         <at 0.7212 1>
         {arriveTangent=-6.7757, interpMode=RCIM_Cubic, leaveTangent=-6.7757, time=0.7822, value=1}
         <at 0.8688 0>
        ]}
       BlueCurve {keys [<at 1 0.3542 RCIM_Cubic>]}
       GreenCurve {keys [<at 1 0.3542 RCIM_Cubic>]}
       RedCurve {keys [<at 1 0.3542 RCIM_Cubic>]}
      }})
    R1 Lime:LinearColor=&ColorFromCurve(ColorCurve=DataInterface() {propertyValues {
       AlphaCurve {keys [
         <at 0.2819 7>
         <at 0.4191 0>
         <at 0.6302 0>
         <at 0.7212 1>
         {arriveTangent=-6.7757, interpMode=RCIM_Cubic, leaveTangent=-6.7757, time=0.7822, value=1}
         <at 0.8688 0>
        ]}
       BlueCurve {keys [<at 1 0.0865 RCIM_Cubic>]}
       GreenCurve {keys [<at 1 0.2865 RCIM_Cubic>]}
       RedCurve {keys [<at 1 0.1697 RCIM_Cubic>]}
      }})
    R1 Orange:LinearColor=&ColorFromCurve(ColorCurve=DataInterface() {propertyValues {
       AlphaCurve {keys [
         <at 0.2819 7>
         <at 0.4191 0>
         <at 0.6302 0>
         <at 0.7212 1>
         {arriveTangent=-6.7757, interpMode=RCIM_Cubic, leaveTangent=-6.7757, time=0.7822, value=1}
         <at 0.8688 0>
        ]}
       BlueCurve {keys [<at 1 0.0688 RCIM_Cubic>]}
       GreenCurve {keys [<at 1 0.1085 RCIM_Cubic>]}
       RedCurve {keys [<at 1 0.3478 RCIM_Cubic>]}
      }})
    R1 Pink:LinearColor=&ColorFromCurve(ColorCurve=DataInterface() {propertyValues {
       AlphaCurve {keys [
         <at 0.2819 7>
         <at 0.4191 0>
         <at 0.6302 0>
         <at 0.7212 1>
         {arriveTangent=-6.7757, interpMode=RCIM_Cubic, leaveTangent=-6.7757, time=0.7822, value=1}
         <at 0.8688 0>
        ]}
       BlueCurve {keys [<at 1 0.2004 RCIM_Cubic>]}
       GreenCurve {keys [<at 1 0 RCIM_Cubic>]}
       RedCurve {keys [<at 1 0.2188 RCIM_Cubic>]}
      }})
    R1 Purple:LinearColor=&ColorFromCurve(ColorCurve=DataInterface() {propertyValues {
       AlphaCurve {keys [
         <at 0.2819 7>
         <at 0.4191 0>
         <at 0.6302 0>
         <at 0.7212 1>
         {arriveTangent=-6.7757, interpMode=RCIM_Cubic, leaveTangent=-6.7757, time=0.7822, value=1}
         <at 0.8688 0>
        ]}
       BlueCurve {keys [<at 1 0.375 RCIM_Cubic>]}
       GreenCurve {keys [<at 1 0 RCIM_Cubic>]}
       RedCurve {keys [<at 1 0.1484 RCIM_Cubic>]}
      }})
    R1 Saffron:LinearColor=&ColorFromCurve(ColorCurve=DataInterface() {propertyValues {
       AlphaCurve {keys [
         <at 0.2819 7>
         <at 0.4191 0>
         <at 0.6302 0>
         <at 0.7212 1>
         {arriveTangent=-6.7757, interpMode=RCIM_Cubic, leaveTangent=-6.7757, time=0.7822, value=1}
         <at 0.8688 0>
        ]}
       BlueCurve {keys [<at 1 0 RCIM_Cubic>]}
       GreenCurve {keys [<at 1 0.2395 RCIM_Cubic>]}
       RedCurve {keys [<at 1 0.2812 RCIM_Cubic>]}
      }})
    R1 SkyBlue:LinearColor=&ColorFromCurve(ColorCurve=DataInterface() {propertyValues {
       AlphaCurve {keys [
         <at 0.2819 7>
         <at 0.4191 0>
         <at 0.6302 0>
         <at 0.7212 1>
         {arriveTangent=-6.7757, interpMode=RCIM_Cubic, leaveTangent=-6.7757, time=0.7822, value=1}
         <at 0.8688 0>
        ]}
       BlueCurve {keys [<at 1 0.3802 RCIM_Cubic>]}
       GreenCurve {keys [<at 1 0.1905 RCIM_Cubic>]}
       RedCurve {keys [<at 1 0 RCIM_Cubic>]}
      }})
    R1 TitaniumWhite:LinearColor=&ColorFromCurve(ColorCurve=DataInterface() {propertyValues {
       AlphaCurve {keys [
         <at 0.2819 2>
         <at 0.4191 0>
         <at 0.6302 0>
         <at 0.7212 1>
         {arriveTangent=-6.7757, interpMode=RCIM_Cubic, leaveTangent=-6.7757, time=0.7822, value=1}
         <at 0.8688 0>
        ]}
       BlueCurve {keys [<at 1 0.6406 RCIM_Cubic>]}
       GreenCurve {keys [<at 1 0.6406 RCIM_Cubic>]}
       RedCurve {keys [<at 1 0.6406 RCIM_Cubic>]}
      }})
    R2 Black:LinearColor=&ColorFromCurve(ColorCurve=DataInterface() {propertyValues {
       AlphaCurve {keys [<at 0 7>,<at 0.2232 1>,<at 0.3261 0.05>,<at 0.5495 0.2>,<at 0.7007 0.2>,<at 0.8474 0 RCIM_Cubic>]}
       BlueCurve {keys [<at 0 0.0194 RCIM_Cubic>,<at 0.1492 0.0333>,<at 0.2681 0.1104>,<at 0.4863 0.9274>,<at 0.6088 0.0305>]}
       GreenCurve {keys [<at 0 0.0194 RCIM_Cubic>,<at 0.1492 0.0333>,<at 0.2681 0.1104>,<at 0.4863 0.9274>,<at 0.6088 0.0305>]}
       RedCurve {keys [<at 0 0.0194 RCIM_Cubic>,<at 0.1492 0.0333>,<at 0.2681 0.1104>,<at 0.4863 0.9274>,<at 0.6088 0.0305>]}
      }})
    R2 BurntSienna:LinearColor=&ColorFromCurve(ColorCurve=DataInterface() {propertyValues {
       AlphaCurve {keys [<at 0 4>,<at 0.1907 1>,<at 0.3261 0.05>,<at 0.593 0.2>,<at 0.8474 0 RCIM_Cubic>]}
       BlueCurve {keys [<at 0 0.008 RCIM_Cubic>,<at 0.2839 0.0202>,<at 0.4863 0.2907>,<at 0.5956 0.008>]}
       GreenCurve {keys [<at 0 0.0203 RCIM_Cubic>,<at 0.2839 0.0323>,<at 0.4863 0.4394>,<at 0.5956 0.0203>]}
       RedCurve {keys [<at 0 0.0648 RCIM_Cubic>,<at 0.2839 0.0762>,<at 0.4863 0.9792>,<at 0.5956 0.0648>]}
      }})
    R2 Cobalt:LinearColor=&ColorFromCurve(ColorCurve=DataInterface() {propertyValues {
       AlphaCurve {keys [<at 0 7>,<at 0.1907 1>,<at 0.3261 0.05>,<at 0.5495 0.6>,<at 0.7007 0.2>,<at 0.8474 0 RCIM_Cubic>]}
       BlueCurve {keys [<at 0 0.6458 RCIM_Cubic>,<at 0.2575 0.9985>,<at 0.4863 1>,<at 0.6115 0.8177>]}
       GreenCurve {keys [<at 0 0.0119 RCIM_Cubic>,<at 0.2575 0.0367>,<at 0.4863 0.2997>,<at 0.6115 0.0527>]}
       RedCurve {keys [<at 0 0 RCIM_Cubic>,<at 0.2575 0.0186>,<at 0.4863 0.2865>,<at 0.6115 0.0383>]}
      }})
    R2 Crimson:LinearColor=&ColorFromCurve(ColorCurve=DataInterface() {propertyValues {
       AlphaCurve {keys [<at 0 7>,<at 0.1907 1>,<at 0.3261 0.05>,<at 0.5495 0.2>,<at 0.7007 0.2>,<at 0.8474 0 RCIM_Cubic>]}
       BlueCurve {keys [<at 0 0 RCIM_Cubic>,<at 0.1492 0.0104>,<at 0.4863 0.2995>,<at 0.6458 0.0894>]}
       GreenCurve {keys [<at 0 0 RCIM_Cubic>,<at 0.1492 0.0104>,<at 0.4863 0.2995>,<at 0.6458 0.0894>]}
       RedCurve {keys [<at 0 0.4323 RCIM_Cubic>,<at 0.1492 0.9992>,<at 0.4863 0.9274>,<at 0.6458 0.993>]}
      }})
    R2 DefaultColor:LinearColor=&ColorFromCurve(ColorCurve=DataInterface() {propertyValues {
       AlphaCurve {keys [<at 0 1>,<at 0.0914 0.2857>,<at 0.308 0.05>,<at 0.5592 0.2>,<at 0.8632 0>]}
       BlueCurve {keys [<at 0.0009 0>,<at 0.2668 -0.2829>,<at 0.5114 0.8849>,<at 0.619 1>,<at 0.682 0.383>]}
       GreenCurve {keys [<at 0.0009 0.0767>,<at 0.2742 -0.1994>,<at 0.5114 0.8849>,<at 0.619 1>,<at 0.682 0.2585>]}
       RedCurve {keys [<at 0.0009 3.2194>,<at 0.0016 3.2194>,<at 0.4479 0.8849>,<at 0.5114 0.8849>,<at 0.619 1>]}
      }})
    R2 ForestGreen:LinearColor=&ColorFromCurve(ColorCurve=DataInterface() {propertyValues {
       AlphaCurve {keys [<at 0 6>,<at 0.1907 1>,<at 0.3261 0.05>,<at 0.5495 0.2>,<at 0.7007 0.2>,<at 0.8474 0 RCIM_Cubic>]}
       BlueCurve {keys [<at 0 0.0172 RCIM_Cubic>,<at 0.2179 0.0322>,<at 0.4863 0.4656>,<at 0.6537 0.1027>]}
       GreenCurve {keys [<at 0 0.3438 RCIM_Cubic>,<at 0.2179 0.245>,<at 0.4863 0.875>,<at 0.6537 0.2987>]}
       RedCurve {keys [<at 0 0 RCIM_Cubic>,<at 0.2179 0.021>,<at 0.4863 0.4466>,<at 0.6537 0.0924>]}
      }})
    R2 Grey:LinearColor=&ColorFromCurve(ColorCurve=DataInterface() {propertyValues {
       AlphaCurve {keys [<at 0 4>,<at 0.1907 1>,<at 0.3261 0.05>,<at 0.5495 0.2>,<at 0.7007 0.2>,<at 0.8474 0 RCIM_Cubic>]}
       BlueCurve {keys [<at 0 0.0692 RCIM_Cubic>,<at 0.2285 0.1336>,<at 0.4863 0.9274>,<at 0.6405 0.132>]}
       GreenCurve {keys [<at 0 0.0781 RCIM_Cubic>,<at 0.2285 0.1336>,<at 0.4863 0.9274>,<at 0.6405 0.132>]}
       RedCurve {keys [<at 0 0.0767 RCIM_Cubic>,<at 0.2285 0.1406>,<at 0.4863 0.9274>,<at 0.6405 0.1354>]}
      }})
    R2 Lime:LinearColor=&ColorFromCurve(ColorCurve=DataInterface() {propertyValues {
       AlphaCurve {keys [<at 0 6>,<at 0.1907 1>,<at 0.3261 0.05>,<at 0.5495 0.2>,<at 0.8122 0 RCIM_Cubic>]}
       BlueCurve {keys [<at 0 0 RCIM_Cubic>,<at 0.2179 0.0459>,<at 0.4863 0.9274>,<at 0.6299 0.0724>]}
       GreenCurve {keys [<at 0 0.6875 RCIM_Cubic>,<at 0.2179 0.4635>,<at 0.4863 0.9274>,<at 0.6299 0.7708>]}
       RedCurve {keys [<at 0 0.3447 RCIM_Cubic>,<at 0.2179 0.2553>,<at 0.4863 0.9274>,<at 0.6299 0.4226>]}
      }})
    R2 Orange:LinearColor=&ColorFromCurve(ColorCurve=DataInterface() {propertyValues {
       AlphaCurve {keys [<at 0 7>,<at 0.1907 1>,<at 0.3261 0.05>,<at 0.5495 0.2>,<at 0.7007 0.2>,<at 0.8474 0 RCIM_Cubic>]}
       BlueCurve {keys [<at 0 0 RCIM_Cubic>,<at 0.2364 0.0598>,<at 0.4863 0.4263>,<at 0.6217 0.1191>]}
       GreenCurve {keys [<at 0 0.2834 RCIM_Cubic>,<at 0.2364 0.2792>,<at 0.4863 0.5719>,<at 0.6217 0.3121>]}
       RedCurve {keys [<at 0 0.8125 RCIM_Cubic>,<at 0.2364 0.6889>,<at 0.4863 0.8438>,<at 0.6217 0.6724>]}
      }})
    R2 Pink:LinearColor=&ColorFromCurve(ColorCurve=DataInterface() {propertyValues {
       AlphaCurve {keys [<at 0 7>,<at 0.1907 1>,<at 0.3261 0.05>,<at 0.5495 0.2>,<at 0.7007 0.2>,<at 0.8474 0 RCIM_Cubic>]}
       BlueCurve {keys [<at 0 0.3419 RCIM_Cubic>,<at 0.2707 0.3564>,<at 0.4863 0.7216>,<at 0.6141 0.5634>]}
       GreenCurve {keys [<at 0 0.0612 RCIM_Cubic>,<at 0.2707 0.1489>,<at 0.4863 0.5789>,<at 0.6141 0.2536>]}
       RedCurve {keys [<at 0 0.5029 RCIM_Cubic>,<at 0.2707 0.5614>,<at 0.4863 0.855>,<at 0.6141 0.8542>]}
      }})
    R2 Purple:LinearColor=&ColorFromCurve(ColorCurve=DataInterface() {propertyValues {
       AlphaCurve {keys [<at 0 7>,<at 0.1907 1>,<at 0.3261 0.05>,<at 0.5495 0.2>,<at 0.7007 0.2>,<at 0.8474 0 RCIM_Cubic>]}
       BlueCurve {keys [<at 0 0.8281 RCIM_Cubic>,<at 0.2919 0.6836>,<at 0.4863 0.8957>,<at 0.6141 0.9565>]}
       GreenCurve {keys [<at 0 0 RCIM_Cubic>,<at 0.2919 0.0563>,<at 0.4863 0.2791>,<at 0.6141 0.1097>]}
       RedCurve {keys [<at 0 0.1924 RCIM_Cubic>,<at 0.2919 0.3349>,<at 0.4863 0.5023>,<at 0.6141 0.3436>]}
      }})
    R2 Saffron:LinearColor=&ColorFromCurve(ColorCurve=DataInterface() {propertyValues {
       AlphaCurve {keys [<at 0 4>,<at 0.1907 1>,<at 0.3261 0.05>,<at 0.5495 0.2>,<at 0.8175 0 RCIM_Cubic>]}
       BlueCurve {keys [<at 0 0 RCIM_Cubic>,<at 0.2697 0.0046>,<at 0.4863 0.3188>,<at 0.6748 0.0798>]}
       GreenCurve {keys [<at 0 0.6105 RCIM_Cubic>,<at 0.2697 0.6881>,<at 0.4863 0.9274>,<at 0.6748 0.6052>]}
       RedCurve {keys [<at 0 0.7969 RCIM_Cubic>,<at 0.2697 0.8967>,<at 0.4863 0.8744>,<at 0.6748 0.7656>]}
      }})
    R2 SkyBlue:LinearColor=&ColorFromCurve(ColorCurve=DataInterface() {propertyValues {
       AlphaCurve {keys [<at 0 7>,<at 0.1907 1>,<at 0.3261 0.05>,<at 0.5495 0.2>,<at 0.8096 0 RCIM_Cubic>]}
       BlueCurve {keys [<at 0 0.5333 RCIM_Cubic>,<at 0.2496 0.5525>,<at 0.4863 0.9323>,<at 0.6141 0.9323>]}
       GreenCurve {keys [<at 0 0.3564 RCIM_Cubic>,<at 0.2496 0.3792>,<at 0.4863 0.8083>,<at 0.6141 0.6988>]}
       RedCurve {keys [<at 0 0 RCIM_Cubic>,<at 0.2496 0.0875>,<at 0.4863 0.5584>,<at 0.6141 0.2282>]}
      }})
    R2 TitaniumWhite:LinearColor=&ColorFromCurve(ColorCurve=DataInterface() {propertyValues {
       AlphaCurve {keys [<at 0 4>,<at 0.1907 1>,<at 0.3261 0.05>,<at 0.5495 0.2>,<at 0.7007 0.2>,<at 0.8175 0 RCIM_Cubic>]}
       BlueCurve {keys [<at 0 0.1846 RCIM_Cubic>,<at 0.2285 0.1336>,<at 0.4863 0.9274>,<at 0.6405 0.3248>]}
       GreenCurve {keys [<at 0 0.2083 RCIM_Cubic>,<at 0.2285 0.1336>,<at 0.4863 0.9274>,<at 0.6405 0.3248>]}
       RedCurve {keys [<at 0 0.2046 RCIM_Cubic>,<at 0.2285 0.1406>,<at 0.4863 0.9274>,<at 0.6405 0.3333>]}
      }})
    R3 Black:LinearColor=&ColorFromCurve(ColorCurve=DataInterface() {propertyValues {
       AlphaCurve {keys [<at 0.2665 3>,<at 0.3905 0>,<at 0.5841 0>]}
       BlueCurve {keys [<at 0.6966 0.3043>]}
       GreenCurve {keys [<at 0.6966 0.3043>]}
       RedCurve {keys [<at 0.6966 0.3043>]}
      }})
    R3 BurntSienna:LinearColor=&ColorFromCurve(ColorCurve=DataInterface() {propertyValues {
       AlphaCurve {keys [<at 0.2665 1>,<at 0.3905 0>,<at 0.5841 0>]}
       BlueCurve {keys [<at 0.6966 0.2729>]}
       GreenCurve {keys [<at 0.6966 0.3693>]}
       RedCurve {keys [<at 0.6966 0.6719>]}
      }})
    R3 Cobalt:LinearColor=&ColorFromCurve(ColorCurve=DataInterface() {propertyValues {
       AlphaCurve {keys [<at 0.2665 3>,<at 0.3905 0>,<at 0.5841 0>]}
       BlueCurve {keys [<at 0.6966 1.5>]}
       GreenCurve {keys [<at 0.6966 0.0598>]}
       RedCurve {keys [<at 0.6966 0>]}
      }})
    R3 Crimson:LinearColor=&ColorFromCurve(ColorCurve=DataInterface() {propertyValues {
       AlphaCurve {keys [<at 0.2665 3>,<at 0.3905 0>,<at 0.5841 0>]}
       BlueCurve {keys [<at 0.6966 0>]}
       GreenCurve {keys [<at 0.6966 0>]}
       RedCurve {keys [<at 0.6966 1>]}
      }})
    R3 DefaultColor:LinearColor=&ColorFromCurve(ColorCurve=DataInterface() {propertyValues {
       AlphaCurve {keys [<at 0.2665 3>,<at 0.3905 0>,<at 0.5841 0>]}
       BlueCurve {keys [<at 0.6966 2>]}
       GreenCurve {keys [<at 0.6966 0.5201>]}
       RedCurve {keys [<at 0.6966 0>]}
      }})
    R3 ForestGreen:LinearColor=&ColorFromCurve(ColorCurve=DataInterface() {propertyValues {
       AlphaCurve {keys [<at 0.2665 3>,<at 0.3905 0>,<at 0.5841 0>]}
       BlueCurve {keys [<at 0.6966 0.3387>]}
       GreenCurve {keys [<at 0.6966 1.5>]}
       RedCurve {keys [<at 0.6966 0>]}
      }})
    R3 Grey:LinearColor=&ColorFromCurve(ColorCurve=DataInterface() {propertyValues {
       AlphaCurve {keys [<at 0.2665 3>,<at 0.5658 0>,<at 0.7715 0>]}
       BlueCurve {keys [<at 0.6966 0.5>]}
       GreenCurve {keys [<at 0.6966 0.5>]}
       RedCurve {keys [<at 0.6966 0.5>]}
      }})
    R3 Lime:LinearColor=&ColorFromCurve(ColorCurve=DataInterface() {propertyValues {
       AlphaCurve {keys [<at 0.2665 3>,<at 0.3905 0>,<at 0.5841 0>]}
       BlueCurve {keys [<at 0.6966 0>]}
       GreenCurve {keys [<at 0.6966 1>]}
       RedCurve {keys [<at 0.6966 0.5901>]}
      }})
    R3 Orange:LinearColor=&ColorFromCurve(ColorCurve=DataInterface() {propertyValues {
       AlphaCurve {keys [<at 0.2665 3>,<at 0.3905 0>,<at 0.5841 0>]}
       BlueCurve {keys [<at 0.6966 0.0304>]}
       GreenCurve {keys [<at 0.6966 0.2737>]}
       RedCurve {keys [<at 0.6966 0.7292>]}
      }})
    R3 Pink:LinearColor=&ColorFromCurve(ColorCurve=DataInterface() {propertyValues {
       AlphaCurve {keys [<at 0.2665 3>,<at 0.3905 0>,<at 0.5841 0>]}
       BlueCurve {keys [<at 0.6966 1.2188>]}
       GreenCurve {keys [<at 0.6966 0.2158>]}
       RedCurve {keys [<at 0.6966 1.1488>]}
      }})
    R3 Purple:LinearColor=&ColorFromCurve(ColorCurve=DataInterface() {propertyValues {
       AlphaCurve {keys [<at 0.2665 3>,<at 0.3905 0>,<at 0.5841 0>]}
       BlueCurve {keys [<at 0.6966 1>]}
       GreenCurve {keys [<at 0.6966 0>]}
       RedCurve {keys [<at 0.6966 0.3627>]}
      }})
    R3 Saffron:LinearColor=&ColorFromCurve(ColorCurve=DataInterface() {propertyValues {
       AlphaCurve {keys [<at 0.2665 3>,<at 0.3905 0>,<at 0.5841 0>]}
       BlueCurve {keys [<at 0.6966 0>]}
       GreenCurve {keys [<at 0.6966 1.8157>]}
       RedCurve {keys [<at 0.6966 2>]}
      }})
    R3 SkyBlue:LinearColor=&ColorFromCurve(ColorCurve=DataInterface() {propertyValues {
       AlphaCurve {keys [<at 0.2665 3>,<at 0.3905 0>,<at 0.5841 0>]}
       BlueCurve {keys [<at 0.6966 1.5>]}
       GreenCurve {keys [<at 0.6966 0.3855>]}
       RedCurve {keys [<at 0.6966 0>]}
      }})
    R3 TitaniumWhite:LinearColor=&ColorFromCurve(ColorCurve=DataInterface() {propertyValues {
       AlphaCurve {keys [<at 0.4952 1>,<at 0.9629 0>]}
       BlueCurve {keys [<at 0.6966 2.5>]}
       GreenCurve {keys [<at 0.6966 2.5>]}
       RedCurve {keys [<at 0.6966 2.5>]}
      }})
    R4 BurntSienna:LinearColor=<rgba 0.0312 0.0096 0.0039 0.6>
    R4 Cobalt:LinearColor=<rgba 0 0.0115 0.6198 0.7>
    R4 Crimson:LinearColor=<rgba 0.2656 0 0.0072 0.7>
    R4 ForestGreen:LinearColor=<rgba 0 0.2552 0.0129 0.6>
    R4 Grey:LinearColor=<rgba 0.0417 0.0417 0.0417 0.7>
    R4 Lime:LinearColor=<rgba 0.1875 0.375 0 0.6>
    R4 Orange:LinearColor=<rgba 0.5521 0.1706 0 0.45>
    R4 Pink:LinearColor=<rgba 0.3906 0.0468 0.2656 0.5>
    R4 Purple:LinearColor=<rgba 0.1567 0 0.7552 0.5>
    R4 Saffron:LinearColor=<rgba 0.5573 0.4299 0 0.3>
    R4 SkyBlue:LinearColor=<rgba 0.0137 0.2957 0.4375 0.5>
    R4 TitaniumWhite:LinearColor=<rgba 0.224 0.224 0.224 0.6>
    R5 Black:LinearColor=<rgba 0.0469 0.0469 0.0469 0.6>
    R5 BurntSienna:LinearColor=<rgba 0.1042 0.0322 0.0128 0.7>
    R5 Cobalt:LinearColor=<rgba 0.0312 0.0401 1 0.6>
    R5 Crimson:LinearColor=<rgba 0.4635 0.0266 0.0384 0.6>
    R5 DefaultColor:LinearColor=<rgba 1 1 1 0.5>
    R5 ForestGreen:LinearColor=<rgba 0.0704 0.901 0.1124 0.4>
    R5 Grey:LinearColor=<rgba 0.1667 0.1667 0.1667 0.4>
    R5 Lime:LinearColor=<rgba 0.2938 0.5104 0.0771 0.25>
    R5 Orange:LinearColor=<rgba 0.8261 0.2993 0.0172 0.3>
    R5 Pink:LinearColor=<rgba 0.651 0.1763 0.4784 0.4>
    R5 Purple:LinearColor=<rgba 0.2704 0.1155 0.7917 0.3>
    R5 Saffron:LinearColor=<rgba 0.8906 0.6824 0 0.15>
    R5 SkyBlue:LinearColor=<rgba 0.0754 0.3822 0.5365 0.4>
    R5 TitaniumWhite:LinearColor=<rgba 0.5729 0.5729 0.5729 0.3>
    Use R1 Painted Overrides=true
    Use R2 Painted Overrides=true
    Use R3 Painted Overrides=true
    Use R4 Painted Overrides=true
    Use R5 Painted Overrides=true
   ) {module=VC_PaintedVariantsOverride001}
   &ParticleState()
   &InheritSourceMovement(
    Applied Position Scale=&Multiply_VectorByFloat(
     Vector=<1 1 1>
     Float=&FloatFromCurve(FloatCurve=DataInterface() {propertyValues {Curve {keys [{interpMode=RCIM_Cubic, tangentMode=RCTM_Break, time=0, value=1}, {interpMode=RCIM_Cubic, tangentMode=RCTM_Break, time=1, value=0}]}}})
    )
    ApplyForce=false
    ApplyPosition=true
    ApplyPositionFalloff=false
   ) {module=InheritSourceMovement001}
   &SolveForcesAndVelocity(Clamp Velocity=false, Limit Acceleration=false)
   &VC_UpdateRibbon_v1(Enable Skeletal Position=true)
   /SetVariables(Particles.DotProduct:NiagaraFloat=&DotProduct(
     Vector A=&TransformVectorByMatrix(
      Vector:Vector3f=<1 0 0>
      Transformation Matrix:NiagaraMatrix=$Engine.Owner.SystemLocalToWorld
      Transform Vector as Position:NiagaraBool=false
     )
     Vector B=&NormalizeVector(
      Vector To Normalize=&SubtractVectorFromPosition(
       Position=$Engine.Owner.Position
       Vector=&ReturnCameraPosition(Coordinate Space:ENiagaraCoordinateSpace=::Simulation, Camera Query:NiagaraDataInterfaceCamera=DataInterface())
      )
      Safe Normalize=true
      Safe Fallback Threshold=0
      Safe Fallback Vector=<0 0 0.001>
     )
    )) {isSetParameters=true, module=SetVariables_#3}
   /SetVariables(
    Particles.RibbonBack:LinearColor=&ColorFromCurve(ColorCurve=DataInterface() {propertyValues {
       AlphaCurve {keys [<at 0.5096 0>,<at 0.6208 2>,<at 0.697 0>]}
       BlueCurve {keys [<at 0.9243 0>]}
       GreenCurve {keys [<at 0.9243 0>]}
       RedCurve {keys [<at 0.9243 0>]}
      }})
    Particles.RibbonBase:LinearColor=&ColorFromCurve(ColorCurve=DataInterface() {propertyValues {
       AlphaCurve {keys [<at 0.108 0>,<at 0.1385 1>,<at 0.31 1>,<at 0.4762 0>]}
       BlueCurve {keys [<at 0.9243 0>]}
       GreenCurve {keys [<at 0.9243 0>]}
       RedCurve {keys [<at 0.9243 0>]}
      }})
    Particles.RibbonGlow:LinearColor=&ColorFromCurve(ColorCurve=DataInterface() {propertyValues {
       AlphaCurve {keys [<at 0.2665 3>,<at 0.3905 0>,<at 0.5841 0>]}
       BlueCurve {keys [<at 0.6966 2>]}
       GreenCurve {keys [<at 0.6966 0.5201>]}
       RedCurve {keys [<at 0.6966 0>]}
      }})
    Particles.RibbonGradient:LinearColor=&ColorFromCurve(ColorCurve=DataInterface() {propertyValues {
       AlphaCurve {keys [<at 0.3296 7>,<at 0.4406 0>]}
       BlueCurve {keys [<at 1 0.1683 RCIM_Cubic>]}
       GreenCurve {keys [<at 1 0 RCIM_Cubic>]}
       RedCurve {keys [<at 1 0.0865 RCIM_Cubic>]}
      }})
   ) {isSetParameters=true}
   /SetVariables(
    Particles.R1_Color_Scaled:LinearColor=&MultiplyLinearColor(
     Linear Color=$Particles.R1 Color
     Scale Factor=&MakeVector4FromVectorAndFloat(
      "Vector (RGB)"=&VectorFromFloat(Value=&FloatFromCurve(
        FloatCurve=DataInterface() {propertyValues {Curve {keys [
            {arriveTangent=-13.0287, interpMode=RCIM_Cubic, time=0.2, value=1}
            <at 0.4191 0>
            <at 0.6302 0>
            <at 0.7212 1>
            {arriveTangent=-6.7757, interpMode=RCIM_Cubic, leaveTangent=-6.7757, time=0.7822, value=1}
            <at 0.8688 0>
           ]}}}
        Scale Curve=&MakeCustomFloatFromBool(Bool=$User.bIsPainted, True Float=$User.Power_EmissivePaintedR4, False Float=2)
       ))
      "Float (Alpha)"=&FloatFromCurve(
       FloatCurve=DataInterface() {propertyValues {Curve {keys [
           <at -0.0449 3.0504 RCIM_Cubic>
           {arriveTangent=0.1754, interpMode=RCIM_Cubic, leaveTangent=0.1754, time=0.111, value=0.7392}
           {arriveTangent=-0.2383, interpMode=RCIM_Cubic, leaveTangent=-0.2383, tangentMode=RCTM_User, time=0.4793, value=3.1423}
           {arriveTangent=-43.3151, interpMode=RCIM_Cubic, leaveTangent=-30.5159, tangentMode=RCTM_Break, time=0.675, value=3.1272}
           {arriveTangent=-9.9909, interpMode=RCIM_Cubic, leaveTangent=-9.9909, time=0.8329, value=0.2261}
           <at 1.01 -0.2199 RCIM_Cubic>
          ]}}}
       Scale Curve=&MakeCustomFloatFromBool(Bool=$User.bIsPainted, True Float=$User.Power_AlphaPaintedR4, False Float=0.5)
      )
     )
    )
    Particles.R3_Color_Scaled:LinearColor=&MultiplyLinearColor(
     Linear Color=$Particles.R3 Color
     Scale Factor=&MakeVector4FromVectorAndFloat(
      "Vector (RGB)"=&VectorFromFloat(Value=&FloatFromCurve(
        FloatCurve=DataInterface() {propertyValues {Curve {keys [<at 0.2665 1>,<at 0.3905 0>,<at 0.5841 0>]}}}
        Scale Curve=&MakeCustomFloatFromBool(Bool=$User.bIsPainted, True Float=$User.Power_EmissivePaintedR4, False Float=2)
       ))
      "Float (Alpha)"=&FloatFromCurve(
       FloatCurve=DataInterface() {propertyValues {Curve {keys [<at 0.2665 1>,<at 0.3905 0>,<at 0.5841 0>]}}}
       Scale Curve=&MakeCustomFloatFromBool(Bool=$User.bIsPainted, True Float=$User.Power_AlphaPaintedR4, False Float=0.5)
      )
     )
    )
    Particles.R4_Color_Scaled:LinearColor=&MultiplyLinearColor(
     Linear Color=$Particles.R4 Color
     Scale Factor=&MakeVector4FromVectorAndFloat(
      "Vector (RGB)"=&VectorFromFloat(Value=&FloatFromCurve(
        FloatCurve=DataInterface() {propertyValues {Curve {keys [
            <at -0.0135 1.393 RCIM_Cubic>
            {arriveTangent=0.2114, interpMode=RCIM_Cubic, leaveTangent=0.2114, time=0.1169, value=2.0317}
            {arriveTangent=-2.6168, interpMode=RCIM_Cubic, leaveTangent=-2.6168, tangentMode=RCTM_User, time=0.887, value=1.5834}
            {interpMode=RCIM_Cubic, leaveTangent=-0.8868, time=1.0581, value=1.0126}
           ]}}}
        Scale Curve=&MakeCustomFloatFromBool(Bool=$User.bIsPainted, True Float=$User.Power_EmissivePaintedR4, False Float=0.6)
       ))
      "Float (Alpha)"=&FloatFromCurve(
       FloatCurve=DataInterface() {propertyValues {Curve {keys [
           <at -0.0014 1.7739 RCIM_Cubic>
           {arriveTangent=-1.9854, interpMode=RCIM_Cubic, leaveTangent=-1.9854, time=0.3044, value=2.258}
           {arriveTangent=-0.2383, interpMode=RCIM_Cubic, leaveTangent=-1.8935, tangentMode=RCTM_Break, time=0.7964, value=0.1899}
           <at 1.4148 -1.634 RCIM_Cubic>
          ]}}}
       Scale Curve=&MakeCustomFloatFromBool(Bool=$User.bIsPainted, True Float=$User.Power_AlphaPaintedR4, False Float=1.2)
      )
     )
    )
    Particles.R5_Color_Scaled:LinearColor=&MultiplyLinearColor(
     Linear Color=$Particles.R5 Color
     Scale Factor=&MakeVector4FromVectorAndFloat(
      "Vector (RGB)"=&VectorFromFloat(Value=&FloatFromCurve(
        FloatCurve=DataInterface() {propertyValues {Curve {keys [
            <at -0.17 16.4711 RCIM_Cubic>
            {arriveTangent=-27.6507, interpMode=RCIM_Cubic, leaveTangent=-27.6507, time=0.1072, value=17.8443}
            {arriveTangent=-3.2592, interpMode=RCIM_Cubic, leaveTangent=-3.2592, tangentMode=RCTM_User, time=0.3385, value=2.4113}
            {arriveTangent=2.5648, interpMode=RCIM_Cubic, leaveTangent=2.5648, tangentMode=RCTM_User, time=0.64, value=1.973}
            {arriveTangent=11.7336, interpMode=RCIM_Cubic, leaveTangent=11.7336, time=1.0714, value=20.2083}
            {arriveTangent=-34.897, interpMode=RCIM_Cubic, leaveTangent=-34.897, time=1.3482, value=10.2825}
            {arriveTangent=-8.1043, interpMode=RCIM_Cubic, leaveTangent=-8.1043, time=1.5789, value=2.4972}
            <at 1.7789 6.7918 RCIM_Cubic>
           ]}}}
        Scale Curve=&MakeCustomFloatFromBool(Bool=$User.bIsPainted, True Float=$User.Power_EmissivePaintedR4, False Float=0.7)
       ))
      "Float (Alpha)"=&FloatFromCurve(
       FloatCurve=DataInterface() {propertyValues {Curve {keys [
           <at -0.0556 3.887 RCIM_Cubic>
           {arriveTangent=-1.3925, interpMode=RCIM_Cubic, leaveTangent=-1.3925, time=0.111, value=0.7392}
           {arriveTangent=-0.2383, interpMode=RCIM_Cubic, leaveTangent=-0.2383, tangentMode=RCTM_User, time=0.4793, value=3.1423}
           {arriveTangent=-6.3349, interpMode=RCIM_Cubic, leaveTangent=-6.3349, time=0.8329, value=0.2261}
           <at 1.01 -0.2199 RCIM_Cubic>
          ]}}}
       Scale Curve=&MakeCustomFloatFromBool(Bool=$User.bIsPainted, True Float=$User.Power_AlphaPaintedR4, False Float=0.3)
      )
     )
    )
   ) {isSetParameters=true}
   &ScaleRibbonWidth(Ribbon Width Scale=&FloatFromCurve(FloatCurve=DataInterface() {propertyValues {Curve {keys [
         {arriveTangent=-0.2605, interpMode=RCIM_Cubic, leaveTangent=-0.2605, tangentMode=RCTM_User, time=0.0043, value=1.4579}
         {arriveTangent=-1.5709, interpMode=RCIM_Cubic, leaveTangent=-1.5709, tangentMode=RCTM_User, time=0.3137, value=1.1734}
         {arriveTangent=-2.0722, interpMode=RCIM_Cubic, leaveTangent=-2.0722, time=0.5043, value=0.5604}
         {arriveTangent=-1.126, interpMode=RCIM_Cubic, leaveTangent=-1.126, time=0.5945, value=0.5916}
         {arriveTangent=-1.0469, tangentMode=RCTM_User, time=1.002, value=0}
        ]}}}))
   /SetVariables(
    Particles.Ribbon_BlackColor:LinearColor=<rgba 0 0 0 1>
    Particles.WaveScaleCurve:NiagaraFloat=&FloatFromCurve(
     FloatCurve=DataInterface() {propertyValues {Curve {keys [
         {arriveTangent=39.5227, interpMode=RCIM_Cubic, leaveTangent=39.5227, tangentMode=RCTM_User, time=0.0092, value=0}
         {
          arriveTangent=0.7219
          arriveTangentWeight=0.0795
          interpMode=RCIM_Cubic
          leaveTangent=0.1806
          leaveTangentWeight=0.0696
          tangentMode=RCTM_Break
          tangentWeightMode=RCTWM_WeightedBoth
          time=0.125
          value=1.25
         }
         {arriveTangent=0.1988, interpMode=RCIM_Cubic, leaveTangent=0.1988, tangentMode=RCTM_User, time=0.45, value=0.625}
        ]}}}
     Scale Curve=10
    )
   ) {isSetParameters=true, module=SetVariables_#5}
   /SetVariables(
    Particles.CustomPosition_R1:NiagaraPosition=&AddVectorToPosition(
     Position=$Particles.Position
     Vector=&Multiply_VectorByFloat(
      Vector=&MakeVector(
       X=&Sine(
        Normalized Angle=&Add_Float(A=0, B=&Subtract_Float(A=$Emitter.NormalizedLoopAge, B=$Particles.NormalizedAge))
        Period=$User.WavePeriod
        Scale=&Multiply_Float(
         A=&SelectFloatFromArray(Direct Array Index=$Particles.SpawnGroup, Float Selection Array=DataInterface() {propertyValues {FloatData [1,-1]}})
         B=$User.WaveMaxScale
        )
        Bias=0
       )
       Y=&Sine(
        Normalized Angle=&Add_Float(A=0, B=&Subtract_Float(A=$Emitter.NormalizedLoopAge, B=$Particles.NormalizedAge))
        Period=$User.WavePeriod
        Scale=&Multiply_Float(
         A=&SelectFloatFromArray(Direct Array Index=$Particles.SpawnGroup, Float Selection Array=DataInterface() {propertyValues {FloatData [1,-1]}})
         B=$User.WaveMaxScale
        )
        Bias=0
       )
       Z=&Cosine(
        Normalized Angle=&Add_Float(A=0, B=&Subtract_Float(A=$Emitter.NormalizedLoopAge, B=$Particles.NormalizedAge))
        Period=$User.WavePeriod
        Scale=$User.WaveMaxScale
        Bias=0
       )
      )
      Float=$Particles.WaveScaleCurve
     )
    )
    Particles.CustomPosition_R2:NiagaraPosition=&AddVectorToPosition(
     Position=$Particles.Position
     Vector=&Multiply_VectorByFloat(
      Vector=&MakeVector(
       X=&Sine(
        Normalized Angle=&Add_Float(
         A=&Multiply_Float(A=$User.WavePeriod, B=0.5)
         B=&Subtract_Float(A=$Emitter.NormalizedLoopAge, B=$Particles.NormalizedAge)
        )
        Period=$User.WavePeriod
        Scale=&Multiply_Float(
         A=&SelectFloatFromArray(Direct Array Index=$Particles.SpawnGroup, Float Selection Array=DataInterface() {propertyValues {FloatData [1,-1]}})
         B=$User.WaveMaxScale
        )
        Bias=0
       )
       Y=&Sine(
        Normalized Angle=&Add_Float(
         A=&Multiply_Float(A=$User.WavePeriod, B=0.5)
         B=&Subtract_Float(A=$Emitter.NormalizedLoopAge, B=$Particles.NormalizedAge)
        )
        Period=$User.WavePeriod
        Scale=&Multiply_Float(
         A=&SelectFloatFromArray(Direct Array Index=$Particles.SpawnGroup, Float Selection Array=DataInterface() {propertyValues {FloatData [1,-1]}})
         B=$User.WaveMaxScale
        )
        Bias=0
       )
       Z=&Cosine(
        Normalized Angle=&Add_Float(
         A=&Multiply_Float(A=$User.WavePeriod, B=0.5)
         B=&Subtract_Float(A=$Emitter.NormalizedLoopAge, B=$Particles.NormalizedAge)
        )
        Period=$User.WavePeriod
        Scale=$User.WaveMaxScale
        Bias=0
       )
      )
      Float=$Particles.WaveScaleCurve
     )
    )
   ) {enabled=false, isSetParameters=true, module=SetVariables_#6}
   /SetVariables(
    Particles.CustomPosition_R1:NiagaraPosition=&AddVectorToPosition(
     Position=$Particles.Position
     Vector=&Multiply_VectorByFloat(
      Vector=&MakeVector(
       X=&Sine(
        Normalized Angle=&Add_Float(A=0, B=&Subtract_Float(A=$Emitter.NormalizedLoopAge, B=$Particles.NormalizedAge))
        Period=$User.WavePeriod
        Scale=&Multiply_Float(
         A=&SelectFloatFromArray(Direct Array Index=$Particles.SpawnGroup, Float Selection Array=DataInterface() {propertyValues {FloatData [1,-1]}})
         B=$User.WaveMaxScale
        )
        Bias=0
       )
       Y=&Sine(
        Normalized Angle=&Add_Float(A=0, B=&Subtract_Float(A=$Emitter.NormalizedLoopAge, B=$Particles.NormalizedAge))
        Period=$User.WavePeriod
        Scale=&Multiply_Float(
         A=&SelectFloatFromArray(Direct Array Index=$Particles.SpawnGroup, Float Selection Array=DataInterface() {propertyValues {FloatData [1,1]}})
         B=$User.WaveMaxScale
        )
        Bias=0
       )
       Z=&Cosine(
        Normalized Angle=&Add_Float(A=0, B=&Subtract_Float(A=$Emitter.NormalizedLoopAge, B=$Particles.NormalizedAge))
        Period=$User.WavePeriod
        Scale=&Multiply_Float(
         A=&SelectFloatFromArray(Direct Array Index=$Particles.SpawnGroup, Float Selection Array=DataInterface() {propertyValues {FloatData [-1,1]}})
         B=$User.WaveMaxScale
        )
        Bias=0
       )
      )
      Float=$Particles.WaveScaleCurve
     )
    )
    Particles.CustomPosition_R2:NiagaraPosition=&AddVectorToPosition(
     Position=$Particles.Position
     Vector=&Multiply_VectorByFloat(
      Vector=&MakeVector(
       X=&Sine(
        Normalized Angle=&Add_Float(
         A=&Multiply_Float(A=$User.WavePeriod, B=0.5)
         B=&Subtract_Float(A=$Emitter.NormalizedLoopAge, B=$Particles.NormalizedAge)
        )
        Period=$User.WavePeriod
        Scale=&Multiply_Float(
         A=&SelectFloatFromArray(Direct Array Index=$Particles.SpawnGroup, Float Selection Array=DataInterface() {propertyValues {FloatData [1,1]}})
         B=$User.WaveMaxScale
        )
        Bias=0
       )
       Y=&Sine(
        Normalized Angle=&Add_Float(
         A=&Multiply_Float(A=$User.WavePeriod, B=0.5)
         B=&Subtract_Float(A=$Emitter.NormalizedLoopAge, B=$Particles.NormalizedAge)
        )
        Period=$User.WavePeriod
        Scale=&Multiply_Float(
         A=&SelectFloatFromArray(Direct Array Index=$Particles.SpawnGroup, Float Selection Array=DataInterface() {propertyValues {FloatData [1,1]}})
         B=$User.WaveMaxScale
        )
        Bias=0
       )
       Z=&Cosine(
        Normalized Angle=&Add_Float(
         A=&Multiply_Float(A=$User.WavePeriod, B=0.5)
         B=&Subtract_Float(A=$Emitter.NormalizedLoopAge, B=$Particles.NormalizedAge)
        )
        Period=$User.WavePeriod
        Scale=&Multiply_Float(
         A=&SelectFloatFromArray(Direct Array Index=$Particles.SpawnGroup, Float Selection Array=DataInterface() {propertyValues {FloatData [-1,1]}})
         B=$User.WaveMaxScale
        )
        Bias=0
       )
      )
      Float=$Particles.WaveScaleCurve
     )
    )
   ) {isSetParameters=true, module=SetVariables_#9}
   /SetVariables(Particles.DynamicMP_R2:Vector4f=&MakeVector4(
     X=&FloatFromCurve(FloatCurve=DataInterface() {propertyValues {Curve {
         defaultValue=1
         keys [
          {tangentMode=RCTM_User, time=0, value=0.5}
          {arriveTangent=-0.0633, interpMode=RCIM_Cubic, leaveTangent=-0.0633, tangentMode=RCTM_User, time=0.65, value=0.5}
          {arriveTangent=-0.0699, interpMode=RCIM_Cubic, leaveTangent=-0.0699, tangentMode=RCTM_User, time=0.75, value=-2}
         ]
        }}})
     Y=0
     Z=0
     W=0
    )) {isSetParameters=true, module=SetVariables_#7}
   &DynamicMaterialParameters(
    Index 0 Param 1=&FloatFromCurve(FloatCurve=DataInterface() {propertyValues {Curve {
        defaultValue=1
        keys [
         {
          arriveTangent=0.0405
          arriveTangentWeight=0.1835
          interpMode=RCIM_Cubic
          leaveTangent=0.0405
          leaveTangentWeight=0.1835
          tangentMode=RCTM_User
          time=0
          value=0
         }
         {
          arriveTangent=0.0201
          arriveTangentWeight=0.1834
          interpMode=RCIM_Cubic
          leaveTangent=0.0201
          leaveTangentWeight=0.1834
          tangentMode=RCTM_User
          time=0.5
          value=1
         }
        ]
       }}})
    Module.Index0Float0Write=true
    Module.Index0Float1Write=true
    Module.Index0Float2Write=true
    Module.Index0Float3Write=true
   )
  ]
  rendererClasses [&NiagaraRibbonRendererProperties]
  renderers [
   {rendererClass=&NiagaraRibbonRendererProperties}
   {rendererClass=&NiagaraRibbonRendererProperties, rendererIndex=1}
   {rendererClass=&NiagaraRibbonRendererProperties, rendererIndex=2}
   {rendererClass=&NiagaraRibbonRendererProperties, rendererIndex=3}
   {rendererClass=&NiagaraRibbonRendererProperties, rendererIndex=4}
  ]
 }
 NE_BlackFlame_Base2001 {
  emitterUpdateScript [
   &EmitterState(Life Cycle Mode=::NewEnumerator1, Loop Duration=1, UseLoopDelay=false)
   &VC_FilteredSocketCount()
   &VC_RateQualityHandler(Epic Spawn Rate=30, High Spawn Rate=30, Low Spawn Rate=30, Use Custom Rate?=true)
   &VC_SpawnRibbon_v1(Number of Ribbons=$Emitter.PortCount, Spawn Time=-1, SpawnRate=$Emitter.Spawn Rate)
  ]
  particleSpawnScript [
   &VC_PaintedVariantsOverride() {enabled=false}
   &VC_InitializeRibbon_v1(
    Color=$Particles.Color
    Custom Offset=true
    Enable Skeletal Position=true
    Lifetime=&MergeHelper_FloatSwitch(0=1, 1=1, 2=1, 3=?, Index=$User.GameModeID, Num=::NewEnumerator1)
    Local Vector Offset=&VC_SetVectorByGameModeID(
     SameValueForLockerAndBR?:NiagaraBool=false
     Locker Vector:Vector3f=<30 0 2.5>
     Battle Royale Vector:Vector3f=<180 0 2.5>
     Rocket Racing Vector:Vector3f=<180 0 2.5>
     GameModeID:NiagaraInt32=$System.GameModeID
    )
    Offset Amount=1
    Ribbon Width=-55
   )
   &AddVelocity(Velocity=<-2300 0 0>) {module=AddVelocity001}
  ]
  particleUpdateScript [
   &VC_PaintedVariantsOverride(
    R1 Black:LinearColor=&ColorFromCurve(ColorCurve=DataInterface() {propertyValues {
       AlphaCurve {
        defaultValue=7
        keys [
         <at 0.2819 7>
         <at 0.6302 1>
         <at 0.7212 1>
         {arriveTangent=-6.7193, interpMode=RCIM_Cubic, leaveTangent=-6.7193, time=0.7822, value=1}
         <at 0.87 0>
        ]
       }
       BlueCurve {keys [<at 1 0.1094 RCIM_Cubic>]}
       GreenCurve {keys [<at 1 0.1094 RCIM_Cubic>]}
       RedCurve {keys [<at 1 0.1094 RCIM_Cubic>]}
      }})
    R1 BurntSienna:LinearColor=&ColorFromCurve(ColorCurve=DataInterface() {propertyValues {
       AlphaCurve {
        defaultValue=7
        keys [
         <at 0.2819 7>
         <at 0.6302 1>
         <at 0.7212 1>
         {arriveTangent=-6.7193, interpMode=RCIM_Cubic, leaveTangent=-6.7193, time=0.7822, value=1}
         <at 0.87 0>
        ]
       }
       BlueCurve {keys [<at 1 0.0825 RCIM_Cubic>]}
       GreenCurve {keys [<at 1 0.0972 RCIM_Cubic>]}
       RedCurve {keys [<at 1 0.1667 RCIM_Cubic>]}
      }})
    R1 Cobalt:LinearColor=&ColorFromCurve(ColorCurve=DataInterface() {propertyValues {
       AlphaCurve {
        defaultValue=7
        keys [
         <at 0.2819 7>
         <at 0.6302 1>
         <at 0.7212 1>
         {arriveTangent=-6.7193, interpMode=RCIM_Cubic, leaveTangent=-6.7193, time=0.7822, value=1}
         <at 0.87 0>
        ]
       }
       BlueCurve {keys [<at 1 0.2344 RCIM_Cubic>]}
       GreenCurve {keys [<at 1 0.0574 RCIM_Cubic>]}
       RedCurve {keys [<at 1 0.0574 RCIM_Cubic>]}
      }})
    R1 Crimson:LinearColor=&ColorFromCurve(ColorCurve=DataInterface() {propertyValues {
       AlphaCurve {
        defaultValue=7
        keys [
         <at 0.2819 7>
         <at 0.6302 1>
         <at 0.7212 1>
         {arriveTangent=-6.7193, interpMode=RCIM_Cubic, leaveTangent=-6.7193, time=0.7822, value=1}
         <at 0.87 0>
        ]
       }
       BlueCurve {keys [<at 1 0.0131 RCIM_Cubic>]}
       GreenCurve {keys [<at 1 0 RCIM_Cubic>]}
       RedCurve {keys [<at 1 0.2348 RCIM_Cubic>]}
      }})
    R1 DefaultColor=&ColorFromCurve(ColorCurve=DataInterface() {propertyValues {
       AlphaCurve {keys [<at 0.2819 7>, <at 0.6302 1>, <at 0.7212 1>, {interpMode=RCIM_Cubic, leaveTangent=-6.7757, time=0.7822, value=1}, <at 0.87 0>]}
       BlueCurve {keys [<at 1 0.1042 RCIM_Cubic>]}
       GreenCurve {keys [<at 1 0.0002 RCIM_Cubic>]}
       RedCurve {keys [<at 1 0.0367 RCIM_Cubic>]}
      }})
    R1 ForestGreen:LinearColor=&ColorFromCurve(ColorCurve=DataInterface() {propertyValues {
       AlphaCurve {
        defaultValue=7
        keys [
         <at 0.2819 7>
         <at 0.6302 1>
         <at 0.7212 1>
         {arriveTangent=-6.7193, interpMode=RCIM_Cubic, leaveTangent=-6.7193, time=0.7822, value=1}
         <at 0.87 0>
        ]
       }
       BlueCurve {keys [<at 0.6169 0.0198 RCIM_Cubic>]}
       GreenCurve {keys [<at 0.6169 0.2031 RCIM_Cubic>]}
       RedCurve {keys [<at 0.6169 0 RCIM_Cubic>]}
      }})
    R1 Grey:LinearColor=&ColorFromCurve(ColorCurve=DataInterface() {propertyValues {
       AlphaCurve {
        defaultValue=1
        keys [
         <at 0.2819 7>
         <at 0.6302 1>
         <at 0.7212 1>
         {arriveTangent=-6.7193, interpMode=RCIM_Cubic, leaveTangent=-6.7193, time=0.7822, value=1}
         <at 0.87 0>
        ]
       }
       BlueCurve {keys [<at 1 0.1979 RCIM_Cubic>]}
       GreenCurve {keys [<at 1 0.1979 RCIM_Cubic>]}
       RedCurve {keys [<at 1 0.1979 RCIM_Cubic>]}
      }})
    R1 Lime:LinearColor=&ColorFromCurve(ColorCurve=DataInterface() {propertyValues {
       AlphaCurve {
        defaultValue=7
        keys [
         <at 0.2819 7>
         <at 0.6302 1>
         <at 0.7212 1>
         {arriveTangent=-6.7193, interpMode=RCIM_Cubic, leaveTangent=-6.7193, time=0.7822, value=1}
         <at 0.87 0>
        ]
       }
       BlueCurve {keys [<at 1 0.0865 RCIM_Cubic>]}
       GreenCurve {keys [<at 1 0.2865 RCIM_Cubic>]}
       RedCurve {keys [<at 1 0.1697 RCIM_Cubic>]}
      }})
    R1 Orange:LinearColor=&ColorFromCurve(ColorCurve=DataInterface() {propertyValues {
       AlphaCurve {
        defaultValue=7
        keys [
         <at 0.2819 7>
         <at 0.6302 1>
         <at 0.7212 1>
         {arriveTangent=-6.7193, interpMode=RCIM_Cubic, leaveTangent=-6.7193, time=0.7822, value=1}
         <at 0.87 0>
        ]
       }
       BlueCurve {keys [<at 1 0.0622 RCIM_Cubic>]}
       GreenCurve {keys [<at 1 0.1169 RCIM_Cubic>]}
       RedCurve {keys [<at 1 0.3073 RCIM_Cubic>]}
      }})
    R1 Pink:LinearColor=&ColorFromCurve(ColorCurve=DataInterface() {propertyValues {
       AlphaCurve {
        defaultValue=7
        keys [
         <at 0.2819 7>
         <at 0.6302 1>
         <at 0.7212 1>
         {arriveTangent=-6.7193, interpMode=RCIM_Cubic, leaveTangent=-6.7193, time=0.7822, value=1}
         <at 0.87 0>
        ]
       }
       BlueCurve {keys [<at 1 0.2004 RCIM_Cubic>]}
       GreenCurve {keys [<at 1 0 RCIM_Cubic>]}
       RedCurve {keys [<at 1 0.2188 RCIM_Cubic>]}
      }})
    R1 Purple:LinearColor=&ColorFromCurve(ColorCurve=DataInterface() {propertyValues {
       AlphaCurve {
        defaultValue=7
        keys [
         <at 0.2819 7>
         <at 0.6302 1>
         <at 0.7212 1>
         {arriveTangent=-6.7193, interpMode=RCIM_Cubic, leaveTangent=-6.7193, time=0.7822, value=1}
         <at 0.87 0>
        ]
       }
       BlueCurve {keys [<at 1 0.375 RCIM_Cubic>]}
       GreenCurve {keys [<at 1 0 RCIM_Cubic>]}
       RedCurve {keys [<at 1 0.1484 RCIM_Cubic>]}
      }})
    R1 Saffron:LinearColor=&ColorFromCurve(ColorCurve=DataInterface() {propertyValues {
       AlphaCurve {
        defaultValue=7
        keys [
         <at 0.2819 7>
         <at 0.6302 1>
         <at 0.7212 1>
         {arriveTangent=-6.7193, interpMode=RCIM_Cubic, leaveTangent=-6.7193, time=0.7822, value=1}
         <at 0.87 0>
        ]
       }
       BlueCurve {keys [<at 1 0.0278 RCIM_Cubic>]}
       GreenCurve {keys [<at 1 0.2596 RCIM_Cubic>]}
       RedCurve {keys [<at 1 0.2812 RCIM_Cubic>]}
      }})
    R1 SkyBlue:LinearColor=&ColorFromCurve(ColorCurve=DataInterface() {propertyValues {
       AlphaCurve {
        defaultValue=7
        keys [
         <at 0.2819 7>
         <at 0.6302 1>
         <at 0.7212 1>
         {arriveTangent=-6.7193, interpMode=RCIM_Cubic, leaveTangent=-6.7193, time=0.7822, value=1}
         <at 0.87 0>
        ]
       }
       BlueCurve {keys [<at 1 0.375 RCIM_Cubic>]}
       GreenCurve {keys [<at 1 0.1761 RCIM_Cubic>]}
       RedCurve {keys [<at 1 0.0547 RCIM_Cubic>]}
      }})
    R1 TitaniumWhite:LinearColor=&ColorFromCurve(ColorCurve=DataInterface() {propertyValues {
       AlphaCurve {
        defaultValue=2
        keys [
         <at 0.2819 7>
         <at 0.6302 1>
         <at 0.7212 1>
         {arriveTangent=-6.7193, interpMode=RCIM_Cubic, leaveTangent=-6.7193, time=0.7822, value=1}
         <at 0.87 0>
        ]
       }
       BlueCurve {keys [<at 0.6118 0.3802 RCIM_Cubic>,<at 1 0.0677>]}
       GreenCurve {keys [<at 0.6118 0.3802 RCIM_Cubic>,<at 1 0.0677>]}
       RedCurve {keys [<at 0.6118 0.3802 RCIM_Cubic>,<at 1 0.0677>]}
      }})
    R2 Black:LinearColor=&ColorFromCurve(ColorCurve=DataInterface() {propertyValues {
       AlphaCurve {keys [<at 0 7>,<at 0.2232 1>,<at 0.3261 0.05>,<at 0.5495 0.2>,<at 0.7007 0.2>,<at 0.8474 0 RCIM_Cubic>]}
       BlueCurve {keys [<at 0 0.0194 RCIM_Cubic>,<at 0.1492 0.0333>,<at 0.2681 0.1104>,<at 0.4863 0.9274>,<at 0.6088 0.0305>]}
       GreenCurve {keys [<at 0 0.0194 RCIM_Cubic>,<at 0.1492 0.0333>,<at 0.2681 0.1104>,<at 0.4863 0.9274>,<at 0.6088 0.0305>]}
       RedCurve {keys [<at 0 0.0194 RCIM_Cubic>,<at 0.1492 0.0333>,<at 0.2681 0.1104>,<at 0.4863 0.9274>,<at 0.6088 0.0305>]}
      }})
    R2 BurntSienna:LinearColor=&ColorFromCurve(ColorCurve=DataInterface() {propertyValues {
       AlphaCurve {keys [<at 0 4>,<at 0.1907 1>,<at 0.3261 0.05>,<at 0.593 0.2>,<at 0.8474 0 RCIM_Cubic>]}
       BlueCurve {keys [<at 0 0.008 RCIM_Cubic>,<at 0.2839 0.0202>,<at 0.4863 0.2907>,<at 0.5956 0.008>]}
       GreenCurve {keys [<at 0 0.0203 RCIM_Cubic>,<at 0.2839 0.0323>,<at 0.4863 0.4394>,<at 0.5956 0.0203>]}
       RedCurve {keys [<at 0 0.0648 RCIM_Cubic>,<at 0.2839 0.0762>,<at 0.4863 0.9792>,<at 0.5956 0.0648>]}
      }})
    R2 Cobalt:LinearColor=&ColorFromCurve(ColorCurve=DataInterface() {propertyValues {
       AlphaCurve {keys [<at 0 7>,<at 0.1907 1>,<at 0.3261 0.05>,<at 0.5495 0.6>,<at 0.7007 0.2>,<at 0.8474 0 RCIM_Cubic>]}
       BlueCurve {keys [<at 0 0.6458 RCIM_Cubic>,<at 0.2575 0.9985>,<at 0.4863 1>,<at 0.6115 0.8177>]}
       GreenCurve {keys [<at 0 0.0119 RCIM_Cubic>,<at 0.2575 0.0367>,<at 0.4863 0.2997>,<at 0.6115 0.0527>]}
       RedCurve {keys [<at 0 0 RCIM_Cubic>,<at 0.2575 0.0186>,<at 0.4863 0.2865>,<at 0.6115 0.0383>]}
      }})
    R2 Crimson:LinearColor=&ColorFromCurve(ColorCurve=DataInterface() {propertyValues {
       AlphaCurve {keys [<at 0 7>,<at 0.1907 1>,<at 0.3261 0.05>,<at 0.5495 0.2>,<at 0.7007 0.2>,<at 0.8474 0 RCIM_Cubic>]}
       BlueCurve {keys [<at 0 0 RCIM_Cubic>,<at 0.1492 0.0104>,<at 0.4863 0.2995>,<at 0.6458 0.0894>]}
       GreenCurve {keys [<at 0 0 RCIM_Cubic>,<at 0.1492 0.0104>,<at 0.4863 0.2995>,<at 0.6458 0.0894>]}
       RedCurve {keys [<at 0 0.4323 RCIM_Cubic>,<at 0.1492 0.9992>,<at 0.4863 0.9274>,<at 0.6458 0.993>]}
      }})
    R2 DefaultColor:LinearColor=&ColorFromCurve(ColorCurve=DataInterface() {propertyValues {
       AlphaCurve {keys [<at 0 1>,<at 0.0914 0.2857>,<at 0.308 0.05>,<at 0.5592 0.2>,<at 0.8632 0>]}
       BlueCurve {keys [<at 0.0009 0>,<at 0.2668 -0.2829>,<at 0.5114 0.8849>,<at 0.619 1>,<at 0.682 0.383>]}
       GreenCurve {keys [<at 0.0009 0.0767>,<at 0.2742 -0.1994>,<at 0.5114 0.8849>,<at 0.619 1>,<at 0.682 0.2585>]}
       RedCurve {keys [<at 0.0009 3.2194>,<at 0.0016 3.2194>,<at 0.4479 0.8849>,<at 0.5114 0.8849>,<at 0.619 1>]}
      }})
    R2 ForestGreen:LinearColor=&ColorFromCurve(ColorCurve=DataInterface() {propertyValues {
       AlphaCurve {keys [<at 0 6>,<at 0.1907 1>,<at 0.3261 0.05>,<at 0.5495 0.2>,<at 0.7007 0.2>,<at 0.8474 0 RCIM_Cubic>]}
       BlueCurve {keys [<at 0 0.0172 RCIM_Cubic>,<at 0.2179 0.0322>,<at 0.4863 0.4656>,<at 0.6537 0.1027>]}
       GreenCurve {keys [<at 0 0.3438 RCIM_Cubic>,<at 0.2179 0.245>,<at 0.4863 0.875>,<at 0.6537 0.2987>]}
       RedCurve {keys [<at 0 0 RCIM_Cubic>,<at 0.2179 0.021>,<at 0.4863 0.4466>,<at 0.6537 0.0924>]}
      }})
    R2 Grey:LinearColor=&ColorFromCurve(ColorCurve=DataInterface() {propertyValues {
       AlphaCurve {keys [<at 0 4>,<at 0.1907 1>,<at 0.3261 0.05>,<at 0.5495 0.2>,<at 0.7007 0.2>,<at 0.8474 0 RCIM_Cubic>]}
       BlueCurve {keys [<at 0 0.0692 RCIM_Cubic>,<at 0.2285 0.1336>,<at 0.4863 0.9274>,<at 0.6405 0.132>]}
       GreenCurve {keys [<at 0 0.0781 RCIM_Cubic>,<at 0.2285 0.1336>,<at 0.4863 0.9274>,<at 0.6405 0.132>]}
       RedCurve {keys [<at 0 0.0767 RCIM_Cubic>,<at 0.2285 0.1406>,<at 0.4863 0.9274>,<at 0.6405 0.1354>]}
      }})
    R2 Lime:LinearColor=&ColorFromCurve(ColorCurve=DataInterface() {propertyValues {
       AlphaCurve {keys [<at 0 6>,<at 0.1907 1>,<at 0.3261 0.05>,<at 0.5495 0.2>,<at 0.8122 0 RCIM_Cubic>]}
       BlueCurve {keys [<at 0 0 RCIM_Cubic>,<at 0.2179 0.0459>,<at 0.4863 0.9274>,<at 0.6299 0.0724>]}
       GreenCurve {keys [<at 0 0.6875 RCIM_Cubic>,<at 0.2179 0.4635>,<at 0.4863 0.9274>,<at 0.6299 0.7708>]}
       RedCurve {keys [<at 0 0.3447 RCIM_Cubic>,<at 0.2179 0.2553>,<at 0.4863 0.9274>,<at 0.6299 0.4226>]}
      }})
    R2 Orange:LinearColor=&ColorFromCurve(ColorCurve=DataInterface() {propertyValues {
       AlphaCurve {keys [<at 0 7>,<at 0.1907 1>,<at 0.3261 0.05>,<at 0.5495 0.2>,<at 0.7007 0.2>,<at 0.8474 0 RCIM_Cubic>]}
       BlueCurve {keys [<at 0 0 RCIM_Cubic>,<at 0.2364 0.0598>,<at 0.4863 0.4263>,<at 0.6217 0.1191>]}
       GreenCurve {keys [<at 0 0.2834 RCIM_Cubic>,<at 0.2364 0.2792>,<at 0.4863 0.5719>,<at 0.6217 0.3121>]}
       RedCurve {keys [<at 0 0.8125 RCIM_Cubic>,<at 0.2364 0.6889>,<at 0.4863 0.8438>,<at 0.6217 0.6724>]}
      }})
    R2 Pink:LinearColor=&ColorFromCurve(ColorCurve=DataInterface() {propertyValues {
       AlphaCurve {keys [<at 0 7>,<at 0.1907 1>,<at 0.3261 0.05>,<at 0.5495 0.2>,<at 0.7007 0.2>,<at 0.8474 0 RCIM_Cubic>]}
       BlueCurve {keys [<at 0 0.3419 RCIM_Cubic>,<at 0.2707 0.3564>,<at 0.4863 0.7216>,<at 0.6141 0.5634>]}
       GreenCurve {keys [<at 0 0.0612 RCIM_Cubic>,<at 0.2707 0.1489>,<at 0.4863 0.5789>,<at 0.6141 0.2536>]}
       RedCurve {keys [<at 0 0.5029 RCIM_Cubic>,<at 0.2707 0.5614>,<at 0.4863 0.855>,<at 0.6141 0.8542>]}
      }})
    R2 Purple:LinearColor=&ColorFromCurve(ColorCurve=DataInterface() {propertyValues {
       AlphaCurve {keys [<at 0 7>,<at 0.1907 1>,<at 0.3261 0.05>,<at 0.5495 0.2>,<at 0.7007 0.2>,<at 0.8474 0 RCIM_Cubic>]}
       BlueCurve {keys [<at 0 0.8281 RCIM_Cubic>,<at 0.2919 0.6836>,<at 0.4863 0.8957>,<at 0.6141 0.9565>]}
       GreenCurve {keys [<at 0 0 RCIM_Cubic>,<at 0.2919 0.0563>,<at 0.4863 0.2791>,<at 0.6141 0.1097>]}
       RedCurve {keys [<at 0 0.1924 RCIM_Cubic>,<at 0.2919 0.3349>,<at 0.4863 0.5023>,<at 0.6141 0.3436>]}
      }})
    R2 Saffron:LinearColor=&ColorFromCurve(ColorCurve=DataInterface() {propertyValues {
       AlphaCurve {keys [<at 0 4>,<at 0.1907 1>,<at 0.3261 0.05>,<at 0.5495 0.2>,<at 0.8175 0 RCIM_Cubic>]}
       BlueCurve {keys [<at 0 0 RCIM_Cubic>,<at 0.2697 0.0046>,<at 0.4863 0.3188>,<at 0.6748 0.0798>]}
       GreenCurve {keys [<at 0 0.6105 RCIM_Cubic>,<at 0.2697 0.6881>,<at 0.4863 0.9274>,<at 0.6748 0.6052>]}
       RedCurve {keys [<at 0 0.7969 RCIM_Cubic>,<at 0.2697 0.8967>,<at 0.4863 0.8744>,<at 0.6748 0.7656>]}
      }})
    R2 SkyBlue:LinearColor=&ColorFromCurve(ColorCurve=DataInterface() {propertyValues {
       AlphaCurve {keys [<at 0 7>,<at 0.1907 1>,<at 0.3261 0.05>,<at 0.5495 0.2>,<at 0.8096 0 RCIM_Cubic>]}
       BlueCurve {keys [<at 0 0.5333 RCIM_Cubic>,<at 0.2496 0.5525>,<at 0.4863 0.9323>,<at 0.6141 0.9323>]}
       GreenCurve {keys [<at 0 0.3564 RCIM_Cubic>,<at 0.2496 0.3792>,<at 0.4863 0.8083>,<at 0.6141 0.6988>]}
       RedCurve {keys [<at 0 0 RCIM_Cubic>,<at 0.2496 0.0875>,<at 0.4863 0.5584>,<at 0.6141 0.2282>]}
      }})
    R2 TitaniumWhite:LinearColor=&ColorFromCurve(ColorCurve=DataInterface() {propertyValues {
       AlphaCurve {keys [<at 0 4>,<at 0.1907 1>,<at 0.3261 0.05>,<at 0.5495 0.2>,<at 0.7007 0.2>,<at 0.8175 0 RCIM_Cubic>]}
       BlueCurve {keys [<at 0 0.1846 RCIM_Cubic>,<at 0.2285 0.1336>,<at 0.4863 0.9274>,<at 0.6405 0.3248>]}
       GreenCurve {keys [<at 0 0.2083 RCIM_Cubic>,<at 0.2285 0.1336>,<at 0.4863 0.9274>,<at 0.6405 0.3248>]}
       RedCurve {keys [<at 0 0.2046 RCIM_Cubic>,<at 0.2285 0.1406>,<at 0.4863 0.9274>,<at 0.6405 0.3333>]}
      }})
    R3 Black:LinearColor=&ColorFromCurve(ColorCurve=DataInterface() {propertyValues {
       AlphaCurve {keys [{tangentMode=RCTM_User, time=0.8086, value=3}, {tangentMode=RCTM_User, time=0.8837, value=0}, {tangentMode=RCTM_User, time=1, value=0}]}
       BlueCurve {keys [<at 0.6966 0.25>]}
       GreenCurve {keys [<at 0.6966 0.25>]}
       RedCurve {keys [<at 0.6966 0.25>]}
      }})
    R3 BurntSienna:LinearColor=&ColorFromCurve(ColorCurve=DataInterface() {propertyValues {
       AlphaCurve {
        defaultValue=1
        keys [{tangentMode=RCTM_User, time=0.8086, value=3}, {tangentMode=RCTM_User, time=0.8837, value=0}, {tangentMode=RCTM_User, time=1, value=0}]
       }
       BlueCurve {keys [<at 0.6966 0.147>]}
       GreenCurve {keys [<at 0.6966 0.2197>]}
       RedCurve {keys [<at 0.6966 0.4479>]}
      }})
    R3 Cobalt:LinearColor=&ColorFromCurve(ColorCurve=DataInterface() {propertyValues {
       AlphaCurve {
        defaultValue=3
        keys [{tangentMode=RCTM_User, time=0.8086, value=3}, {tangentMode=RCTM_User, time=0.8837, value=0}, {tangentMode=RCTM_User, time=1, value=0}]
       }
       BlueCurve {keys [<at 0.6966 0.9688>]}
       GreenCurve {keys [<at 0.6966 0>]}
       RedCurve {keys [<at 0.6966 0>]}
      }})
    R3 Crimson:LinearColor=&ColorFromCurve(ColorCurve=DataInterface() {propertyValues {
       AlphaCurve {
        defaultValue=3
        keys [{tangentMode=RCTM_User, time=0.8086, value=3}, {tangentMode=RCTM_User, time=0.8837, value=0}, {tangentMode=RCTM_User, time=1, value=0}]
       }
       BlueCurve {keys [<at 0.6966 0>]}
       GreenCurve {keys [<at 0.6966 0>]}
       RedCurve {keys [<at 0.6966 1>]}
      }})
    R3 DefaultColor:LinearColor=&ColorFromCurve(ColorCurve=DataInterface() {propertyValues {
       AlphaCurve {keys [<at 0.8086 3>,<at 0.8837 0>,<at 1 0>]}
       BlueCurve {keys [<at 0.6966 1.25>]}
       GreenCurve {keys [<at 0.6966 0.3251>]}
       RedCurve {keys [<at 0.6966 0>]}
      }})
    R3 ForestGreen:LinearColor=&ColorFromCurve(ColorCurve=DataInterface() {propertyValues {
       AlphaCurve {
        defaultValue=3
        keys [{tangentMode=RCTM_User, time=0.8086, value=3}, {tangentMode=RCTM_User, time=0.8837, value=0}, {tangentMode=RCTM_User, time=1, value=0}]
       }
       BlueCurve {keys [<at 0.6966 0.2593>]}
       GreenCurve {keys [<at 0.6966 0.6562>]}
       RedCurve {keys [<at 0.6966 0.1436>]}
      }})
    R3 Grey:LinearColor=&ColorFromCurve(ColorCurve=DataInterface() {propertyValues {
       AlphaCurve {
        defaultValue=3
        keys [{tangentMode=RCTM_User, time=0.8086, value=3}, {tangentMode=RCTM_User, time=0.8837, value=0}, {tangentMode=RCTM_User, time=1, value=0}]
       }
       BlueCurve {keys [<at 0.6966 0.3646>]}
       GreenCurve {keys [<at 0.6966 0.3646>]}
       RedCurve {keys [<at 0.6966 0.3646>]}
      }})
    R3 Lime:LinearColor=&ColorFromCurve(ColorCurve=DataInterface() {propertyValues {
       AlphaCurve {
        defaultValue=3
        keys [{tangentMode=RCTM_User, time=0.8086, value=3}, {tangentMode=RCTM_User, time=0.8837, value=0}, {tangentMode=RCTM_User, time=1, value=0}]
       }
       BlueCurve {keys [<at 0.6966 0>]}
       GreenCurve {keys [<at 0.6966 1>]}
       RedCurve {keys [<at 0.6966 0.5901>]}
      }})
    R3 Orange:LinearColor=&ColorFromCurve(ColorCurve=DataInterface() {propertyValues {
       AlphaCurve {
        defaultValue=3
        keys [{tangentMode=RCTM_User, time=0.8086, value=3}, {tangentMode=RCTM_User, time=0.8837, value=0}, {tangentMode=RCTM_User, time=1, value=0}]
       }
       BlueCurve {keys [<at 0.6966 0.1367>]}
       GreenCurve {keys [<at 0.6966 0.2734>]}
       RedCurve {keys [<at 0.6966 0.7292>]}
      }})
    R3 Pink:LinearColor=&ColorFromCurve(ColorCurve=DataInterface() {propertyValues {
       AlphaCurve {
        defaultValue=3
        keys [{tangentMode=RCTM_User, time=0.8086, value=3}, {tangentMode=RCTM_User, time=0.8837, value=0}, {tangentMode=RCTM_User, time=1, value=0}]
       }
       BlueCurve {keys [<at 0.6966 0.6855>]}
       GreenCurve {keys [<at 0.6966 0.1214>]}
       RedCurve {keys [<at 0.6966 0.6462>]}
      }})
    R3 Purple:LinearColor=&ColorFromCurve(ColorCurve=DataInterface() {propertyValues {
       AlphaCurve {
        defaultValue=3
        keys [{tangentMode=RCTM_User, time=0.8086, value=3}, {tangentMode=RCTM_User, time=0.8837, value=0}, {tangentMode=RCTM_User, time=1, value=0}]
       }
       BlueCurve {keys [<at 0.6966 0.8906>]}
       GreenCurve {keys [<at 0.6966 0>]}
       RedCurve {keys [<at 0.6966 0.2841>]}
      }})
    R3 Saffron:LinearColor=&ColorFromCurve(ColorCurve=DataInterface() {propertyValues {
       AlphaCurve {
        defaultValue=3
        keys [{tangentMode=RCTM_User, time=0.8086, value=3}, {tangentMode=RCTM_User, time=0.8837, value=0}, {tangentMode=RCTM_User, time=1, value=0}]
       }
       BlueCurve {keys [<at 0.6966 0.199>]}
       GreenCurve {keys [<at 0.6966 1.2571>]}
       RedCurve {keys [<at 0.6966 1.3646>]}
      }})
    R3 SkyBlue:LinearColor=&ColorFromCurve(ColorCurve=DataInterface() {propertyValues {
       AlphaCurve {
        defaultValue=3
        keys [{tangentMode=RCTM_User, time=0.8086, value=3}, {tangentMode=RCTM_User, time=0.8837, value=0}, {tangentMode=RCTM_User, time=1, value=0}]
       }
       BlueCurve {keys [<at 0.6966 0.9375>]}
       GreenCurve {keys [<at 0.6966 0.4441>]}
       RedCurve {keys [<at 0.6966 0>]}
      }})
    R3 TitaniumWhite:LinearColor=&ColorFromCurve(ColorCurve=DataInterface() {propertyValues {
       AlphaCurve {keys [<at 0.4952 1>,<at 0.9629 0>]}
       BlueCurve {keys [<at 0.6966 2.5>]}
       GreenCurve {keys [<at 0.6966 2.5>]}
       RedCurve {keys [<at 0.6966 2.5>]}
      }})
    R4 Black:LinearColor=&ColorFromCurve(ColorCurve=DataInterface() {propertyValues {
       AlphaCurve {keys [<at 0.0115 0.9906>,<at 0.3154 1>,<at 0.5062 0>]}
       BlueCurve {keys [<at 0.3014 0>]}
       GreenCurve {keys [<at 0.3014 0>]}
       RedCurve {keys [<at 0.3014 0>]}
      }})
    R4 BurntSienna:LinearColor=&ColorFromCurve(ColorCurve=DataInterface() {propertyValues {
       AlphaCurve {keys [<at 0.0115 0.9906>,<at 0.3154 1>,<at 0.5062 0>]}
       BlueCurve {keys [<at 0.3014 0>]}
       GreenCurve {keys [<at 0.3014 0>]}
       RedCurve {keys [<at 0.3014 0>]}
      }})
    R4 Cobalt:LinearColor=&ColorFromCurve(ColorCurve=DataInterface() {propertyValues {
       AlphaCurve {keys [<at 0.0115 0.9906>,<at 0.3154 1>,<at 0.5062 0>]}
       BlueCurve {keys [<at 0.3014 0>]}
       GreenCurve {keys [<at 0.3014 0>]}
       RedCurve {keys [<at 0.3014 0>]}
      }})
    R4 Crimson:LinearColor=&ColorFromCurve(ColorCurve=DataInterface() {propertyValues {
       AlphaCurve {keys [<at 0.0115 0.9906>,<at 0.3154 1>,<at 0.5062 0>]}
       BlueCurve {keys [<at 0.0688 0.069>,<at 0.2065 0.0676>,<at 0.3014 0>]}
       GreenCurve {keys [<at 0.0688 0.069>,<at 0.2065 0.0676>,<at 0.3014 0>]}
       RedCurve {keys [<at 0.0688 0.25>,<at 0.2065 0.2448>,<at 0.3014 0>]}
      }})
    R4 DefaultColor:LinearColor=&ColorFromCurve(ColorCurve=DataInterface() {propertyValues {
       AlphaCurve {keys [<at 0.0115 0.9906>,<at 0.3154 1>,<at 0.5062 0>]}
       BlueCurve {keys [<at 0.1589 0.1979>,<at 0.3466 0>,<at 0.3491 0>]}
       GreenCurve {keys [<at 0.1589 0.0166>,<at 0.3466 0>,<at 0.3491 0>]}
       RedCurve {keys [<at 0.1589 0.105>,<at 0.3466 0>,<at 0.3491 0>]}
      }})
    R4 ForestGreen:LinearColor=&ColorFromCurve(ColorCurve=DataInterface() {propertyValues {
       AlphaCurve {keys [<at 0.0115 0.9906>,<at 0.3154 1>,<at 0.5062 0>]}
       BlueCurve {keys [<at 0.229 0.0107>,<at 0.4942 0>]}
       GreenCurve {keys [<at 0.229 0.2031>,<at 0.4942 0>]}
       RedCurve {keys [<at 0.229 0>,<at 0.4942 0>]}
      }})
    R4 Grey:LinearColor=&ColorFromCurve(ColorCurve=DataInterface() {propertyValues {
       AlphaCurve {keys [<at 0.0115 0.9906>,<at 0.3154 1>,<at 0.5062 0>]}
       BlueCurve {keys [<at 0.1914 0.1875>,<at 0.3014 0>]}
       GreenCurve {keys [<at 0.1914 0.1875>,<at 0.3014 0>]}
       RedCurve {keys [<at 0.1914 0.1875>,<at 0.3014 0>]}
      }})
    R4 Lime:LinearColor=&ColorFromCurve(ColorCurve=DataInterface() {propertyValues {
       AlphaCurve {keys [<at 0.0115 0.9906>,<at 0.3154 1>,<at 0.5062 0>]}
       BlueCurve {keys [<at 0.3014 0.0328>,<at 0.4842 0.0068>]}
       GreenCurve {keys [<at 0.3014 0.151>,<at 0.4842 0.0312>]}
       RedCurve {keys [<at 0.3014 0.0925>,<at 0.4842 0.0191>]}
      }})
    R4 Orange:LinearColor=&ColorFromCurve(ColorCurve=DataInterface() {propertyValues {
       AlphaCurve {keys [<at 0.0115 0.9906>,<at 0.3154 1>,<at 0.5062 0>]}
       BlueCurve {keys [<at 0.2565 0.0204>,<at 0.4842 0>]}
       GreenCurve {keys [<at 0.2565 0.0617>,<at 0.4842 0>]}
       RedCurve {keys [<at 0.2565 0.276>,<at 0.4842 0>]}
      }})
    R4 Pink:LinearColor=&ColorFromCurve(ColorCurve=DataInterface() {propertyValues {
       AlphaCurve {keys [<at 0.0115 0.9906>,<at 0.3154 1>,<at 0.5062 0>]}
       BlueCurve {keys [<at 0.279 0.6169>,<at 0.4817 0>]}
       GreenCurve {keys [<at 0.279 0.087>,<at 0.4817 0>]}
       RedCurve {keys [<at 0.279 0.6198>,<at 0.4817 0>]}
      }})
    R4 Purple:LinearColor=&ColorFromCurve(ColorCurve=DataInterface() {propertyValues {
       AlphaCurve {keys [<at 0.0115 0.9906>,<at 0.3154 1>,<at 0.5062 0>]}
       BlueCurve {keys [<at 0.3014 0>]}
       GreenCurve {keys [<at 0.3014 0>]}
       RedCurve {keys [<at 0.3014 0>]}
      }})
    R4 Saffron:LinearColor=&ColorFromCurve(ColorCurve=DataInterface() {propertyValues {
       AlphaCurve {keys [<at 0.0115 0.9906>,<at 0.3154 1>,<at 0.5062 0>]}
       BlueCurve {keys [<at 0.3014 0>]}
       GreenCurve {keys [<at 0.3014 0>]}
       RedCurve {keys [<at 0.3014 0>]}
      }})
    R4 SkyBlue:LinearColor=&ColorFromCurve(ColorCurve=DataInterface() {propertyValues {
       AlphaCurve {keys [<at 0.0115 0.9906>,<at 0.3154 1>,<at 0.5062 0>]}
       BlueCurve {keys [<at 0.3014 0>]}
       GreenCurve {keys [<at 0.3014 0>]}
       RedCurve {keys [<at 0.3014 0>]}
      }})
    R4 TitaniumWhite:LinearColor=&ColorFromCurve(ColorCurve=DataInterface() {propertyValues {
       AlphaCurve {keys [<at 0.0115 0.9906>,<at 0.1682 1>,<at 0.3697 0>]}
       BlueCurve {keys [<at 0.3509 0.2708>,<at 0.4792 0.0521>]}
       GreenCurve {keys [<at 0.3509 0.2708>,<at 0.4792 0.0521>]}
       RedCurve {keys [<at 0.3509 0.2708>,<at 0.4792 0.0521>]}
      }})
    R5 Black:LinearColor=<rgba 0.0469 0.0469 0.0469 0.6>
    R5 BurntSienna:LinearColor=<rgba 0.1042 0.0322 0.0128 0.7>
    R5 Cobalt:LinearColor=<rgba 0.0312 0.0401 1 0.6>
    R5 Crimson:LinearColor=<rgba 0.4635 0.0266 0.0384 0.6>
    R5 DefaultColor:LinearColor=<rgba 1 1 1 0.5>
    R5 ForestGreen:LinearColor=<rgba 0.0704 0.901 0.1124 0.4>
    R5 Grey:LinearColor=<rgba 0.1667 0.1667 0.1667 0.4>
    R5 Lime:LinearColor=<rgba 0.2938 0.5104 0.0771 0.25>
    R5 Orange:LinearColor=<rgba 0.8261 0.2993 0.0172 0.3>
    R5 Pink:LinearColor=<rgba 0.651 0.1763 0.4784 0.4>
    R5 Purple:LinearColor=<rgba 0.2704 0.1155 0.7917 0.3>
    R5 Saffron:LinearColor=<rgba 0.8906 0.6824 0 0.15>
    R5 SkyBlue:LinearColor=<rgba 0.0754 0.3822 0.5365 0.4>
    R5 TitaniumWhite:LinearColor=<rgba 0.5729 0.5729 0.5729 0.3>
    Use R1 Painted Overrides=true
    Use R2 Painted Overrides=true
    Use R3 Painted Overrides=true
    Use R4 Painted Overrides=true
    Use R5 Painted Overrides=true
   ) {module=VC_PaintedVariantsOverride001}
   &ParticleState()
   &InheritSourceMovement(
    Applied Position Scale=&Multiply_VectorByFloat(
     Vector=<1 1 1>
     Float=&FloatFromCurve(FloatCurve=DataInterface() {propertyValues {Curve {keys [{interpMode=RCIM_Cubic, tangentMode=RCTM_Break, time=0, value=1}, {interpMode=RCIM_Cubic, tangentMode=RCTM_Break, time=1, value=0}]}}})
    )
    ApplyForce=false
    ApplyPosition=true
    ApplyPositionFalloff=false
   ) {module=InheritSourceMovement001}
   &SolveForcesAndVelocity(Clamp Velocity=false, Limit Acceleration=false)
   &VC_UpdateRibbon_v1(Enable Skeletal Position=true)
   /SetVariables(Particles.DotProduct:NiagaraFloat=&DotProduct(
     Vector A=&TransformVectorByMatrix(
      Vector:Vector3f=<1 0 0>
      Transformation Matrix:NiagaraMatrix=$Engine.Owner.SystemLocalToWorld
      Transform Vector as Position:NiagaraBool=false
     )
     Vector B=&NormalizeVector(
      Vector To Normalize=&SubtractVectorFromPosition(
       Position=$Engine.Owner.Position
       Vector=&ReturnCameraPosition(Coordinate Space:ENiagaraCoordinateSpace=::Simulation, Camera Query:NiagaraDataInterfaceCamera=DataInterface())
      )
      Safe Normalize=true
      Safe Fallback Threshold=0
      Safe Fallback Vector=<0 0 0.001>
     )
    )) {isSetParameters=true, module=SetVariables_#3}
   /SetVariables(
    Particles.R1_Color_Scaled:LinearColor=&MultiplyLinearColor(
     Linear Color=$Particles.R1 Color
     Scale Factor=&MakeVector4FromVectorAndFloat(
      "Vector (RGB)"=&VectorFromFloat(Value=&FloatFromCurve(
        FloatCurve=DataInterface() {propertyValues {Curve {keys [
            {arriveTangent=-13.0287, interpMode=RCIM_Cubic, time=0.2667, value=0.9059}
            {arriveTangent=-1.5185, tangentMode=RCTM_User, time=0.4953, value=0.8118}
            <at 0.6886 0>
            <at 0.7212 1>
            {arriveTangent=-6.7757, interpMode=RCIM_Cubic, leaveTangent=-6.7757, time=0.7822, value=1}
            <at 0.8688 0>
           ]}}}
        Scale Curve=&MakeCustomFloatFromBool(Bool=$User.bIsPainted, True Float=$User.Power_EmissivePaintedR4, False Float=2)
       ))
      "Float (Alpha)"=&FloatFromCurve(
       FloatCurve=DataInterface() {propertyValues {Curve {keys [
           <at -0.0383 3.4294 RCIM_Cubic>
           {arriveTangent=-0.5546, interpMode=RCIM_Cubic, leaveTangent=-0.5546, time=0.1518, value=2.4766}
           {arriveTangent=-0.2383, interpMode=RCIM_Cubic, leaveTangent=-0.2383, tangentMode=RCTM_User, time=0.4793, value=3.1423}
           {arriveTangent=-43.3151, interpMode=RCIM_Cubic, leaveTangent=-30.5159, tangentMode=RCTM_Break, time=0.675, value=3.1272}
           {arriveTangent=-9.9908, interpMode=RCIM_Cubic, leaveTangent=-9.9908, time=0.8263, value=1.0954}
           <at 1.01 -0.2199 RCIM_Cubic>
          ]}}}
       Scale Curve=&MakeCustomFloatFromBool(Bool=$User.bIsPainted, True Float=$User.Power_AlphaPaintedR4, False Float=0.5)
      )
     )
    )
    Particles.R3_Color_Scaled:LinearColor=&MultiplyLinearColor(
     Linear Color=$Particles.R3 Color
     Scale Factor=&MakeVector4FromVectorAndFloat(
      "Vector (RGB)"=&VectorFromFloat(Value=&FloatFromCurve(
        FloatCurve=DataInterface() {propertyValues {Curve {keys [<at 0.3572 0.8681>,<at 0.49 0.675>,<at 0.6121 0.1761>,<at 0.7769 0>]}}}
        Scale Curve=&MakeCustomFloatFromBool(Bool=$User.bIsPainted, True Float=$User.Power_EmissivePaintedR4, False Float=2)
       ))
      "Float (Alpha)"=&FloatFromCurve(
       FloatCurve=DataInterface() {propertyValues {Curve {keys [<at 0.2665 1>,<at 0.4676 0.5559>,<at 0.5805 0.7676>,<at 0.7 0>]}}}
       Scale Curve=&MakeCustomFloatFromBool(Bool=$User.bIsPainted, True Float=$User.Power_AlphaPaintedR4, False Float=0.5)
      )
     )
    )
    Particles.R4_Color_Scaled:LinearColor=&MultiplyLinearColor(
     Linear Color=$Particles.R4 Color
     Scale Factor=&MakeVector4FromVectorAndFloat(
      "Vector (RGB)"=&VectorFromFloat(Value=&FloatFromCurve(
        FloatCurve=DataInterface() {propertyValues {Curve {keys [<at 0.0107 1>,<at 0.0753 1>,<at 0.3572 1>,<at 0.49 1>,<at 0.5253 1>,<at 0.6121 1>,<at 0.7769 0>]}}}
        Scale Curve=&MakeCustomFloatFromBool(Bool=$User.bIsPainted, True Float=$User.Power_EmissivePaintedR4, False Float=0.6)
       ))
      "Float (Alpha)"=&FloatFromCurve(
       FloatCurve=DataInterface() {propertyValues {Curve {keys [<at 0.106 0>,<at 0.1223 1.0029>,<at 0.2665 1>,<at 0.4464 0.9028>,<at 0.5244 0.8384>,<at 0.6688 0>]}}}
       Scale Curve=&MakeCustomFloatFromBool(Bool=$User.bIsPainted, True Float=$User.Power_AlphaPaintedR4, False Float=1.2)
      )
     )
    )
    Particles.R5_Color_Scaled:LinearColor=&MultiplyLinearColor(
     Linear Color=$Particles.R5 Color
     Scale Factor=&MakeVector4FromVectorAndFloat(
      "Vector (RGB)"=&VectorFromFloat(Value=&FloatFromCurve(
        FloatCurve=DataInterface() {propertyValues {Curve {keys [
            <at -0.17 16.4711 RCIM_Cubic>
            {arriveTangent=-27.6507, interpMode=RCIM_Cubic, leaveTangent=-27.6507, time=0.1072, value=17.8443}
            {arriveTangent=-3.2592, interpMode=RCIM_Cubic, leaveTangent=-3.2592, tangentMode=RCTM_User, time=0.3385, value=2.4113}
            {arriveTangent=2.5648, interpMode=RCIM_Cubic, leaveTangent=2.5648, tangentMode=RCTM_User, time=0.64, value=1.973}
            {arriveTangent=11.7336, interpMode=RCIM_Cubic, leaveTangent=11.7336, time=1.0714, value=20.2083}
            {arriveTangent=-34.897, interpMode=RCIM_Cubic, leaveTangent=-34.897, time=1.3482, value=10.2825}
            {arriveTangent=-8.1043, interpMode=RCIM_Cubic, leaveTangent=-8.1043, time=1.5789, value=2.4972}
            <at 1.7789 6.7918 RCIM_Cubic>
           ]}}}
        Scale Curve=&MakeCustomFloatFromBool(Bool=$User.bIsPainted, True Float=$User.Power_EmissivePaintedR4, False Float=0.7)
       ))
      "Float (Alpha)"=&FloatFromCurve(
       FloatCurve=DataInterface() {propertyValues {Curve {keys [
           <at -0.0556 3.887 RCIM_Cubic>
           {arriveTangent=-1.3925, interpMode=RCIM_Cubic, leaveTangent=-1.3925, time=0.111, value=0.7392}
           {arriveTangent=-0.2383, interpMode=RCIM_Cubic, leaveTangent=-0.2383, tangentMode=RCTM_User, time=0.4793, value=3.1423}
           {arriveTangent=-6.3349, interpMode=RCIM_Cubic, leaveTangent=-6.3349, time=0.8329, value=0.2261}
           <at 1.01 -0.2199 RCIM_Cubic>
          ]}}}
       Scale Curve=&MakeCustomFloatFromBool(Bool=$User.bIsPainted, True Float=$User.Power_AlphaPaintedR4, False Float=0.3)
      )
     )
    )
   ) {isSetParameters=true}
   &ScaleRibbonWidth(Ribbon Width Scale=&FloatFromCurve(FloatCurve=DataInterface() {propertyValues {Curve {keys [
         {arriveTangent=-0.2605, interpMode=RCIM_Cubic, leaveTangent=-0.2605, tangentMode=RCTM_User, time=0.0016, value=1.2864}
         {arriveTangent=-0.5099, interpMode=RCIM_Cubic, leaveTangent=-0.5099, tangentMode=RCTM_User, time=0.3323, value=1.3116}
         {arriveTangent=-1.7452, interpMode=RCIM_Cubic, leaveTangent=-1.7452, time=0.4222, value=1.0505}
         {arriveTangent=-1.5491, interpMode=RCIM_Cubic, leaveTangent=-1.5491, tangentMode=RCTM_User, time=0.5737, value=0.8904}
         {arriveTangent=-1.0204, interpMode=RCIM_Cubic, leaveTangent=-1.0204, time=0.6152, value=0.7069}
         {arriveTangent=-1.8279, interpMode=RCIM_Cubic, leaveTangent=-1.8279, time=0.6738, value=0.7882}
         {arriveTangent=-1.0469, tangentMode=RCTM_User, time=1.002, value=0}
        ]}}})) {module=ScaleRibbonWidth001}
   /SetVariables(
    Particles.Ribbon_BlackColor:LinearColor=<rgba 0 0 0 1>
    Particles.WaveScaleCurve:NiagaraFloat=&FloatFromCurve(
     FloatCurve=DataInterface() {propertyValues {Curve {keys [
         {arriveTangent=39.5227, interpMode=RCIM_Cubic, leaveTangent=39.5227, tangentMode=RCTM_User, time=0.0092, value=0}
         {
          arriveTangent=0.7219
          arriveTangentWeight=0.0795
          interpMode=RCIM_Cubic
          leaveTangent=0.1806
          leaveTangentWeight=0.0696
          tangentMode=RCTM_Break
          tangentWeightMode=RCTWM_WeightedBoth
          time=0.125
          value=1.25
         }
         {arriveTangent=0.1988, interpMode=RCIM_Cubic, leaveTangent=0.1988, tangentMode=RCTM_User, time=0.45, value=0.625}
        ]}}}
     Scale Curve=10
    )
   ) {isSetParameters=true, module=SetVariables_#5}
   /SetVariables(
    Particles.CustomPosition_R1:NiagaraPosition=&AddVectorToPosition(
     Position=$Particles.Position
     Vector=&Multiply_VectorByFloat(
      Vector=&MakeVector(
       X=&Sine(
        Normalized Angle=&Add_Float(A=0, B=&Subtract_Float(A=$Emitter.NormalizedLoopAge, B=$Particles.NormalizedAge))
        Period=$User.WavePeriod
        Scale=&Multiply_Float(
         A=&SelectFloatFromArray(Direct Array Index=$Particles.SpawnGroup, Float Selection Array=DataInterface() {propertyValues {FloatData [1,-1]}})
         B=$User.WaveMaxScale
        )
        Bias=0
       )
       Y=&Sine(
        Normalized Angle=&Add_Float(A=0, B=&Subtract_Float(A=$Emitter.NormalizedLoopAge, B=$Particles.NormalizedAge))
        Period=$User.WavePeriod
        Scale=&Multiply_Float(
         A=&SelectFloatFromArray(Direct Array Index=$Particles.SpawnGroup, Float Selection Array=DataInterface() {propertyValues {FloatData [1,1]}})
         B=$User.WaveMaxScale
        )
        Bias=0
       )
       Z=&Cosine(
        Normalized Angle=&Add_Float(A=0, B=&Subtract_Float(A=$Emitter.NormalizedLoopAge, B=$Particles.NormalizedAge))
        Period=$User.WavePeriod
        Scale=&Multiply_Float(
         A=&SelectFloatFromArray(Direct Array Index=$Particles.SpawnGroup, Float Selection Array=DataInterface() {propertyValues {FloatData [-1,1]}})
         B=$User.WaveMaxScale
        )
        Bias=0
       )
      )
      Float=$Particles.WaveScaleCurve
     )
    )
    Particles.CustomPosition_R2:NiagaraPosition=&AddVectorToPosition(
     Position=$Particles.Position
     Vector=&Multiply_VectorByFloat(
      Vector=&MakeVector(
       X=&Sine(
        Normalized Angle=&Add_Float(
         A=&Multiply_Float(A=$User.WavePeriod, B=0.5)
         B=&Subtract_Float(A=$Emitter.NormalizedLoopAge, B=$Particles.NormalizedAge)
        )
        Period=$User.WavePeriod
        Scale=&Multiply_Float(
         A=&SelectFloatFromArray(Direct Array Index=$Particles.SpawnGroup, Float Selection Array=DataInterface() {propertyValues {FloatData [1,1]}})
         B=$User.WaveMaxScale
        )
        Bias=0
       )
       Y=&Sine(
        Normalized Angle=&Add_Float(
         A=&Multiply_Float(A=$User.WavePeriod, B=0.5)
         B=&Subtract_Float(A=$Emitter.NormalizedLoopAge, B=$Particles.NormalizedAge)
        )
        Period=$User.WavePeriod
        Scale=&Multiply_Float(
         A=&SelectFloatFromArray(Direct Array Index=$Particles.SpawnGroup, Float Selection Array=DataInterface() {propertyValues {FloatData [1,1]}})
         B=$User.WaveMaxScale
        )
        Bias=0
       )
       Z=&Cosine(
        Normalized Angle=&Add_Float(
         A=&Multiply_Float(A=$User.WavePeriod, B=0.5)
         B=&Subtract_Float(A=$Emitter.NormalizedLoopAge, B=$Particles.NormalizedAge)
        )
        Period=$User.WavePeriod
        Scale=&Multiply_Float(
         A=&SelectFloatFromArray(Direct Array Index=$Particles.SpawnGroup, Float Selection Array=DataInterface() {propertyValues {FloatData [-1,1]}})
         B=$User.WaveMaxScale
        )
        Bias=0
       )
      )
      Float=$Particles.WaveScaleCurve
     )
    )
   ) {isSetParameters=true, module=SetVariables_#9}
   /SetVariables(Particles.DynamicMP_R2:Vector4f=&MakeVector4(
     X=&FloatFromCurve(FloatCurve=DataInterface() {propertyValues {Curve {
         defaultValue=1
         keys [
          {tangentMode=RCTM_User, time=0, value=0.5}
          {arriveTangent=-0.0633, interpMode=RCIM_Cubic, leaveTangent=-0.0633, tangentMode=RCTM_User, time=0.65, value=0.5}
          {arriveTangent=-0.0699, interpMode=RCIM_Cubic, leaveTangent=-0.0699, tangentMode=RCTM_User, time=0.75, value=-2}
         ]
        }}})
     Y=0
     Z=0
     W=0
    )) {isSetParameters=true, module=SetVariables_#7}
   &DynamicMaterialParameters(
    Index 0 Param 1=&FloatFromCurve(FloatCurve=DataInterface() {propertyValues {Curve {
        defaultValue=1
        keys [
         {
          arriveTangent=0.0405
          arriveTangentWeight=0.1835
          interpMode=RCIM_Cubic
          leaveTangent=0.0405
          leaveTangentWeight=0.1835
          tangentMode=RCTM_User
          time=0
          value=0
         }
         {arriveTangent=2.786, interpMode=RCIM_Cubic, leaveTangent=2.786, time=0.2814, value=0.2128}
         {arriveTangent=3.1289, interpMode=RCIM_Cubic, leaveTangent=3.1289, tangentMode=RCTM_User, time=0.3119, value=0.869}
         {
          arriveTangent=0.0201
          arriveTangentWeight=0.1834
          interpMode=RCIM_Cubic
          leaveTangent=0.0201
          leaveTangentWeight=0.1834
          tangentMode=RCTM_User
          time=0.5
          value=1
         }
        ]
       }}})
    Module.Index0Float0Write=true
    Module.Index0Float1Write=true
    Module.Index0Float2Write=true
    Module.Index0Float3Write=true
   )
  ]
  rendererClasses [&NiagaraRibbonRendererProperties]
  renderers [
   {rendererClass=&NiagaraRibbonRendererProperties}
   {rendererClass=&NiagaraRibbonRendererProperties, rendererIndex=1}
   {rendererClass=&NiagaraRibbonRendererProperties, rendererIndex=2}
   {rendererClass=&NiagaraRibbonRendererProperties, rendererIndex=3}
  ]
 }
 NE_BlackFlame_Mid1 {
  bEnabled=false
  emitterUpdateScript [
   &EmitterState(Life Cycle Mode=::NewEnumerator1, Loop Duration=1, UseLoopDelay=false)
   &VC_FilteredSocketCount()
   &VC_RateQualityHandler(Epic Spawn Rate=30, High Spawn Rate=30, Low Spawn Rate=30, Use Custom Rate?=true)
   &VC_SpawnRibbon_v1(Number of Ribbons=$Emitter.PortCount, Spawn Time=-1, SpawnRate=$Emitter.Spawn Rate)
  ]
  particleSpawnScript [
   &VC_PaintedVariantsOverride() {enabled=false}
   &VC_InitializeRibbon_v1(
    Color=$Particles.Color
    Custom Offset=true
    Enable Skeletal Position=true
    Lifetime=&MergeHelper_FloatSwitch(0=1, 1=1, 2=1, 3=?, Index=$User.GameModeID, Num=::NewEnumerator1)
    Local Vector Offset=&AddVectorToPosition(Position=<30 0 2.5>, Vector=<0 0 0>)
    Offset Amount=1
    Ribbon Width=55
   )
   &AddVelocity(Velocity=<-2000 0 0>) {module=AddVelocity001}
   /SetVariables(
    Particles.R1_Width:NiagaraFloat=&SelectFloatFromArray(Direct Array Index=&SpawnGroup(), Float Selection Array=DataInterface() {propertyValues {FloatData [75,0]}})
    Particles.R2_Width:NiagaraFloat=&SelectFloatFromArray(Direct Array Index=&SpawnGroup(), Float Selection Array=DataInterface() {propertyValues {FloatData [0,-75]}})
    Particles.R3_Color:LinearColor=<rgba 0.2241 0.5211 0.6609 3>
   ) {enabled=false, isSetParameters=true, module=SetVariables_#10}
  ]
  particleUpdateScript [
   &VC_PaintedVariantsOverride(
    R1 Black:LinearColor=&ColorFromCurve(ColorCurve=DataInterface() {propertyValues {
       AlphaCurve {keys [
         {tangentMode=RCTM_User, time=0.1264, value=0}
         {tangentMode=RCTM_User, time=0.2268, value=0.5}
         <at 0.5035 0.5>
         {tangentMode=RCTM_User, time=0.557, value=0}
        ]}
       BlueCurve {keys [<at 1 0.1094 RCIM_Cubic>]}
       GreenCurve {keys [<at 1 0.1094 RCIM_Cubic>]}
       RedCurve {keys [<at 1 0.1094 RCIM_Cubic>]}
      }})
    R1 BurntSienna:LinearColor=&ColorFromCurve(ColorCurve=DataInterface() {propertyValues {
       AlphaCurve {
        defaultValue=7
        keys [
         {tangentMode=RCTM_User, time=0.1264, value=0}
         {tangentMode=RCTM_User, time=0.2268, value=0.5}
         <at 0.5035 0.5>
         {tangentMode=RCTM_User, time=0.557, value=0}
        ]
       }
       BlueCurve {keys [<at 1 0.0686 RCIM_Cubic>]}
       GreenCurve {keys [<at 1 0.0858 RCIM_Cubic>]}
       RedCurve {keys [<at 1 0.1667 RCIM_Cubic>]}
      }})
    R1 Cobalt:LinearColor=&ColorFromCurve(ColorCurve=DataInterface() {propertyValues {
       AlphaCurve {
        defaultValue=7
        keys [
         {tangentMode=RCTM_User, time=0.1264, value=0}
         {tangentMode=RCTM_User, time=0.2268, value=0.5}
         <at 0.5035 0.5>
         {tangentMode=RCTM_User, time=0.557, value=0}
        ]
       }
       BlueCurve {keys [<at 1 0.3802 RCIM_Cubic>]}
       GreenCurve {keys [<at 1 0.0931 RCIM_Cubic>]}
       RedCurve {keys [<at 1 0.0931 RCIM_Cubic>]}
      }})
    R1 Crimson:LinearColor=&ColorFromCurve(ColorCurve=DataInterface() {propertyValues {
       AlphaCurve {
        defaultValue=7
        keys [
         {tangentMode=RCTM_User, time=0.1264, value=0}
         {tangentMode=RCTM_User, time=0.2268, value=0.5}
         <at 0.5035 0.5>
         {tangentMode=RCTM_User, time=0.557, value=0}
        ]
       }
       BlueCurve {keys [<at 1 0.0131 RCIM_Cubic>]}
       GreenCurve {keys [<at 1 0 RCIM_Cubic>]}
       RedCurve {keys [<at 1 0.2348 RCIM_Cubic>]}
      }})
    R1 DefaultColor=&ColorFromCurve(ColorCurve=DataInterface() {propertyValues {
       AlphaCurve {keys [<at 0.1264 0>,<at 0.2268 0.5>,<at 0.5035 0.5>,<at 0.557 0>]}
       BlueCurve {keys [<at 1 0.0999 RCIM_Cubic>]}
       GreenCurve {keys [<at 1 0 RCIM_Cubic>]}
       RedCurve {keys [<at 1 0.0513 RCIM_Cubic>]}
      }})
    R1 ForestGreen:LinearColor=&ColorFromCurve(ColorCurve=DataInterface() {propertyValues {
       AlphaCurve {
        defaultValue=7
        keys [
         {tangentMode=RCTM_User, time=0.1264, value=0}
         {tangentMode=RCTM_User, time=0.2268, value=0.5}
         <at 0.5035 0.5>
         {tangentMode=RCTM_User, time=0.557, value=0}
        ]
       }
       BlueCurve {keys [<at 1 0.0198 RCIM_Cubic>]}
       GreenCurve {keys [<at 1 0.2031 RCIM_Cubic>]}
       RedCurve {keys [<at 1 0 RCIM_Cubic>]}
      }})
    R1 Grey:LinearColor=&ColorFromCurve(ColorCurve=DataInterface() {propertyValues {
       AlphaCurve {
        defaultValue=1
        keys [
         {tangentMode=RCTM_User, time=0.1264, value=0}
         {tangentMode=RCTM_User, time=0.2268, value=0.5}
         <at 0.5035 0.5>
         {tangentMode=RCTM_User, time=0.557, value=0}
        ]
       }
       BlueCurve {keys [<at 1 0.2344 RCIM_Cubic>]}
       GreenCurve {keys [<at 1 0.2344 RCIM_Cubic>]}
       RedCurve {keys [<at 1 0.2344 RCIM_Cubic>]}
      }})
    R1 Lime:LinearColor=&ColorFromCurve(ColorCurve=DataInterface() {propertyValues {
       AlphaCurve {
        defaultValue=7
        keys [
         {tangentMode=RCTM_User, time=0.1264, value=0}
         {tangentMode=RCTM_User, time=0.2268, value=0.5}
         <at 0.5035 0.5>
         {tangentMode=RCTM_User, time=0.557, value=0}
        ]
       }
       BlueCurve {keys [<at 1 0.0865 RCIM_Cubic>]}
       GreenCurve {keys [<at 1 0.2865 RCIM_Cubic>]}
       RedCurve {keys [<at 1 0.1697 RCIM_Cubic>]}
      }})
    R1 Orange:LinearColor=&ColorFromCurve(ColorCurve=DataInterface() {propertyValues {
       AlphaCurve {
        defaultValue=7
        keys [
         {tangentMode=RCTM_User, time=0.1264, value=0}
         {tangentMode=RCTM_User, time=0.2268, value=0.5}
         <at 0.5035 0.5>
         {tangentMode=RCTM_User, time=0.557, value=0}
        ]
       }
       BlueCurve {keys [<at 1 0.0688 RCIM_Cubic>]}
       GreenCurve {keys [<at 1 0.1085 RCIM_Cubic>]}
       RedCurve {keys [<at 1 0.3478 RCIM_Cubic>]}
      }})
    R1 Pink:LinearColor=&ColorFromCurve(ColorCurve=DataInterface() {propertyValues {
       AlphaCurve {
        defaultValue=7
        keys [
         {tangentMode=RCTM_User, time=0.1264, value=0}
         {tangentMode=RCTM_User, time=0.2268, value=0.5}
         <at 0.5035 0.5>
         {tangentMode=RCTM_User, time=0.557, value=0}
        ]
       }
       BlueCurve {keys [<at 1 0.2004 RCIM_Cubic>]}
       GreenCurve {keys [<at 1 0 RCIM_Cubic>]}
       RedCurve {keys [<at 1 0.2188 RCIM_Cubic>]}
      }})
    R1 Purple:LinearColor=&ColorFromCurve(ColorCurve=DataInterface() {propertyValues {
       AlphaCurve {
        defaultValue=7
        keys [
         {tangentMode=RCTM_User, time=0.1264, value=0}
         {tangentMode=RCTM_User, time=0.2268, value=0.5}
         <at 0.5035 0.5>
         {tangentMode=RCTM_User, time=0.557, value=0}
        ]
       }
       BlueCurve {keys [<at 1 0.375 RCIM_Cubic>]}
       GreenCurve {keys [<at 1 0 RCIM_Cubic>]}
       RedCurve {keys [<at 1 0.1484 RCIM_Cubic>]}
      }})
    R1 Saffron:LinearColor=&ColorFromCurve(ColorCurve=DataInterface() {propertyValues {
       AlphaCurve {
        defaultValue=7
        keys [
         {tangentMode=RCTM_User, time=0.1264, value=0}
         {tangentMode=RCTM_User, time=0.2268, value=0.5}
         <at 0.5035 0.5>
         {tangentMode=RCTM_User, time=0.557, value=0}
        ]
       }
       BlueCurve {keys [<at 1 0 RCIM_Cubic>]}
       GreenCurve {keys [<at 1 0.2395 RCIM_Cubic>]}
       RedCurve {keys [<at 1 0.2812 RCIM_Cubic>]}
      }})
    R1 SkyBlue:LinearColor=&ColorFromCurve(ColorCurve=DataInterface() {propertyValues {
       AlphaCurve {
        defaultValue=7
        keys [
         {tangentMode=RCTM_User, time=0.1264, value=0}
         {tangentMode=RCTM_User, time=0.2268, value=0.5}
         <at 0.5035 0.5>
         {tangentMode=RCTM_User, time=0.557, value=0}
        ]
       }
       BlueCurve {keys [<at 1 0.3802 RCIM_Cubic>]}
       GreenCurve {keys [<at 1 0.1905 RCIM_Cubic>]}
       RedCurve {keys [<at 1 0 RCIM_Cubic>]}
      }})
    R1 TitaniumWhite:LinearColor=&ColorFromCurve(ColorCurve=DataInterface() {propertyValues {
       AlphaCurve {
        defaultValue=2
        keys [
         {tangentMode=RCTM_User, time=0.1264, value=0}
         {tangentMode=RCTM_User, time=0.2268, value=0.5}
         <at 0.5035 0.5>
         {tangentMode=RCTM_User, time=0.557, value=0}
        ]
       }
       BlueCurve {keys [<at 1 0.6406 RCIM_Cubic>]}
       GreenCurve {keys [<at 1 0.6406 RCIM_Cubic>]}
       RedCurve {keys [<at 1 0.6406 RCIM_Cubic>]}
      }})
    R2 Black:LinearColor=&ColorFromCurve(ColorCurve=DataInterface() {propertyValues {
       AlphaCurve {keys [<at 0 7>,<at 0.2232 1>,<at 0.3261 0.05>,<at 0.5495 0.2>,<at 0.7007 0.2>,<at 0.8474 0 RCIM_Cubic>]}
       BlueCurve {keys [<at 0 0.0194 RCIM_Cubic>,<at 0.1492 0.0333>,<at 0.2681 0.1104>,<at 0.4863 0.9274>,<at 0.6088 0.0305>]}
       GreenCurve {keys [<at 0 0.0194 RCIM_Cubic>,<at 0.1492 0.0333>,<at 0.2681 0.1104>,<at 0.4863 0.9274>,<at 0.6088 0.0305>]}
       RedCurve {keys [<at 0 0.0194 RCIM_Cubic>,<at 0.1492 0.0333>,<at 0.2681 0.1104>,<at 0.4863 0.9274>,<at 0.6088 0.0305>]}
      }})
    R2 BurntSienna:LinearColor=&ColorFromCurve(ColorCurve=DataInterface() {propertyValues {
       AlphaCurve {keys [<at 0 4>,<at 0.1907 1>,<at 0.3261 0.05>,<at 0.593 0.2>,<at 0.8474 0 RCIM_Cubic>]}
       BlueCurve {keys [<at 0 0.008 RCIM_Cubic>,<at 0.2839 0.0202>,<at 0.4863 0.2907>,<at 0.5956 0.008>]}
       GreenCurve {keys [<at 0 0.0203 RCIM_Cubic>,<at 0.2839 0.0323>,<at 0.4863 0.4394>,<at 0.5956 0.0203>]}
       RedCurve {keys [<at 0 0.0648 RCIM_Cubic>,<at 0.2839 0.0762>,<at 0.4863 0.9792>,<at 0.5956 0.0648>]}
      }})
    R2 Cobalt:LinearColor=&ColorFromCurve(ColorCurve=DataInterface() {propertyValues {
       AlphaCurve {keys [<at 0 7>,<at 0.1907 1>,<at 0.3261 0.05>,<at 0.5495 0.6>,<at 0.7007 0.2>,<at 0.8474 0 RCIM_Cubic>]}
       BlueCurve {keys [<at 0 0.6458 RCIM_Cubic>,<at 0.2575 0.9985>,<at 0.4863 1>,<at 0.6115 0.8177>]}
       GreenCurve {keys [<at 0 0.0119 RCIM_Cubic>,<at 0.2575 0.0367>,<at 0.4863 0.2997>,<at 0.6115 0.0527>]}
       RedCurve {keys [<at 0 0 RCIM_Cubic>,<at 0.2575 0.0186>,<at 0.4863 0.2865>,<at 0.6115 0.0383>]}
      }})
    R2 Crimson:LinearColor=&ColorFromCurve(ColorCurve=DataInterface() {propertyValues {
       AlphaCurve {keys [<at 0 7>,<at 0.1907 1>,<at 0.3261 0.05>,<at 0.5495 0.2>,<at 0.7007 0.2>,<at 0.8474 0 RCIM_Cubic>]}
       BlueCurve {keys [<at 0 0 RCIM_Cubic>,<at 0.1492 0.0104>,<at 0.4863 0.2995>,<at 0.6458 0.0894>]}
       GreenCurve {keys [<at 0 0 RCIM_Cubic>,<at 0.1492 0.0104>,<at 0.4863 0.2995>,<at 0.6458 0.0894>]}
       RedCurve {keys [<at 0 0.4323 RCIM_Cubic>,<at 0.1492 0.9992>,<at 0.4863 0.9274>,<at 0.6458 0.993>]}
      }})
    R2 DefaultColor:LinearColor=&ColorFromCurve(ColorCurve=DataInterface() {propertyValues {
       AlphaCurve {keys [<at 0 1>,<at 0.0914 0.2857>,<at 0.308 0.05>,<at 0.5592 0.2>,<at 0.8632 0>]}
       BlueCurve {keys [<at 0.0009 0>,<at 0.2668 -0.2829>,<at 0.5114 0.8849>,<at 0.619 1>,<at 0.682 0.383>]}
       GreenCurve {keys [<at 0.0009 0.0767>,<at 0.2742 -0.1994>,<at 0.5114 0.8849>,<at 0.619 1>,<at 0.682 0.2585>]}
       RedCurve {keys [<at 0.0009 3.2194>,<at 0.0016 3.2194>,<at 0.4479 0.8849>,<at 0.5114 0.8849>,<at 0.619 1>]}
      }})
    R2 ForestGreen:LinearColor=&ColorFromCurve(ColorCurve=DataInterface() {propertyValues {
       AlphaCurve {keys [<at 0 6>,<at 0.1907 1>,<at 0.3261 0.05>,<at 0.5495 0.2>,<at 0.7007 0.2>,<at 0.8474 0 RCIM_Cubic>]}
       BlueCurve {keys [<at 0 0.0172 RCIM_Cubic>,<at 0.2179 0.0322>,<at 0.4863 0.4656>,<at 0.6537 0.1027>]}
       GreenCurve {keys [<at 0 0.3438 RCIM_Cubic>,<at 0.2179 0.245>,<at 0.4863 0.875>,<at 0.6537 0.2987>]}
       RedCurve {keys [<at 0 0 RCIM_Cubic>,<at 0.2179 0.021>,<at 0.4863 0.4466>,<at 0.6537 0.0924>]}
      }})
    R2 Grey:LinearColor=&ColorFromCurve(ColorCurve=DataInterface() {propertyValues {
       AlphaCurve {keys [<at 0 4>,<at 0.1907 1>,<at 0.3261 0.05>,<at 0.5495 0.2>,<at 0.7007 0.2>,<at 0.8474 0 RCIM_Cubic>]}
       BlueCurve {keys [<at 0 0.0692 RCIM_Cubic>,<at 0.2285 0.1336>,<at 0.4863 0.9274>,<at 0.6405 0.132>]}
       GreenCurve {keys [<at 0 0.0781 RCIM_Cubic>,<at 0.2285 0.1336>,<at 0.4863 0.9274>,<at 0.6405 0.132>]}
       RedCurve {keys [<at 0 0.0767 RCIM_Cubic>,<at 0.2285 0.1406>,<at 0.4863 0.9274>,<at 0.6405 0.1354>]}
      }})
    R2 Lime:LinearColor=&ColorFromCurve(ColorCurve=DataInterface() {propertyValues {
       AlphaCurve {keys [<at 0 6>,<at 0.1907 1>,<at 0.3261 0.05>,<at 0.5495 0.2>,<at 0.8122 0 RCIM_Cubic>]}
       BlueCurve {keys [<at 0 0 RCIM_Cubic>,<at 0.2179 0.0459>,<at 0.4863 0.9274>,<at 0.6299 0.0724>]}
       GreenCurve {keys [<at 0 0.6875 RCIM_Cubic>,<at 0.2179 0.4635>,<at 0.4863 0.9274>,<at 0.6299 0.7708>]}
       RedCurve {keys [<at 0 0.3447 RCIM_Cubic>,<at 0.2179 0.2553>,<at 0.4863 0.9274>,<at 0.6299 0.4226>]}
      }})
    R2 Orange:LinearColor=&ColorFromCurve(ColorCurve=DataInterface() {propertyValues {
       AlphaCurve {keys [<at 0 7>,<at 0.1907 1>,<at 0.3261 0.05>,<at 0.5495 0.2>,<at 0.7007 0.2>,<at 0.8474 0 RCIM_Cubic>]}
       BlueCurve {keys [<at 0 0 RCIM_Cubic>,<at 0.2364 0.0598>,<at 0.4863 0.4263>,<at 0.6217 0.1191>]}
       GreenCurve {keys [<at 0 0.2834 RCIM_Cubic>,<at 0.2364 0.2792>,<at 0.4863 0.5719>,<at 0.6217 0.3121>]}
       RedCurve {keys [<at 0 0.8125 RCIM_Cubic>,<at 0.2364 0.6889>,<at 0.4863 0.8438>,<at 0.6217 0.6724>]}
      }})
    R2 Pink:LinearColor=&ColorFromCurve(ColorCurve=DataInterface() {propertyValues {
       AlphaCurve {keys [<at 0 7>,<at 0.1907 1>,<at 0.3261 0.05>,<at 0.5495 0.2>,<at 0.7007 0.2>,<at 0.8474 0 RCIM_Cubic>]}
       BlueCurve {keys [<at 0 0.3419 RCIM_Cubic>,<at 0.2707 0.3564>,<at 0.4863 0.7216>,<at 0.6141 0.5634>]}
       GreenCurve {keys [<at 0 0.0612 RCIM_Cubic>,<at 0.2707 0.1489>,<at 0.4863 0.5789>,<at 0.6141 0.2536>]}
       RedCurve {keys [<at 0 0.5029 RCIM_Cubic>,<at 0.2707 0.5614>,<at 0.4863 0.855>,<at 0.6141 0.8542>]}
      }})
    R2 Purple:LinearColor=&ColorFromCurve(ColorCurve=DataInterface() {propertyValues {
       AlphaCurve {keys [<at 0 7>,<at 0.1907 1>,<at 0.3261 0.05>,<at 0.5495 0.2>,<at 0.7007 0.2>,<at 0.8474 0 RCIM_Cubic>]}
       BlueCurve {keys [<at 0 0.8281 RCIM_Cubic>,<at 0.2919 0.6836>,<at 0.4863 0.8957>,<at 0.6141 0.9565>]}
       GreenCurve {keys [<at 0 0 RCIM_Cubic>,<at 0.2919 0.0563>,<at 0.4863 0.2791>,<at 0.6141 0.1097>]}
       RedCurve {keys [<at 0 0.1924 RCIM_Cubic>,<at 0.2919 0.3349>,<at 0.4863 0.5023>,<at 0.6141 0.3436>]}
      }})
    R2 Saffron:LinearColor=&ColorFromCurve(ColorCurve=DataInterface() {propertyValues {
       AlphaCurve {keys [<at 0 4>,<at 0.1907 1>,<at 0.3261 0.05>,<at 0.5495 0.2>,<at 0.8175 0 RCIM_Cubic>]}
       BlueCurve {keys [<at 0 0 RCIM_Cubic>,<at 0.2697 0.0046>,<at 0.4863 0.3188>,<at 0.6748 0.0798>]}
       GreenCurve {keys [<at 0 0.6105 RCIM_Cubic>,<at 0.2697 0.6881>,<at 0.4863 0.9274>,<at 0.6748 0.6052>]}
       RedCurve {keys [<at 0 0.7969 RCIM_Cubic>,<at 0.2697 0.8967>,<at 0.4863 0.8744>,<at 0.6748 0.7656>]}
      }})
    R2 SkyBlue:LinearColor=&ColorFromCurve(ColorCurve=DataInterface() {propertyValues {
       AlphaCurve {keys [<at 0 7>,<at 0.1907 1>,<at 0.3261 0.05>,<at 0.5495 0.2>,<at 0.8096 0 RCIM_Cubic>]}
       BlueCurve {keys [<at 0 0.5333 RCIM_Cubic>,<at 0.2496 0.5525>,<at 0.4863 0.9323>,<at 0.6141 0.9323>]}
       GreenCurve {keys [<at 0 0.3564 RCIM_Cubic>,<at 0.2496 0.3792>,<at 0.4863 0.8083>,<at 0.6141 0.6988>]}
       RedCurve {keys [<at 0 0 RCIM_Cubic>,<at 0.2496 0.0875>,<at 0.4863 0.5584>,<at 0.6141 0.2282>]}
      }})
    R2 TitaniumWhite:LinearColor=&ColorFromCurve(ColorCurve=DataInterface() {propertyValues {
       AlphaCurve {keys [<at 0 4>,<at 0.1907 1>,<at 0.3261 0.05>,<at 0.5495 0.2>,<at 0.7007 0.2>,<at 0.8175 0 RCIM_Cubic>]}
       BlueCurve {keys [<at 0 0.1846 RCIM_Cubic>,<at 0.2285 0.1336>,<at 0.4863 0.9274>,<at 0.6405 0.3248>]}
       GreenCurve {keys [<at 0 0.2083 RCIM_Cubic>,<at 0.2285 0.1336>,<at 0.4863 0.9274>,<at 0.6405 0.3248>]}
       RedCurve {keys [<at 0 0.2046 RCIM_Cubic>,<at 0.2285 0.1406>,<at 0.4863 0.9274>,<at 0.6405 0.3333>]}
      }})
    R3 Black:LinearColor=&ColorFromCurve(ColorCurve=DataInterface() {propertyValues {
       AlphaCurve {defaultValue=0, keys [{tangentMode=RCTM_User, time=0.2127, value=0}, <at 0.3257 1>, <at 0.5324 1>, <at 0.6291 0>]}
       BlueCurve {keys [<at 0.6966 0.3043>]}
       GreenCurve {keys [<at 0.6966 0.3043>]}
       RedCurve {keys [<at 0.6966 0.3043>]}
      }})
    R3 BurntSienna:LinearColor=&ColorFromCurve(ColorCurve=DataInterface() {propertyValues {
       AlphaCurve {defaultValue=0, keys [{tangentMode=RCTM_User, time=0.2127, value=0}, <at 0.3257 1>, <at 0.5324 1>, <at 0.6291 0>]}
       BlueCurve {keys [<at 0.6966 0.2729>]}
       GreenCurve {keys [<at 0.6966 0.3693>]}
       RedCurve {keys [<at 0.6966 0.6719>]}
      }})
    R3 Cobalt:LinearColor=&ColorFromCurve(ColorCurve=DataInterface() {propertyValues {
       AlphaCurve {defaultValue=0, keys [{tangentMode=RCTM_User, time=0.2127, value=0}, <at 0.3257 1>, <at 0.5324 1>, <at 0.6291 0>]}
       BlueCurve {keys [<at 0.6966 1.5>]}
       GreenCurve {keys [<at 0.6966 0.0598>]}
       RedCurve {keys [<at 0.6966 0>]}
      }})
    R3 Crimson:LinearColor=&ColorFromCurve(ColorCurve=DataInterface() {propertyValues {
       AlphaCurve {defaultValue=0, keys [{tangentMode=RCTM_User, time=0.2127, value=0}, <at 0.3257 1>, <at 0.5324 1>, <at 0.6291 0>]}
       BlueCurve {keys [<at 0.6966 0>]}
       GreenCurve {keys [<at 0.6966 0>]}
       RedCurve {keys [<at 0.6966 1>]}
      }})
    R3 DefaultColor:LinearColor=&ColorFromCurve(ColorCurve=DataInterface() {propertyValues {
       AlphaCurve {keys [<at 0.2127 0>,<at 0.3257 1>,<at 0.5324 1>,<at 0.6291 0>]}
       BlueCurve {keys [<at 0.6966 2>]}
       GreenCurve {keys [<at 0.6966 0.5201>]}
       RedCurve {keys [<at 0.6966 0>]}
      }})
    R3 ForestGreen:LinearColor=&ColorFromCurve(ColorCurve=DataInterface() {propertyValues {
       AlphaCurve {defaultValue=0, keys [{tangentMode=RCTM_User, time=0.2127, value=0}, <at 0.3257 1>, <at 0.5324 1>, <at 0.6291 0>]}
       BlueCurve {keys [<at 0.6966 0.3387>]}
       GreenCurve {keys [<at 0.6966 1.5>]}
       RedCurve {keys [<at 0.6966 0>]}
      }})
    R3 Grey:LinearColor=&ColorFromCurve(ColorCurve=DataInterface() {propertyValues {
       AlphaCurve {keys [{tangentMode=RCTM_User, time=0.2127, value=0}, <at 0.3257 1>, <at 0.5324 1>, <at 0.6291 0>, <at 0.7715 0>]}
       BlueCurve {keys [<at 0.6966 0.5>]}
       GreenCurve {keys [<at 0.6966 0.5>]}
       RedCurve {keys [<at 0.6966 0.5>]}
      }})
    R3 Lime:LinearColor=&ColorFromCurve(ColorCurve=DataInterface() {propertyValues {
       AlphaCurve {defaultValue=0, keys [{tangentMode=RCTM_User, time=0.2127, value=0}, <at 0.3257 1>, <at 0.5324 1>, <at 0.6291 0>]}
       BlueCurve {keys [<at 0.6966 0>]}
       GreenCurve {keys [<at 0.6966 1>]}
       RedCurve {keys [<at 0.6966 0.5901>]}
      }})
    R3 Orange:LinearColor=&ColorFromCurve(ColorCurve=DataInterface() {propertyValues {
       AlphaCurve {defaultValue=0, keys [{tangentMode=RCTM_User, time=0.2127, value=0}, <at 0.3257 1>, <at 0.5324 1>, <at 0.6291 0>]}
       BlueCurve {keys [<at 0.6966 0.0304>]}
       GreenCurve {keys [<at 0.6966 0.2737>]}
       RedCurve {keys [<at 0.6966 0.7292>]}
      }})
    R3 Pink:LinearColor=&ColorFromCurve(ColorCurve=DataInterface() {propertyValues {
       AlphaCurve {defaultValue=0, keys [{tangentMode=RCTM_User, time=0.2127, value=0}, <at 0.3257 1>, <at 0.5324 1>, <at 0.6291 0>]}
       BlueCurve {keys [<at 0.6966 1.2188>]}
       GreenCurve {keys [<at 0.6966 0.2158>]}
       RedCurve {keys [<at 0.6966 1.1488>]}
      }})
    R3 Purple:LinearColor=&ColorFromCurve(ColorCurve=DataInterface() {propertyValues {
       AlphaCurve {defaultValue=0, keys [{tangentMode=RCTM_User, time=0.2127, value=0}, <at 0.3257 1>, <at 0.5324 1>, <at 0.6291 0>]}
       BlueCurve {keys [<at 0.6966 1>]}
       GreenCurve {keys [<at 0.6966 0>]}
       RedCurve {keys [<at 0.6966 0.3627>]}
      }})
    R3 Saffron:LinearColor=&ColorFromCurve(ColorCurve=DataInterface() {propertyValues {
       AlphaCurve {defaultValue=0, keys [{tangentMode=RCTM_User, time=0.2127, value=0}, <at 0.3257 1>, <at 0.5324 1>, <at 0.6291 0>]}
       BlueCurve {keys [<at 0.6966 0>]}
       GreenCurve {keys [<at 0.6966 1.8157>]}
       RedCurve {keys [<at 0.6966 2>]}
      }})
    R3 SkyBlue:LinearColor=&ColorFromCurve(ColorCurve=DataInterface() {propertyValues {
       AlphaCurve {defaultValue=0, keys [{tangentMode=RCTM_User, time=0.2127, value=0}, <at 0.3257 1>, <at 0.5324 1>, <at 0.6291 0>]}
       BlueCurve {keys [<at 0.6966 1.5>]}
       GreenCurve {keys [<at 0.6966 0.3855>]}
       RedCurve {keys [<at 0.6966 0>]}
      }})
    R3 TitaniumWhite:LinearColor=&ColorFromCurve(ColorCurve=DataInterface() {propertyValues {
       AlphaCurve {keys [{tangentMode=RCTM_User, time=0.2127, value=0}, <at 0.3257 1>, <at 0.5324 1>, <at 0.6291 0>, <at 0.9629 0>]}
       BlueCurve {keys [<at 0.6966 2.5>]}
       GreenCurve {keys [<at 0.6966 2.5>]}
       RedCurve {keys [<at 0.6966 2.5>]}
      }})
    R4 BurntSienna:LinearColor=<rgba 0.0312 0.0096 0.0039 0.6>
    R4 Cobalt:LinearColor=<rgba 0 0.0115 0.6198 0.7>
    R4 Crimson:LinearColor=<rgba 0.2656 0 0.0072 0.7>
    R4 ForestGreen:LinearColor=<rgba 0 0.2552 0.0129 0.6>
    R4 Grey:LinearColor=<rgba 0.0417 0.0417 0.0417 0.7>
    R4 Lime:LinearColor=<rgba 0.1875 0.375 0 0.6>
    R4 Orange:LinearColor=<rgba 0.5521 0.1706 0 0.45>
    R4 Pink:LinearColor=<rgba 0.3906 0.0468 0.2656 0.5>
    R4 Purple:LinearColor=<rgba 0.1567 0 0.7552 0.5>
    R4 Saffron:LinearColor=<rgba 0.5573 0.4299 0 0.3>
    R4 SkyBlue:LinearColor=<rgba 0.0137 0.2957 0.4375 0.5>
    R4 TitaniumWhite:LinearColor=<rgba 0.224 0.224 0.224 0.6>
    R5 Black:LinearColor=<rgba 0.0469 0.0469 0.0469 0.6>
    R5 BurntSienna:LinearColor=<rgba 0.1042 0.0322 0.0128 0.7>
    R5 Cobalt:LinearColor=<rgba 0.0312 0.0401 1 0.6>
    R5 Crimson:LinearColor=<rgba 0.4635 0.0266 0.0384 0.6>
    R5 DefaultColor:LinearColor=<rgba 1 1 1 0.5>
    R5 ForestGreen:LinearColor=<rgba 0.0704 0.901 0.1124 0.4>
    R5 Grey:LinearColor=<rgba 0.1667 0.1667 0.1667 0.4>
    R5 Lime:LinearColor=<rgba 0.2938 0.5104 0.0771 0.25>
    R5 Orange:LinearColor=<rgba 0.8261 0.2993 0.0172 0.3>
    R5 Pink:LinearColor=<rgba 0.651 0.1763 0.4784 0.4>
    R5 Purple:LinearColor=<rgba 0.2704 0.1155 0.7917 0.3>
    R5 Saffron:LinearColor=<rgba 0.8906 0.6824 0 0.15>
    R5 SkyBlue:LinearColor=<rgba 0.0754 0.3822 0.5365 0.4>
    R5 TitaniumWhite:LinearColor=<rgba 0.5729 0.5729 0.5729 0.3>
    Use R1 Painted Overrides=true
    Use R2 Painted Overrides=true
    Use R3 Painted Overrides=true
    Use R4 Painted Overrides=true
    Use R5 Painted Overrides=true
   ) {module=VC_PaintedVariantsOverride001}
   &ParticleState()
   &InheritSourceMovement(
    Applied Position Scale=&Multiply_VectorByFloat(
     Vector=<1 1 1>
     Float=&FloatFromCurve(FloatCurve=DataInterface() {propertyValues {Curve {keys [{interpMode=RCIM_Cubic, tangentMode=RCTM_Break, time=0, value=1}, {interpMode=RCIM_Cubic, tangentMode=RCTM_Break, time=1, value=0}]}}})
    )
    ApplyForce=false
    ApplyPosition=true
    ApplyPositionFalloff=false
   ) {module=InheritSourceMovement001}
   &SolveForcesAndVelocity(Clamp Velocity=false, Limit Acceleration=false)
   &VC_UpdateRibbon_v1(Enable Skeletal Position=true)
   /SetVariables(Particles.DotProduct:NiagaraFloat=&DotProduct(
     Vector A=&TransformVectorByMatrix(
      Vector:Vector3f=<1 0 0>
      Transformation Matrix:NiagaraMatrix=$Engine.Owner.SystemLocalToWorld
      Transform Vector as Position:NiagaraBool=false
     )
     Vector B=&NormalizeVector(
      Vector To Normalize=&SubtractVectorFromPosition(
       Position=$Engine.Owner.Position
       Vector=&ReturnCameraPosition(Coordinate Space:ENiagaraCoordinateSpace=::Simulation, Camera Query:NiagaraDataInterfaceCamera=DataInterface())
      )
      Safe Normalize=true
      Safe Fallback Threshold=0
      Safe Fallback Vector=<0 0 0.001>
     )
    )) {isSetParameters=true, module=SetVariables_#3}
   /SetVariables(
    Particles.RibbonBase:LinearColor=&ColorFromCurve(ColorCurve=DataInterface() {propertyValues {
       AlphaCurve {keys [<at 0.277 0>,<at 0.4251 1>,<at 0.4666 1>,<at 0.5835 0>]}
       BlueCurve {keys [<at 0.9243 0>]}
       GreenCurve {keys [<at 0.9243 0>]}
       RedCurve {keys [<at 0.9243 0>]}
      }})
    Particles.RibbonGlow:LinearColor=&ColorFromCurve(ColorCurve=DataInterface() {propertyValues {
       AlphaCurve {keys [<at 0.2847 0>,<at 0.4022 1>,<at 0.4726 1>,<at 0.5529 0>]}
       BlueCurve {keys [<at 0.6966 2>]}
       GreenCurve {keys [<at 0.6966 0.5201>]}
       RedCurve {keys [<at 0.6966 0>]}
      }})
    Particles.RibbonGradient:LinearColor=&ColorFromCurve(ColorCurve=DataInterface() {propertyValues {
       AlphaCurve {keys [<at 0.3 0>,<at 0.4456 0.5>,<at 0.522 0.25>,<at 0.5858 0>]}
       BlueCurve {keys [<at 1 0.1683 RCIM_Cubic>]}
       GreenCurve {keys [<at 1 0 RCIM_Cubic>]}
       RedCurve {keys [<at 1 0.0865 RCIM_Cubic>]}
      }})
   ) {isSetParameters=true, module=SetVariables_#4}
   /SetVariables(
    Particles.R1_Color_Scaled:LinearColor=&MultiplyLinearColor(
     Linear Color=$Particles.R1 Color
     Scale Factor=&MakeVector4FromVectorAndFloat(
      "Vector (RGB)"=&VectorFromFloat(Value=&FloatFromCurve(
        FloatCurve=DataInterface() {propertyValues {Curve {keys [
            {arriveTangent=1.0866, interpMode=RCIM_Cubic, leaveTangent=6.1307, tangentMode=RCTM_Break, time=0.28, value=0}
            {arriveTangent=7.4465, interpMode=RCIM_Cubic, tangentMode=RCTM_Break, time=0.4231, value=1}
            <at 0.53 1>
            <at 0.62 0>
           ]}}}
        Scale Curve=&MakeCustomFloatFromBool(Bool=$User.bIsPainted, True Float=$User.Power_EmissivePaintedR4, False Float=2)
       ))
      "Float (Alpha)"=&FloatFromCurve(
       FloatCurve=DataInterface() {propertyValues {Curve {keys [
           {arriveTangent=1.0866, interpMode=RCIM_Cubic, leaveTangent=6.1307, tangentMode=RCTM_Break, time=0.28, value=0}
           {arriveTangent=7.4465, interpMode=RCIM_Cubic, tangentMode=RCTM_Break, time=0.4231, value=1}
           <at 0.53 1>
           <at 0.62 0>
          ]}}}
       Scale Curve=&MakeCustomFloatFromBool(Bool=$User.bIsPainted, True Float=$User.Power_AlphaPaintedR4, False Float=0.5)
      )
     )
    )
    Particles.R3_Color_Scaled:LinearColor=&MultiplyLinearColor(
     Linear Color=$Particles.R3 Color
     Scale Factor=&MakeVector4FromVectorAndFloat(
      "Vector (RGB)"=&VectorFromFloat(Value=&FloatFromCurve(
        FloatCurve=DataInterface() {propertyValues {Curve {keys [
            {arriveTangent=1.0866, interpMode=RCIM_Cubic, leaveTangent=6.1307, tangentMode=RCTM_Break, time=0.28, value=0}
            {arriveTangent=7.4465, interpMode=RCIM_Cubic, tangentMode=RCTM_Break, time=0.4231, value=1}
            <at 0.53 1>
            <at 0.62 0>
           ]}}}
        Scale Curve=&MakeCustomFloatFromBool(Bool=$User.bIsPainted, True Float=$User.Power_EmissivePaintedR4, False Float=2)
       ))
      "Float (Alpha)"=&FloatFromCurve(
       FloatCurve=DataInterface() {propertyValues {Curve {keys [
           {arriveTangent=1.0866, interpMode=RCIM_Cubic, leaveTangent=6.1307, tangentMode=RCTM_Break, time=0.28, value=0}
           {arriveTangent=7.4465, interpMode=RCIM_Cubic, tangentMode=RCTM_Break, time=0.4231, value=1}
           <at 0.53 1>
           <at 0.62 0>
          ]}}}
       Scale Curve=&MakeCustomFloatFromBool(Bool=$User.bIsPainted, True Float=$User.Power_AlphaPaintedR4, False Float=1)
      )
     )
    )
    Particles.R4_Color_Scaled:LinearColor=&MultiplyLinearColor(
     Linear Color=$Particles.R4 Color
     Scale Factor=&MakeVector4FromVectorAndFloat(
      "Vector (RGB)"=&VectorFromFloat(Value=&FloatFromCurve(
        FloatCurve=DataInterface() {propertyValues {Curve {keys [
            {arriveTangent=1.0866, interpMode=RCIM_Cubic, leaveTangent=6.1307, tangentMode=RCTM_Break, time=0.28, value=0}
            {arriveTangent=7.4465, interpMode=RCIM_Cubic, tangentMode=RCTM_Break, time=0.4231, value=1}
            <at 0.53 1>
            <at 0.62 0>
           ]}}}
        Scale Curve=&MakeCustomFloatFromBool(Bool=$User.bIsPainted, True Float=$User.Power_EmissivePaintedR4, False Float=0.6)
       ))
      "Float (Alpha)"=&FloatFromCurve(
       FloatCurve=DataInterface() {propertyValues {Curve {keys [
           <at -0.0014 1.7739 RCIM_Cubic>
           {arriveTangent=-1.9854, interpMode=RCIM_Cubic, leaveTangent=-1.9854, time=0.3044, value=2.258}
           {arriveTangent=-0.2383, interpMode=RCIM_Cubic, leaveTangent=-1.8935, tangentMode=RCTM_Break, time=0.7964, value=0.1899}
           <at 1.4148 -1.634 RCIM_Cubic>
          ]}}}
       Scale Curve=&MakeCustomFloatFromBool(Bool=$User.bIsPainted, True Float=$User.Power_AlphaPaintedR4, False Float=1.2)
      )
     )
    )
    Particles.R5_Color_Scaled:LinearColor=&MultiplyLinearColor(
     Linear Color=$Particles.R5 Color
     Scale Factor=&MakeVector4FromVectorAndFloat(
      "Vector (RGB)"=&VectorFromFloat(Value=&FloatFromCurve(
        FloatCurve=DataInterface() {propertyValues {Curve {keys [
            <at -0.17 16.4711 RCIM_Cubic>
            {arriveTangent=-27.6507, interpMode=RCIM_Cubic, leaveTangent=-27.6507, time=0.1072, value=17.8443}
            {arriveTangent=-3.2592, interpMode=RCIM_Cubic, leaveTangent=-3.2592, tangentMode=RCTM_User, time=0.3385, value=2.4113}
            {arriveTangent=2.5648, interpMode=RCIM_Cubic, leaveTangent=2.5648, tangentMode=RCTM_User, time=0.64, value=1.973}
            {arriveTangent=11.7336, interpMode=RCIM_Cubic, leaveTangent=11.7336, time=1.0714, value=20.2083}
            {arriveTangent=-34.897, interpMode=RCIM_Cubic, leaveTangent=-34.897, time=1.3482, value=10.2825}
            {arriveTangent=-8.1043, interpMode=RCIM_Cubic, leaveTangent=-8.1043, time=1.5789, value=2.4972}
            <at 1.7789 6.7918 RCIM_Cubic>
           ]}}}
        Scale Curve=&MakeCustomFloatFromBool(Bool=$User.bIsPainted, True Float=$User.Power_EmissivePaintedR4, False Float=0.7)
       ))
      "Float (Alpha)"=&FloatFromCurve(
       FloatCurve=DataInterface() {propertyValues {Curve {keys [
           <at -0.0556 3.887 RCIM_Cubic>
           {arriveTangent=-1.3925, interpMode=RCIM_Cubic, leaveTangent=-1.3925, time=0.111, value=0.7392}
           {arriveTangent=-0.2383, interpMode=RCIM_Cubic, leaveTangent=-0.2383, tangentMode=RCTM_User, time=0.4793, value=3.1423}
           {arriveTangent=-6.3349, interpMode=RCIM_Cubic, leaveTangent=-6.3349, time=0.8329, value=0.2261}
           <at 1.01 -0.2199 RCIM_Cubic>
          ]}}}
       Scale Curve=&MakeCustomFloatFromBool(Bool=$User.bIsPainted, True Float=$User.Power_AlphaPaintedR4, False Float=0.3)
      )
     )
    )
   ) {enabled=false, isSetParameters=true}
   /SetVariables(
    Particles.R1_Color_Scaled:LinearColor=&MultiplyLinearColor(
     Linear Color=$Particles.R1 Color
     Scale Factor=&MakeVector4FromVectorAndFloat(
      "Vector (RGB)"=&VectorFromFloat(Value=&FloatFromCurve(
        FloatCurve=DataInterface() {propertyValues {Curve {keys [
            {arriveTangent=1.0866, interpMode=RCIM_Cubic, leaveTangent=6.1307, tangentMode=RCTM_Break, time=0.1791, value=0}
            {arriveTangent=7.4465, interpMode=RCIM_Cubic, tangentMode=RCTM_Break, time=0.264, value=0.9529}
            <at 0.5072 1>
            <at 0.62 0>
           ]}}}
        Scale Curve=&MakeCustomFloatFromBool(Bool=$User.bIsPainted, True Float=$User.Power_EmissivePaintedR4, False Float=2)
       ))
      "Float (Alpha)"=&FloatFromCurve(
       FloatCurve=DataInterface() {propertyValues {Curve {keys [
           {arriveTangent=1.0866, interpMode=RCIM_Cubic, leaveTangent=6.1307, tangentMode=RCTM_Break, time=0.1557, value=0}
           {arriveTangent=7.4465, interpMode=RCIM_Cubic, tangentMode=RCTM_Break, time=0.2455, value=0.9765}
           <at 0.5287 0.9647>
           <at 0.62 0>
          ]}}}
       Scale Curve=&MakeCustomFloatFromBool(Bool=$User.bIsPainted, True Float=$User.Power_AlphaPaintedR4, False Float=0.5)
      )
     )
    )
    Particles.R3_Color_Scaled:LinearColor=&MultiplyLinearColor(
     Linear Color=$Particles.R3 Color
     Scale Factor=&MakeVector4FromVectorAndFloat(
      "Vector (RGB)"=&VectorFromFloat(Value=&FloatFromCurve(
        FloatCurve=DataInterface() {propertyValues {Curve {keys [
            {arriveTangent=1.0866, interpMode=RCIM_Cubic, leaveTangent=6.1307, tangentMode=RCTM_Break, time=0.0826, value=0.0196}
            {arriveTangent=7.4465, interpMode=RCIM_Cubic, tangentMode=RCTM_Break, time=0.1611, value=0.9533}
            <at 0.5179 1>
            <at 0.62 0>
           ]}}}
        Scale Curve=&MakeCustomFloatFromBool(Bool=$User.bIsPainted, True Float=$User.Power_EmissivePaintedR4, False Float=2)
       ))
      "Float (Alpha)"=&FloatFromCurve(
       FloatCurve=DataInterface() {propertyValues {Curve {keys [
           {arriveTangent=1.0866, interpMode=RCIM_Cubic, leaveTangent=6.1307, tangentMode=RCTM_Break, time=0.1266, value=-0.0192}
           {arriveTangent=7.4465, interpMode=RCIM_Cubic, tangentMode=RCTM_Break, time=0.213, value=0.9638}
           <at 0.5107 0.9882>
           <at 0.62 0>
          ]}}}
       Scale Curve=&MakeCustomFloatFromBool(Bool=$User.bIsPainted, True Float=$User.Power_AlphaPaintedR4, False Float=1)
      )
     )
    )
    Particles.R4_Color_Scaled:LinearColor=&MultiplyLinearColor(
     Linear Color=$Particles.R4 Color
     Scale Factor=&MakeVector4FromVectorAndFloat(
      "Vector (RGB)"=&VectorFromFloat(Value=&FloatFromCurve(
        FloatCurve=DataInterface() {propertyValues {Curve {keys [
            {arriveTangent=1.0866, interpMode=RCIM_Cubic, leaveTangent=6.1307, tangentMode=RCTM_Break, time=0.28, value=0}
            {arriveTangent=7.4465, interpMode=RCIM_Cubic, tangentMode=RCTM_Break, time=0.3458, value=0.9882}
            <at 0.53 1>
            <at 0.62 0>
           ]}}}
        Scale Curve=&MakeCustomFloatFromBool(Bool=$User.bIsPainted, True Float=$User.Power_EmissivePaintedR4, False Float=0.6)
       ))
      "Float (Alpha)"=&FloatFromCurve(
       FloatCurve=DataInterface() {propertyValues {Curve {keys [
           <at -0.0014 1.7739 RCIM_Cubic>
           {arriveTangent=-1.9854, interpMode=RCIM_Cubic, leaveTangent=-1.9854, time=0.3044, value=2.258}
           {arriveTangent=-0.2383, interpMode=RCIM_Cubic, leaveTangent=-1.8935, tangentMode=RCTM_Break, time=0.7964, value=0.1899}
           <at 1.4148 -1.634 RCIM_Cubic>
          ]}}}
       Scale Curve=&MakeCustomFloatFromBool(Bool=$User.bIsPainted, True Float=$User.Power_AlphaPaintedR4, False Float=1.2)
      )
     )
    )
    Particles.R5_Color_Scaled:LinearColor=&MultiplyLinearColor(
     Linear Color=$Particles.R5 Color
     Scale Factor=&MakeVector4FromVectorAndFloat(
      "Vector (RGB)"=&VectorFromFloat(Value=&FloatFromCurve(
        FloatCurve=DataInterface() {propertyValues {Curve {keys [
            <at -0.17 16.4711 RCIM_Cubic>
            {arriveTangent=-27.6507, interpMode=RCIM_Cubic, leaveTangent=-27.6507, time=0.1072, value=17.8443}
            {arriveTangent=-3.2592, interpMode=RCIM_Cubic, leaveTangent=-3.2592, tangentMode=RCTM_User, time=0.3385, value=2.4113}
            {arriveTangent=2.5648, interpMode=RCIM_Cubic, leaveTangent=2.5648, tangentMode=RCTM_User, time=0.64, value=1.973}
            {arriveTangent=11.7336, interpMode=RCIM_Cubic, leaveTangent=11.7336, time=1.0714, value=20.2083}
            {arriveTangent=-34.897, interpMode=RCIM_Cubic, leaveTangent=-34.897, time=1.3482, value=10.2825}
            {arriveTangent=-8.1043, interpMode=RCIM_Cubic, leaveTangent=-8.1043, time=1.5789, value=2.4972}
            <at 1.7789 6.7918 RCIM_Cubic>
           ]}}}
        Scale Curve=&MakeCustomFloatFromBool(Bool=$User.bIsPainted, True Float=$User.Power_EmissivePaintedR4, False Float=0.7)
       ))
      "Float (Alpha)"=&FloatFromCurve(
       FloatCurve=DataInterface() {propertyValues {Curve {keys [
           <at -0.0556 3.887 RCIM_Cubic>
           {arriveTangent=-1.3925, interpMode=RCIM_Cubic, leaveTangent=-1.3925, time=0.111, value=0.7392}
           {arriveTangent=-0.2383, interpMode=RCIM_Cubic, leaveTangent=-0.2383, tangentMode=RCTM_User, time=0.4793, value=3.1423}
           {arriveTangent=-6.3349, interpMode=RCIM_Cubic, leaveTangent=-6.3349, time=0.8329, value=0.2261}
           <at 1.01 -0.2199 RCIM_Cubic>
          ]}}}
       Scale Curve=&MakeCustomFloatFromBool(Bool=$User.bIsPainted, True Float=$User.Power_AlphaPaintedR4, False Float=0.3)
      )
     )
    )
   ) {isSetParameters=true}
   &ScaleRibbonWidth(Ribbon Width Scale=&FloatFromCurve(FloatCurve=DataInterface() {propertyValues {Curve {keys [
         {arriveTangent=-0.2605, interpMode=RCIM_Cubic, leaveTangent=-0.2605, tangentMode=RCTM_User, time=0.0043, value=1.4579}
         {arriveTangent=-1.5709, interpMode=RCIM_Cubic, leaveTangent=-1.5709, tangentMode=RCTM_User, time=0.3137, value=1.1734}
         {arriveTangent=-1.467, interpMode=RCIM_Cubic, leaveTangent=-1.467, time=0.5043, value=0.5604}
         {arriveTangent=-1.126, interpMode=RCIM_Cubic, leaveTangent=-1.126, time=0.6051, value=0.746}
         {arriveTangent=-1.0469, tangentMode=RCTM_User, time=1.002, value=0}
        ]}}}))
   /SetVariables(
    Particles.Ribbon_BlackColor:LinearColor=<rgba 0 0 0 1>
    Particles.WaveScaleCurve:NiagaraFloat=&FloatFromCurve(
     FloatCurve=DataInterface() {propertyValues {Curve {keys [
         {arriveTangent=39.5227, interpMode=RCIM_Cubic, leaveTangent=39.5227, tangentMode=RCTM_User, time=0.0092, value=0}
         {
          arriveTangent=0.7219
          arriveTangentWeight=0.0795
          interpMode=RCIM_Cubic
          leaveTangent=0.1806
          leaveTangentWeight=0.0696
          tangentMode=RCTM_Break
          tangentWeightMode=RCTWM_WeightedBoth
          time=0.125
          value=1.25
         }
         {arriveTangent=0.1988, interpMode=RCIM_Cubic, leaveTangent=0.1988, tangentMode=RCTM_User, time=0.4488, value=0.4338}
        ]}}}
     Scale Curve=10
    )
   ) {isSetParameters=true, module=SetVariables_#5}
   /SetVariables(
    Particles.CustomPosition_R1:NiagaraPosition=&AddVectorToPosition(
     Position=$Particles.Position
     Vector=&Multiply_VectorByFloat(
      Vector=&MakeVector(
       X=&Sine(
        Normalized Angle=&Add_Float(A=0, B=&Subtract_Float(A=$Emitter.NormalizedLoopAge, B=$Particles.NormalizedAge))
        Period=$User.WavePeriod
        Scale=&Multiply_Float(
         A=&SelectFloatFromArray(Direct Array Index=$Particles.SpawnGroup, Float Selection Array=DataInterface() {propertyValues {FloatData [1,-1]}})
         B=$User.WaveMaxScale
        )
        Bias=0
       )
       Y=&Sine(
        Normalized Angle=&Add_Float(A=0, B=&Subtract_Float(A=$Emitter.NormalizedLoopAge, B=$Particles.NormalizedAge))
        Period=$User.WavePeriod
        Scale=&Multiply_Float(
         A=&SelectFloatFromArray(Direct Array Index=$Particles.SpawnGroup, Float Selection Array=DataInterface() {propertyValues {FloatData [1,-1]}})
         B=$User.WaveMaxScale
        )
        Bias=0
       )
       Z=&Cosine(
        Normalized Angle=&Add_Float(A=0, B=&Subtract_Float(A=$Emitter.NormalizedLoopAge, B=$Particles.NormalizedAge))
        Period=$User.WavePeriod
        Scale=$User.WaveMaxScale
        Bias=0
       )
      )
      Float=$Particles.WaveScaleCurve
     )
    )
    Particles.CustomPosition_R2:NiagaraPosition=&AddVectorToPosition(
     Position=$Particles.Position
     Vector=&Multiply_VectorByFloat(
      Vector=&MakeVector(
       X=&Sine(
        Normalized Angle=&Add_Float(
         A=&Multiply_Float(A=$User.WavePeriod, B=0.5)
         B=&Subtract_Float(A=$Emitter.NormalizedLoopAge, B=$Particles.NormalizedAge)
        )
        Period=$User.WavePeriod
        Scale=&Multiply_Float(
         A=&SelectFloatFromArray(Direct Array Index=$Particles.SpawnGroup, Float Selection Array=DataInterface() {propertyValues {FloatData [1,-1]}})
         B=$User.WaveMaxScale
        )
        Bias=0
       )
       Y=&Sine(
        Normalized Angle=&Add_Float(
         A=&Multiply_Float(A=$User.WavePeriod, B=0.5)
         B=&Subtract_Float(A=$Emitter.NormalizedLoopAge, B=$Particles.NormalizedAge)
        )
        Period=$User.WavePeriod
        Scale=&Multiply_Float(
         A=&SelectFloatFromArray(Direct Array Index=$Particles.SpawnGroup, Float Selection Array=DataInterface() {propertyValues {FloatData [1,-1]}})
         B=$User.WaveMaxScale
        )
        Bias=0
       )
       Z=&Cosine(
        Normalized Angle=&Add_Float(
         A=&Multiply_Float(A=$User.WavePeriod, B=0.5)
         B=&Subtract_Float(A=$Emitter.NormalizedLoopAge, B=$Particles.NormalizedAge)
        )
        Period=$User.WavePeriod
        Scale=$User.WaveMaxScale
        Bias=0
       )
      )
      Float=$Particles.WaveScaleCurve
     )
    )
   ) {enabled=false, isSetParameters=true, module=SetVariables_#6}
   /SetVariables(
    Particles.CustomPosition_R1:NiagaraPosition=&AddVectorToPosition(
     Position=$Particles.Position
     Vector=&Multiply_VectorByFloat(
      Vector=&MakeVector(
       X=&Sine(
        Normalized Angle=&Add_Float(A=0, B=&Subtract_Float(A=$Emitter.NormalizedLoopAge, B=$Particles.NormalizedAge))
        Period=$User.WavePeriod
        Scale=&Multiply_Float(
         A=&SelectFloatFromArray(Direct Array Index=$Particles.SpawnGroup, Float Selection Array=DataInterface() {propertyValues {FloatData [1,-1]}})
         B=$User.WaveMaxScale
        )
        Bias=0
       )
       Y=&Sine(
        Normalized Angle=&Add_Float(A=0, B=&Subtract_Float(A=$Emitter.NormalizedLoopAge, B=$Particles.NormalizedAge))
        Period=$User.WavePeriod
        Scale=&Multiply_Float(
         A=&SelectFloatFromArray(Direct Array Index=$Particles.SpawnGroup, Float Selection Array=DataInterface() {propertyValues {FloatData [1,-1]}})
         B=$User.WaveMaxScale
        )
        Bias=0
       )
       Z=&Cosine(
        Normalized Angle=&Add_Float(A=0, B=&Subtract_Float(A=$Emitter.NormalizedLoopAge, B=$Particles.NormalizedAge))
        Period=$User.WavePeriod
        Scale=&Multiply_Float(
         A=&SelectFloatFromArray(Direct Array Index=$Particles.SpawnGroup, Float Selection Array=DataInterface() {propertyValues {FloatData [1,-1]}})
         B=$User.WaveMaxScale
        )
        Bias=0
       )
      )
      Float=$Particles.WaveScaleCurve
     )
    )
    Particles.CustomPosition_R2:NiagaraPosition=&AddVectorToPosition(
     Position=$Particles.Position
     Vector=&Multiply_VectorByFloat(
      Vector=&MakeVector(
       X=&Sine(
        Normalized Angle=&Add_Float(
         A=&Multiply_Float(A=$User.WavePeriod, B=0.5)
         B=&Subtract_Float(A=$Emitter.NormalizedLoopAge, B=$Particles.NormalizedAge)
        )
        Period=$User.WavePeriod
        Scale=&Multiply_Float(
         A=&SelectFloatFromArray(Direct Array Index=$Particles.SpawnGroup, Float Selection Array=DataInterface() {propertyValues {FloatData [1,1]}})
         B=$User.WaveMaxScale
        )
        Bias=0
       )
       Y=&Sine(
        Normalized Angle=&Add_Float(
         A=&Multiply_Float(A=$User.WavePeriod, B=0.5)
         B=&Subtract_Float(A=$Emitter.NormalizedLoopAge, B=$Particles.NormalizedAge)
        )
        Period=$User.WavePeriod
        Scale=&Multiply_Float(
         A=&SelectFloatFromArray(Direct Array Index=$Particles.SpawnGroup, Float Selection Array=DataInterface() {propertyValues {FloatData [1,1]}})
         B=$User.WaveMaxScale
        )
        Bias=0
       )
       Z=&Cosine(
        Normalized Angle=&Add_Float(
         A=&Multiply_Float(A=$User.WavePeriod, B=0.5)
         B=&Subtract_Float(A=$Emitter.NormalizedLoopAge, B=$Particles.NormalizedAge)
        )
        Period=$User.WavePeriod
        Scale=&Multiply_Float(
         A=&SelectFloatFromArray(Direct Array Index=$Particles.SpawnGroup, Float Selection Array=DataInterface() {propertyValues {FloatData [-1,1]}})
         B=$User.WaveMaxScale
        )
        Bias=0
       )
      )
      Float=$Particles.WaveScaleCurve
     )
    )
   ) {isSetParameters=true}
   /SetVariables(Particles.DynamicMP_R2:Vector4f=&MakeVector4(
     X=&FloatFromCurve(FloatCurve=DataInterface() {propertyValues {Curve {
         defaultValue=1
         keys [
          {tangentMode=RCTM_User, time=0, value=0.5}
          {arriveTangent=-0.0633, interpMode=RCIM_Cubic, leaveTangent=-0.0633, tangentMode=RCTM_User, time=0.65, value=0.5}
          {arriveTangent=-0.0699, interpMode=RCIM_Cubic, leaveTangent=-0.0699, tangentMode=RCTM_User, time=0.75, value=-2}
         ]
        }}})
     Y=0
     Z=0
     W=0
    )) {isSetParameters=true, module=SetVariables_#7}
   &DynamicMaterialParameters(
    Index 0 Param 1=&FloatFromCurve(FloatCurve=DataInterface() {propertyValues {Curve {
        defaultValue=1
        keys [
         {
          arriveTangent=0.0405
          arriveTangentWeight=0.1835
          interpMode=RCIM_Cubic
          leaveTangent=0.0405
          leaveTangentWeight=0.1835
          tangentMode=RCTM_User
          time=0
          value=0
         }
         {
          arriveTangent=0.0201
          arriveTangentWeight=0.1834
          interpMode=RCIM_Cubic
          leaveTangent=0.0201
          leaveTangentWeight=0.1834
          tangentMode=RCTM_User
          time=0.5
          value=1
         }
        ]
       }}})
    Module.Index0Float0Write=true
    Module.Index0Float1Write=true
    Module.Index0Float2Write=true
    Module.Index0Float3Write=true
   )
  ]
  rendererClasses [&NiagaraRibbonRendererProperties]
  renderers [
   {rendererClass=&NiagaraRibbonRendererProperties}
   {rendererClass=&NiagaraRibbonRendererProperties, rendererIndex=1}
   {rendererClass=&NiagaraRibbonRendererProperties, rendererIndex=2}
   {rendererClass=&NiagaraRibbonRendererProperties, rendererIndex=3}
  ]
 }
 NE_BlackFlame_Mid2 {
  bEnabled=false
  emitterUpdateScript [
   &EmitterState(Life Cycle Mode=::NewEnumerator1, Loop Duration=1, UseLoopDelay=false)
   &VC_FilteredSocketCount()
   &VC_RateQualityHandler(Epic Spawn Rate=30, High Spawn Rate=30, Low Spawn Rate=30, Use Custom Rate?=true)
   &VC_SpawnRibbon_v1(Number of Ribbons=$Emitter.PortCount, Spawn Time=-1, SpawnRate=$Emitter.Spawn Rate)
  ]
  particleSpawnScript [
   &VC_PaintedVariantsOverride() {enabled=false}
   &VC_InitializeRibbon_v1(
    Color=$Particles.Color
    Custom Offset=true
    Enable Skeletal Position=true
    Lifetime=&MergeHelper_FloatSwitch(0=1, 1=1, 2=1, 3=?, Index=$User.GameModeID, Num=::NewEnumerator1)
    Local Vector Offset=&AddVectorToPosition(Position=<30 0 2.5>, Vector=<0 0 0>)
    Offset Amount=1
    Ribbon Width=-55
   )
   &AddVelocity(Velocity=<-2000 0 0>) {module=AddVelocity001}
   /SetVariables(
    Particles.R1_Width:NiagaraFloat=&SelectFloatFromArray(Direct Array Index=&SpawnGroup(), Float Selection Array=DataInterface() {propertyValues {FloatData [75,0]}})
    Particles.R2_Width:NiagaraFloat=&SelectFloatFromArray(Direct Array Index=&SpawnGroup(), Float Selection Array=DataInterface() {propertyValues {FloatData [0,-75]}})
    Particles.R3_Color:LinearColor=<rgba 0.2241 0.5211 0.6609 3>
   ) {enabled=false, isSetParameters=true, module=SetVariables_#10}
  ]
  particleUpdateScript [
   &VC_PaintedVariantsOverride(
    R1 Black:LinearColor=&ColorFromCurve(ColorCurve=DataInterface() {propertyValues {
       AlphaCurve {keys [
         {tangentMode=RCTM_User, time=0.1264, value=0}
         {tangentMode=RCTM_User, time=0.2268, value=0.5}
         <at 0.5035 0.5>
         {tangentMode=RCTM_User, time=0.557, value=0}
        ]}
       BlueCurve {keys [<at 1 0.1094 RCIM_Cubic>]}
       GreenCurve {keys [<at 1 0.1094 RCIM_Cubic>]}
       RedCurve {keys [<at 1 0.1094 RCIM_Cubic>]}
      }})
    R1 BurntSienna:LinearColor=&ColorFromCurve(ColorCurve=DataInterface() {propertyValues {
       AlphaCurve {
        defaultValue=7
        keys [
         {tangentMode=RCTM_User, time=0.1264, value=0}
         {tangentMode=RCTM_User, time=0.2268, value=0.5}
         <at 0.5035 0.5>
         {tangentMode=RCTM_User, time=0.557, value=0}
        ]
       }
       BlueCurve {keys [<at 1 0.0686 RCIM_Cubic>]}
       GreenCurve {keys [<at 1 0.0858 RCIM_Cubic>]}
       RedCurve {keys [<at 1 0.1667 RCIM_Cubic>]}
      }})
    R1 Cobalt:LinearColor=&ColorFromCurve(ColorCurve=DataInterface() {propertyValues {
       AlphaCurve {
        defaultValue=7
        keys [
         {tangentMode=RCTM_User, time=0.1264, value=0}
         {tangentMode=RCTM_User, time=0.2268, value=0.5}
         <at 0.5035 0.5>
         {tangentMode=RCTM_User, time=0.557, value=0}
        ]
       }
       BlueCurve {keys [<at 1 0.3802 RCIM_Cubic>]}
       GreenCurve {keys [<at 1 0.0931 RCIM_Cubic>]}
       RedCurve {keys [<at 1 0.0931 RCIM_Cubic>]}
      }})
    R1 Crimson:LinearColor=&ColorFromCurve(ColorCurve=DataInterface() {propertyValues {
       AlphaCurve {
        defaultValue=7
        keys [
         {tangentMode=RCTM_User, time=0.1264, value=0}
         {tangentMode=RCTM_User, time=0.2268, value=0.5}
         <at 0.5035 0.5>
         {tangentMode=RCTM_User, time=0.557, value=0}
        ]
       }
       BlueCurve {keys [<at 1 0.0131 RCIM_Cubic>]}
       GreenCurve {keys [<at 1 0 RCIM_Cubic>]}
       RedCurve {keys [<at 1 0.2348 RCIM_Cubic>]}
      }})
    R1 DefaultColor=&ColorFromCurve(ColorCurve=DataInterface() {propertyValues {
       AlphaCurve {keys [<at 0.1264 0>,<at 0.2268 0.5>,<at 0.5035 0.5>,<at 0.557 0>]}
       BlueCurve {keys [<at 1 0.0999 RCIM_Cubic>]}
       GreenCurve {keys [<at 1 0 RCIM_Cubic>]}
       RedCurve {keys [<at 1 0.0513 RCIM_Cubic>]}
      }})
    R1 ForestGreen:LinearColor=&ColorFromCurve(ColorCurve=DataInterface() {propertyValues {
       AlphaCurve {
        defaultValue=7
        keys [
         {tangentMode=RCTM_User, time=0.1264, value=0}
         {tangentMode=RCTM_User, time=0.2268, value=0.5}
         <at 0.5035 0.5>
         {tangentMode=RCTM_User, time=0.557, value=0}
        ]
       }
       BlueCurve {keys [<at 1 0.0198 RCIM_Cubic>]}
       GreenCurve {keys [<at 1 0.2031 RCIM_Cubic>]}
       RedCurve {keys [<at 1 0 RCIM_Cubic>]}
      }})
    R1 Grey:LinearColor=&ColorFromCurve(ColorCurve=DataInterface() {propertyValues {
       AlphaCurve {
        defaultValue=1
        keys [
         {tangentMode=RCTM_User, time=0.1264, value=0}
         {tangentMode=RCTM_User, time=0.2268, value=0.5}
         <at 0.5035 0.5>
         {tangentMode=RCTM_User, time=0.557, value=0}
        ]
       }
       BlueCurve {keys [<at 1 0.2344 RCIM_Cubic>]}
       GreenCurve {keys [<at 1 0.2344 RCIM_Cubic>]}
       RedCurve {keys [<at 1 0.2344 RCIM_Cubic>]}
      }})
    R1 Lime:LinearColor=&ColorFromCurve(ColorCurve=DataInterface() {propertyValues {
       AlphaCurve {
        defaultValue=7
        keys [
         {tangentMode=RCTM_User, time=0.1264, value=0}
         {tangentMode=RCTM_User, time=0.2268, value=0.5}
         <at 0.5035 0.5>
         {tangentMode=RCTM_User, time=0.557, value=0}
        ]
       }
       BlueCurve {keys [<at 1 0.0865 RCIM_Cubic>]}
       GreenCurve {keys [<at 1 0.2865 RCIM_Cubic>]}
       RedCurve {keys [<at 1 0.1697 RCIM_Cubic>]}
      }})
    R1 Orange:LinearColor=&ColorFromCurve(ColorCurve=DataInterface() {propertyValues {
       AlphaCurve {
        defaultValue=7
        keys [
         {tangentMode=RCTM_User, time=0.1264, value=0}
         {tangentMode=RCTM_User, time=0.2268, value=0.5}
         <at 0.5035 0.5>
         {tangentMode=RCTM_User, time=0.557, value=0}
        ]
       }
       BlueCurve {keys [<at 1 0.0688 RCIM_Cubic>]}
       GreenCurve {keys [<at 1 0.1085 RCIM_Cubic>]}
       RedCurve {keys [<at 1 0.3478 RCIM_Cubic>]}
      }})
    R1 Pink:LinearColor=&ColorFromCurve(ColorCurve=DataInterface() {propertyValues {
       AlphaCurve {
        defaultValue=7
        keys [
         {tangentMode=RCTM_User, time=0.1264, value=0}
         {tangentMode=RCTM_User, time=0.2268, value=0.5}
         <at 0.5035 0.5>
         {tangentMode=RCTM_User, time=0.557, value=0}
        ]
       }
       BlueCurve {keys [<at 1 0.2004 RCIM_Cubic>]}
       GreenCurve {keys [<at 1 0 RCIM_Cubic>]}
       RedCurve {keys [<at 1 0.2188 RCIM_Cubic>]}
      }})
    R1 Purple:LinearColor=&ColorFromCurve(ColorCurve=DataInterface() {propertyValues {
       AlphaCurve {
        defaultValue=7
        keys [
         {tangentMode=RCTM_User, time=0.1264, value=0}
         {tangentMode=RCTM_User, time=0.2268, value=0.5}
         <at 0.5035 0.5>
         {tangentMode=RCTM_User, time=0.557, value=0}
        ]
       }
       BlueCurve {keys [<at 1 0.375 RCIM_Cubic>]}
       GreenCurve {keys [<at 1 0 RCIM_Cubic>]}
       RedCurve {keys [<at 1 0.1484 RCIM_Cubic>]}
      }})
    R1 Saffron:LinearColor=&ColorFromCurve(ColorCurve=DataInterface() {propertyValues {
       AlphaCurve {
        defaultValue=7
        keys [
         {tangentMode=RCTM_User, time=0.1264, value=0}
         {tangentMode=RCTM_User, time=0.2268, value=0.5}
         <at 0.5035 0.5>
         {tangentMode=RCTM_User, time=0.557, value=0}
        ]
       }
       BlueCurve {keys [<at 1 0 RCIM_Cubic>]}
       GreenCurve {keys [<at 1 0.2395 RCIM_Cubic>]}
       RedCurve {keys [<at 1 0.2812 RCIM_Cubic>]}
      }})
    R1 SkyBlue:LinearColor=&ColorFromCurve(ColorCurve=DataInterface() {propertyValues {
       AlphaCurve {
        defaultValue=7
        keys [
         {tangentMode=RCTM_User, time=0.1264, value=0}
         {tangentMode=RCTM_User, time=0.2268, value=0.5}
         <at 0.5035 0.5>
         {tangentMode=RCTM_User, time=0.557, value=0}
        ]
       }
       BlueCurve {keys [<at 1 0.3802 RCIM_Cubic>]}
       GreenCurve {keys [<at 1 0.1905 RCIM_Cubic>]}
       RedCurve {keys [<at 1 0 RCIM_Cubic>]}
      }})
    R1 TitaniumWhite:LinearColor=&ColorFromCurve(ColorCurve=DataInterface() {propertyValues {
       AlphaCurve {
        defaultValue=2
        keys [
         {tangentMode=RCTM_User, time=0.1264, value=0}
         {tangentMode=RCTM_User, time=0.2268, value=0.5}
         <at 0.5035 0.5>
         {tangentMode=RCTM_User, time=0.557, value=0}
        ]
       }
       BlueCurve {keys [<at 1 0.6406 RCIM_Cubic>]}
       GreenCurve {keys [<at 1 0.6406 RCIM_Cubic>]}
       RedCurve {keys [<at 1 0.6406 RCIM_Cubic>]}
      }})
    R2 Black:LinearColor=&ColorFromCurve(ColorCurve=DataInterface() {propertyValues {
       AlphaCurve {keys [<at 0 7>,<at 0.2232 1>,<at 0.3261 0.05>,<at 0.5495 0.2>,<at 0.7007 0.2>,<at 0.8474 0 RCIM_Cubic>]}
       BlueCurve {keys [<at 0 0.0194 RCIM_Cubic>,<at 0.1492 0.0333>,<at 0.2681 0.1104>,<at 0.4863 0.9274>,<at 0.6088 0.0305>]}
       GreenCurve {keys [<at 0 0.0194 RCIM_Cubic>,<at 0.1492 0.0333>,<at 0.2681 0.1104>,<at 0.4863 0.9274>,<at 0.6088 0.0305>]}
       RedCurve {keys [<at 0 0.0194 RCIM_Cubic>,<at 0.1492 0.0333>,<at 0.2681 0.1104>,<at 0.4863 0.9274>,<at 0.6088 0.0305>]}
      }})
    R2 BurntSienna:LinearColor=&ColorFromCurve(ColorCurve=DataInterface() {propertyValues {
       AlphaCurve {keys [<at 0 4>,<at 0.1907 1>,<at 0.3261 0.05>,<at 0.593 0.2>,<at 0.8474 0 RCIM_Cubic>]}
       BlueCurve {keys [<at 0 0.008 RCIM_Cubic>,<at 0.2839 0.0202>,<at 0.4863 0.2907>,<at 0.5956 0.008>]}
       GreenCurve {keys [<at 0 0.0203 RCIM_Cubic>,<at 0.2839 0.0323>,<at 0.4863 0.4394>,<at 0.5956 0.0203>]}
       RedCurve {keys [<at 0 0.0648 RCIM_Cubic>,<at 0.2839 0.0762>,<at 0.4863 0.9792>,<at 0.5956 0.0648>]}
      }})
    R2 Cobalt:LinearColor=&ColorFromCurve(ColorCurve=DataInterface() {propertyValues {
       AlphaCurve {keys [<at 0 7>,<at 0.1907 1>,<at 0.3261 0.05>,<at 0.5495 0.6>,<at 0.7007 0.2>,<at 0.8474 0 RCIM_Cubic>]}
       BlueCurve {keys [<at 0 0.6458 RCIM_Cubic>,<at 0.2575 0.9985>,<at 0.4863 1>,<at 0.6115 0.8177>]}
       GreenCurve {keys [<at 0 0.0119 RCIM_Cubic>,<at 0.2575 0.0367>,<at 0.4863 0.2997>,<at 0.6115 0.0527>]}
       RedCurve {keys [<at 0 0 RCIM_Cubic>,<at 0.2575 0.0186>,<at 0.4863 0.2865>,<at 0.6115 0.0383>]}
      }})
    R2 Crimson:LinearColor=&ColorFromCurve(ColorCurve=DataInterface() {propertyValues {
       AlphaCurve {keys [<at 0 7>,<at 0.1907 1>,<at 0.3261 0.05>,<at 0.5495 0.2>,<at 0.7007 0.2>,<at 0.8474 0 RCIM_Cubic>]}
       BlueCurve {keys [<at 0 0 RCIM_Cubic>,<at 0.1492 0.0104>,<at 0.4863 0.2995>,<at 0.6458 0.0894>]}
       GreenCurve {keys [<at 0 0 RCIM_Cubic>,<at 0.1492 0.0104>,<at 0.4863 0.2995>,<at 0.6458 0.0894>]}
       RedCurve {keys [<at 0 0.4323 RCIM_Cubic>,<at 0.1492 0.9992>,<at 0.4863 0.9274>,<at 0.6458 0.993>]}
      }})
    R2 DefaultColor:LinearColor=&ColorFromCurve(ColorCurve=DataInterface() {propertyValues {
       AlphaCurve {keys [<at 0 1>,<at 0.0914 0.2857>,<at 0.308 0.05>,<at 0.5592 0.2>,<at 0.8632 0>]}
       BlueCurve {keys [<at 0.0009 0>,<at 0.2668 -0.2829>,<at 0.5114 0.8849>,<at 0.619 1>,<at 0.682 0.383>]}
       GreenCurve {keys [<at 0.0009 0.0767>,<at 0.2742 -0.1994>,<at 0.5114 0.8849>,<at 0.619 1>,<at 0.682 0.2585>]}
       RedCurve {keys [<at 0.0009 3.2194>,<at 0.0016 3.2194>,<at 0.4479 0.8849>,<at 0.5114 0.8849>,<at 0.619 1>]}
      }})
    R2 ForestGreen:LinearColor=&ColorFromCurve(ColorCurve=DataInterface() {propertyValues {
       AlphaCurve {keys [<at 0 6>,<at 0.1907 1>,<at 0.3261 0.05>,<at 0.5495 0.2>,<at 0.7007 0.2>,<at 0.8474 0 RCIM_Cubic>]}
       BlueCurve {keys [<at 0 0.0172 RCIM_Cubic>,<at 0.2179 0.0322>,<at 0.4863 0.4656>,<at 0.6537 0.1027>]}
       GreenCurve {keys [<at 0 0.3438 RCIM_Cubic>,<at 0.2179 0.245>,<at 0.4863 0.875>,<at 0.6537 0.2987>]}
       RedCurve {keys [<at 0 0 RCIM_Cubic>,<at 0.2179 0.021>,<at 0.4863 0.4466>,<at 0.6537 0.0924>]}
      }})
    R2 Grey:LinearColor=&ColorFromCurve(ColorCurve=DataInterface() {propertyValues {
       AlphaCurve {keys [<at 0 4>,<at 0.1907 1>,<at 0.3261 0.05>,<at 0.5495 0.2>,<at 0.7007 0.2>,<at 0.8474 0 RCIM_Cubic>]}
       BlueCurve {keys [<at 0 0.0692 RCIM_Cubic>,<at 0.2285 0.1336>,<at 0.4863 0.9274>,<at 0.6405 0.132>]}
       GreenCurve {keys [<at 0 0.0781 RCIM_Cubic>,<at 0.2285 0.1336>,<at 0.4863 0.9274>,<at 0.6405 0.132>]}
       RedCurve {keys [<at 0 0.0767 RCIM_Cubic>,<at 0.2285 0.1406>,<at 0.4863 0.9274>,<at 0.6405 0.1354>]}
      }})
    R2 Lime:LinearColor=&ColorFromCurve(ColorCurve=DataInterface() {propertyValues {
       AlphaCurve {keys [<at 0 6>,<at 0.1907 1>,<at 0.3261 0.05>,<at 0.5495 0.2>,<at 0.8122 0 RCIM_Cubic>]}
       BlueCurve {keys [<at 0 0 RCIM_Cubic>,<at 0.2179 0.0459>,<at 0.4863 0.9274>,<at 0.6299 0.0724>]}
       GreenCurve {keys [<at 0 0.6875 RCIM_Cubic>,<at 0.2179 0.4635>,<at 0.4863 0.9274>,<at 0.6299 0.7708>]}
       RedCurve {keys [<at 0 0.3447 RCIM_Cubic>,<at 0.2179 0.2553>,<at 0.4863 0.9274>,<at 0.6299 0.4226>]}
      }})
    R2 Orange:LinearColor=&ColorFromCurve(ColorCurve=DataInterface() {propertyValues {
       AlphaCurve {keys [<at 0 7>,<at 0.1907 1>,<at 0.3261 0.05>,<at 0.5495 0.2>,<at 0.7007 0.2>,<at 0.8474 0 RCIM_Cubic>]}
       BlueCurve {keys [<at 0 0 RCIM_Cubic>,<at 0.2364 0.0598>,<at 0.4863 0.4263>,<at 0.6217 0.1191>]}
       GreenCurve {keys [<at 0 0.2834 RCIM_Cubic>,<at 0.2364 0.2792>,<at 0.4863 0.5719>,<at 0.6217 0.3121>]}
       RedCurve {keys [<at 0 0.8125 RCIM_Cubic>,<at 0.2364 0.6889>,<at 0.4863 0.8438>,<at 0.6217 0.6724>]}
      }})
    R2 Pink:LinearColor=&ColorFromCurve(ColorCurve=DataInterface() {propertyValues {
       AlphaCurve {keys [<at 0 7>,<at 0.1907 1>,<at 0.3261 0.05>,<at 0.5495 0.2>,<at 0.7007 0.2>,<at 0.8474 0 RCIM_Cubic>]}
       BlueCurve {keys [<at 0 0.3419 RCIM_Cubic>,<at 0.2707 0.3564>,<at 0.4863 0.7216>,<at 0.6141 0.5634>]}
       GreenCurve {keys [<at 0 0.0612 RCIM_Cubic>,<at 0.2707 0.1489>,<at 0.4863 0.5789>,<at 0.6141 0.2536>]}
       RedCurve {keys [<at 0 0.5029 RCIM_Cubic>,<at 0.2707 0.5614>,<at 0.4863 0.855>,<at 0.6141 0.8542>]}
      }})
    R2 Purple:LinearColor=&ColorFromCurve(ColorCurve=DataInterface() {propertyValues {
       AlphaCurve {keys [<at 0 7>,<at 0.1907 1>,<at 0.3261 0.05>,<at 0.5495 0.2>,<at 0.7007 0.2>,<at 0.8474 0 RCIM_Cubic>]}
       BlueCurve {keys [<at 0 0.8281 RCIM_Cubic>,<at 0.2919 0.6836>,<at 0.4863 0.8957>,<at 0.6141 0.9565>]}
       GreenCurve {keys [<at 0 0 RCIM_Cubic>,<at 0.2919 0.0563>,<at 0.4863 0.2791>,<at 0.6141 0.1097>]}
       RedCurve {keys [<at 0 0.1924 RCIM_Cubic>,<at 0.2919 0.3349>,<at 0.4863 0.5023>,<at 0.6141 0.3436>]}
      }})
    R2 Saffron:LinearColor=&ColorFromCurve(ColorCurve=DataInterface() {propertyValues {
       AlphaCurve {keys [<at 0 4>,<at 0.1907 1>,<at 0.3261 0.05>,<at 0.5495 0.2>,<at 0.8175 0 RCIM_Cubic>]}
       BlueCurve {keys [<at 0 0 RCIM_Cubic>,<at 0.2697 0.0046>,<at 0.4863 0.3188>,<at 0.6748 0.0798>]}
       GreenCurve {keys [<at 0 0.6105 RCIM_Cubic>,<at 0.2697 0.6881>,<at 0.4863 0.9274>,<at 0.6748 0.6052>]}
       RedCurve {keys [<at 0 0.7969 RCIM_Cubic>,<at 0.2697 0.8967>,<at 0.4863 0.8744>,<at 0.6748 0.7656>]}
      }})
    R2 SkyBlue:LinearColor=&ColorFromCurve(ColorCurve=DataInterface() {propertyValues {
       AlphaCurve {keys [<at 0 7>,<at 0.1907 1>,<at 0.3261 0.05>,<at 0.5495 0.2>,<at 0.8096 0 RCIM_Cubic>]}
       BlueCurve {keys [<at 0 0.5333 RCIM_Cubic>,<at 0.2496 0.5525>,<at 0.4863 0.9323>,<at 0.6141 0.9323>]}
       GreenCurve {keys [<at 0 0.3564 RCIM_Cubic>,<at 0.2496 0.3792>,<at 0.4863 0.8083>,<at 0.6141 0.6988>]}
       RedCurve {keys [<at 0 0 RCIM_Cubic>,<at 0.2496 0.0875>,<at 0.4863 0.5584>,<at 0.6141 0.2282>]}
      }})
    R2 TitaniumWhite:LinearColor=&ColorFromCurve(ColorCurve=DataInterface() {propertyValues {
       AlphaCurve {keys [<at 0 4>,<at 0.1907 1>,<at 0.3261 0.05>,<at 0.5495 0.2>,<at 0.7007 0.2>,<at 0.8175 0 RCIM_Cubic>]}
       BlueCurve {keys [<at 0 0.1846 RCIM_Cubic>,<at 0.2285 0.1336>,<at 0.4863 0.9274>,<at 0.6405 0.3248>]}
       GreenCurve {keys [<at 0 0.2083 RCIM_Cubic>,<at 0.2285 0.1336>,<at 0.4863 0.9274>,<at 0.6405 0.3248>]}
       RedCurve {keys [<at 0 0.2046 RCIM_Cubic>,<at 0.2285 0.1406>,<at 0.4863 0.9274>,<at 0.6405 0.3333>]}
      }})
    R3 Black:LinearColor=&ColorFromCurve(ColorCurve=DataInterface() {propertyValues {
       AlphaCurve {defaultValue=0, keys [{tangentMode=RCTM_User, time=0.2127, value=0}, <at 0.3257 1>, <at 0.5324 1>, <at 0.6291 0>]}
       BlueCurve {keys [<at 0.6966 0.3043>]}
       GreenCurve {keys [<at 0.6966 0.3043>]}
       RedCurve {keys [<at 0.6966 0.3043>]}
      }})
    R3 BurntSienna:LinearColor=&ColorFromCurve(ColorCurve=DataInterface() {propertyValues {
       AlphaCurve {defaultValue=0, keys [{tangentMode=RCTM_User, time=0.2127, value=0}, <at 0.3257 1>, <at 0.5324 1>, <at 0.6291 0>]}
       BlueCurve {keys [<at 0.6966 0.2729>]}
       GreenCurve {keys [<at 0.6966 0.3693>]}
       RedCurve {keys [<at 0.6966 0.6719>]}
      }})
    R3 Cobalt:LinearColor=&ColorFromCurve(ColorCurve=DataInterface() {propertyValues {
       AlphaCurve {defaultValue=0, keys [{tangentMode=RCTM_User, time=0.2127, value=0}, <at 0.3257 1>, <at 0.5324 1>, <at 0.6291 0>]}
       BlueCurve {keys [<at 0.6966 1.5>]}
       GreenCurve {keys [<at 0.6966 0.0598>]}
       RedCurve {keys [<at 0.6966 0>]}
      }})
    R3 Crimson:LinearColor=&ColorFromCurve(ColorCurve=DataInterface() {propertyValues {
       AlphaCurve {defaultValue=0, keys [{tangentMode=RCTM_User, time=0.2127, value=0}, <at 0.3257 1>, <at 0.5324 1>, <at 0.6291 0>]}
       BlueCurve {keys [<at 0.6966 0>]}
       GreenCurve {keys [<at 0.6966 0>]}
       RedCurve {keys [<at 0.6966 1>]}
      }})
    R3 DefaultColor:LinearColor=&ColorFromCurve(ColorCurve=DataInterface() {propertyValues {
       AlphaCurve {keys [<at 0.2127 0>,<at 0.3257 1>,<at 0.5324 1>,<at 0.6291 0>]}
       BlueCurve {keys [<at 0.6966 2>]}
       GreenCurve {keys [<at 0.6966 0.5201>]}
       RedCurve {keys [<at 0.6966 0>]}
      }})
    R3 ForestGreen:LinearColor=&ColorFromCurve(ColorCurve=DataInterface() {propertyValues {
       AlphaCurve {defaultValue=0, keys [{tangentMode=RCTM_User, time=0.2127, value=0}, <at 0.3257 1>, <at 0.5324 1>, <at 0.6291 0>]}
       BlueCurve {keys [<at 0.6966 0.3387>]}
       GreenCurve {keys [<at 0.6966 1.5>]}
       RedCurve {keys [<at 0.6966 0>]}
      }})
    R3 Grey:LinearColor=&ColorFromCurve(ColorCurve=DataInterface() {propertyValues {
       AlphaCurve {keys [{tangentMode=RCTM_User, time=0.2127, value=0}, <at 0.3257 1>, <at 0.5324 1>, <at 0.6291 0>, <at 0.7715 0>]}
       BlueCurve {keys [<at 0.6966 0.5>]}
       GreenCurve {keys [<at 0.6966 0.5>]}
       RedCurve {keys [<at 0.6966 0.5>]}
      }})
    R3 Lime:LinearColor=&ColorFromCurve(ColorCurve=DataInterface() {propertyValues {
       AlphaCurve {defaultValue=0, keys [{tangentMode=RCTM_User, time=0.2127, value=0}, <at 0.3257 1>, <at 0.5324 1>, <at 0.6291 0>]}
       BlueCurve {keys [<at 0.6966 0>]}
       GreenCurve {keys [<at 0.6966 1>]}
       RedCurve {keys [<at 0.6966 0.5901>]}
      }})
    R3 Orange:LinearColor=&ColorFromCurve(ColorCurve=DataInterface() {propertyValues {
       AlphaCurve {defaultValue=0, keys [{tangentMode=RCTM_User, time=0.2127, value=0}, <at 0.3257 1>, <at 0.5324 1>, <at 0.6291 0>]}
       BlueCurve {keys [<at 0.6966 0.0304>]}
       GreenCurve {keys [<at 0.6966 0.2737>]}
       RedCurve {keys [<at 0.6966 0.7292>]}
      }})
    R3 Pink:LinearColor=&ColorFromCurve(ColorCurve=DataInterface() {propertyValues {
       AlphaCurve {defaultValue=0, keys [{tangentMode=RCTM_User, time=0.2127, value=0}, <at 0.3257 1>, <at 0.5324 1>, <at 0.6291 0>]}
       BlueCurve {keys [<at 0.6966 1.2188>]}
       GreenCurve {keys [<at 0.6966 0.2158>]}
       RedCurve {keys [<at 0.6966 1.1488>]}
      }})
    R3 Purple:LinearColor=&ColorFromCurve(ColorCurve=DataInterface() {propertyValues {
       AlphaCurve {defaultValue=0, keys [{tangentMode=RCTM_User, time=0.2127, value=0}, <at 0.3257 1>, <at 0.5324 1>, <at 0.6291 0>]}
       BlueCurve {keys [<at 0.6966 1>]}
       GreenCurve {keys [<at 0.6966 0>]}
       RedCurve {keys [<at 0.6966 0.3627>]}
      }})
    R3 Saffron:LinearColor=&ColorFromCurve(ColorCurve=DataInterface() {propertyValues {
       AlphaCurve {defaultValue=0, keys [{tangentMode=RCTM_User, time=0.2127, value=0}, <at 0.3257 1>, <at 0.5324 1>, <at 0.6291 0>]}
       BlueCurve {keys [<at 0.6966 0>]}
       GreenCurve {keys [<at 0.6966 1.8157>]}
       RedCurve {keys [<at 0.6966 2>]}
      }})
    R3 SkyBlue:LinearColor=&ColorFromCurve(ColorCurve=DataInterface() {propertyValues {
       AlphaCurve {defaultValue=0, keys [{tangentMode=RCTM_User, time=0.2127, value=0}, <at 0.3257 1>, <at 0.5324 1>, <at 0.6291 0>]}
       BlueCurve {keys [<at 0.6966 1.5>]}
       GreenCurve {keys [<at 0.6966 0.3855>]}
       RedCurve {keys [<at 0.6966 0>]}
      }})
    R3 TitaniumWhite:LinearColor=&ColorFromCurve(ColorCurve=DataInterface() {propertyValues {
       AlphaCurve {keys [{tangentMode=RCTM_User, time=0.2127, value=0}, <at 0.3257 1>, <at 0.5324 1>, <at 0.6291 0>, <at 0.9629 0>]}
       BlueCurve {keys [<at 0.6966 2.5>]}
       GreenCurve {keys [<at 0.6966 2.5>]}
       RedCurve {keys [<at 0.6966 2.5>]}
      }})
    R4 BurntSienna:LinearColor=<rgba 0.0312 0.0096 0.0039 0.6>
    R4 Cobalt:LinearColor=<rgba 0 0.0115 0.6198 0.7>
    R4 Crimson:LinearColor=<rgba 0.2656 0 0.0072 0.7>
    R4 ForestGreen:LinearColor=<rgba 0 0.2552 0.0129 0.6>
    R4 Grey:LinearColor=<rgba 0.0417 0.0417 0.0417 0.7>
    R4 Lime:LinearColor=<rgba 0.1875 0.375 0 0.6>
    R4 Orange:LinearColor=<rgba 0.5521 0.1706 0 0.45>
    R4 Pink:LinearColor=<rgba 0.3906 0.0468 0.2656 0.5>
    R4 Purple:LinearColor=<rgba 0.1567 0 0.7552 0.5>
    R4 Saffron:LinearColor=<rgba 0.5573 0.4299 0 0.3>
    R4 SkyBlue:LinearColor=<rgba 0.0137 0.2957 0.4375 0.5>
    R4 TitaniumWhite:LinearColor=<rgba 0.224 0.224 0.224 0.6>
    R5 Black:LinearColor=<rgba 0.0469 0.0469 0.0469 0.6>
    R5 BurntSienna:LinearColor=<rgba 0.1042 0.0322 0.0128 0.7>
    R5 Cobalt:LinearColor=<rgba 0.0312 0.0401 1 0.6>
    R5 Crimson:LinearColor=<rgba 0.4635 0.0266 0.0384 0.6>
    R5 DefaultColor:LinearColor=<rgba 1 1 1 0.5>
    R5 ForestGreen:LinearColor=<rgba 0.0704 0.901 0.1124 0.4>
    R5 Grey:LinearColor=<rgba 0.1667 0.1667 0.1667 0.4>
    R5 Lime:LinearColor=<rgba 0.2938 0.5104 0.0771 0.25>
    R5 Orange:LinearColor=<rgba 0.8261 0.2993 0.0172 0.3>
    R5 Pink:LinearColor=<rgba 0.651 0.1763 0.4784 0.4>
    R5 Purple:LinearColor=<rgba 0.2704 0.1155 0.7917 0.3>
    R5 Saffron:LinearColor=<rgba 0.8906 0.6824 0 0.15>
    R5 SkyBlue:LinearColor=<rgba 0.0754 0.3822 0.5365 0.4>
    R5 TitaniumWhite:LinearColor=<rgba 0.5729 0.5729 0.5729 0.3>
    Use R1 Painted Overrides=true
    Use R2 Painted Overrides=true
    Use R3 Painted Overrides=true
    Use R4 Painted Overrides=true
    Use R5 Painted Overrides=true
   ) {module=VC_PaintedVariantsOverride001}
   &ParticleState()
   &InheritSourceMovement(
    Applied Position Scale=&Multiply_VectorByFloat(
     Vector=<1 1 1>
     Float=&FloatFromCurve(FloatCurve=DataInterface() {propertyValues {Curve {keys [{interpMode=RCIM_Cubic, tangentMode=RCTM_Break, time=0, value=1}, {interpMode=RCIM_Cubic, tangentMode=RCTM_Break, time=1, value=0}]}}})
    )
    ApplyForce=false
    ApplyPosition=true
    ApplyPositionFalloff=false
   ) {module=InheritSourceMovement001}
   &SolveForcesAndVelocity(Clamp Velocity=false, Limit Acceleration=false)
   &VC_UpdateRibbon_v1(Enable Skeletal Position=true)
   /SetVariables(Particles.DotProduct:NiagaraFloat=&DotProduct(
     Vector A=&TransformVectorByMatrix(
      Vector:Vector3f=<1 0 0>
      Transformation Matrix:NiagaraMatrix=$Engine.Owner.SystemLocalToWorld
      Transform Vector as Position:NiagaraBool=false
     )
     Vector B=&NormalizeVector(
      Vector To Normalize=&SubtractVectorFromPosition(
       Position=$Engine.Owner.Position
       Vector=&ReturnCameraPosition(Coordinate Space:ENiagaraCoordinateSpace=::Simulation, Camera Query:NiagaraDataInterfaceCamera=DataInterface())
      )
      Safe Normalize=true
      Safe Fallback Threshold=0
      Safe Fallback Vector=<0 0 0.001>
     )
    )) {isSetParameters=true, module=SetVariables_#3}
   /SetVariables(
    Particles.RibbonBase:LinearColor=&ColorFromCurve(ColorCurve=DataInterface() {propertyValues {
       AlphaCurve {keys [<at 0.277 0>,<at 0.4251 1>,<at 0.4872 2>,<at 0.5514 0>]}
       BlueCurve {keys [<at 0.9243 0>]}
       GreenCurve {keys [<at 0.9243 0>]}
       RedCurve {keys [<at 0.9243 0>]}
      }})
    Particles.RibbonGlow:LinearColor=&ColorFromCurve(ColorCurve=DataInterface() {propertyValues {
       AlphaCurve {keys [<at 0.2847 0>,<at 0.4022 1>,<at 0.5324 1>,<at 0.6291 0>]}
       BlueCurve {keys [<at 0.6966 2>]}
       GreenCurve {keys [<at 0.6966 0.5201>]}
       RedCurve {keys [<at 0.6966 0>]}
      }})
    Particles.RibbonGradient:LinearColor=&ColorFromCurve(ColorCurve=DataInterface() {propertyValues {
       AlphaCurve {keys [<at 0.3 0>,<at 0.4293 0.5>,<at 0.5035 0.5>,<at 0.557 0>]}
       BlueCurve {keys [<at 1 0.1683 RCIM_Cubic>]}
       GreenCurve {keys [<at 1 0 RCIM_Cubic>]}
       RedCurve {keys [<at 1 0.0865 RCIM_Cubic>]}
      }})
   ) {isSetParameters=true, module=SetVariables_#4}
   /SetVariables(
    Particles.R1_Color_Scaled:LinearColor=&MultiplyLinearColor(
     Linear Color=$Particles.R1 Color
     Scale Factor=&MakeVector4FromVectorAndFloat(
      "Vector (RGB)"=&VectorFromFloat(Value=&FloatFromCurve(
        FloatCurve=DataInterface() {propertyValues {Curve {keys [
            {arriveTangent=1.0866, interpMode=RCIM_Cubic, leaveTangent=6.1307, tangentMode=RCTM_Break, time=0.1791, value=0}
            {arriveTangent=7.4465, interpMode=RCIM_Cubic, tangentMode=RCTM_Break, time=0.264, value=0.9529}
            <at 0.5072 1>
            <at 0.62 0>
           ]}}}
        Scale Curve=&MakeCustomFloatFromBool(Bool=$User.bIsPainted, True Float=$User.Power_EmissivePaintedR4, False Float=2)
       ))
      "Float (Alpha)"=&FloatFromCurve(
       FloatCurve=DataInterface() {propertyValues {Curve {keys [
           {arriveTangent=1.0866, interpMode=RCIM_Cubic, leaveTangent=6.1307, tangentMode=RCTM_Break, time=0.1557, value=0}
           {arriveTangent=7.4465, interpMode=RCIM_Cubic, tangentMode=RCTM_Break, time=0.2455, value=0.9765}
           <at 0.5287 0.9647>
           <at 0.62 0>
          ]}}}
       Scale Curve=&MakeCustomFloatFromBool(Bool=$User.bIsPainted, True Float=$User.Power_AlphaPaintedR4, False Float=0.5)
      )
     )
    )
    Particles.R3_Color_Scaled:LinearColor=&MultiplyLinearColor(
     Linear Color=$Particles.R3 Color
     Scale Factor=&MakeVector4FromVectorAndFloat(
      "Vector (RGB)"=&VectorFromFloat(Value=&FloatFromCurve(
        FloatCurve=DataInterface() {propertyValues {Curve {keys [
            {arriveTangent=1.0866, interpMode=RCIM_Cubic, leaveTangent=6.1307, tangentMode=RCTM_Break, time=0.0826, value=0.0196}
            {arriveTangent=7.4465, interpMode=RCIM_Cubic, tangentMode=RCTM_Break, time=0.1611, value=0.9533}
            <at 0.5179 1>
            <at 0.62 0>
           ]}}}
        Scale Curve=&MakeCustomFloatFromBool(Bool=$User.bIsPainted, True Float=$User.Power_EmissivePaintedR4, False Float=2)
       ))
      "Float (Alpha)"=&FloatFromCurve(
       FloatCurve=DataInterface() {propertyValues {Curve {keys [
           {arriveTangent=1.0866, interpMode=RCIM_Cubic, leaveTangent=6.1307, tangentMode=RCTM_Break, time=0.1266, value=-0.0192}
           {arriveTangent=7.4465, interpMode=RCIM_Cubic, tangentMode=RCTM_Break, time=0.213, value=0.9638}
           <at 0.5107 0.9882>
           <at 0.62 0>
          ]}}}
       Scale Curve=&MakeCustomFloatFromBool(Bool=$User.bIsPainted, True Float=$User.Power_AlphaPaintedR4, False Float=1)
      )
     )
    )
    Particles.R4_Color_Scaled:LinearColor=&MultiplyLinearColor(
     Linear Color=$Particles.R4 Color
     Scale Factor=&MakeVector4FromVectorAndFloat(
      "Vector (RGB)"=&VectorFromFloat(Value=&FloatFromCurve(
        FloatCurve=DataInterface() {propertyValues {Curve {keys [
            {arriveTangent=1.0866, interpMode=RCIM_Cubic, leaveTangent=6.1307, tangentMode=RCTM_Break, time=0.28, value=0}
            {arriveTangent=7.4465, interpMode=RCIM_Cubic, tangentMode=RCTM_Break, time=0.3458, value=0.9882}
            <at 0.53 1>
            <at 0.62 0>
           ]}}}
        Scale Curve=&MakeCustomFloatFromBool(Bool=$User.bIsPainted, True Float=$User.Power_EmissivePaintedR4, False Float=0.6)
       ))
      "Float (Alpha)"=&FloatFromCurve(
       FloatCurve=DataInterface() {propertyValues {Curve {keys [
           <at -0.0014 1.7739 RCIM_Cubic>
           {arriveTangent=-1.9854, interpMode=RCIM_Cubic, leaveTangent=-1.9854, time=0.3044, value=2.258}
           {arriveTangent=-0.2383, interpMode=RCIM_Cubic, leaveTangent=-1.8935, tangentMode=RCTM_Break, time=0.7964, value=0.1899}
           <at 1.4148 -1.634 RCIM_Cubic>
          ]}}}
       Scale Curve=&MakeCustomFloatFromBool(Bool=$User.bIsPainted, True Float=$User.Power_AlphaPaintedR4, False Float=1.2)
      )
     )
    )
    Particles.R5_Color_Scaled:LinearColor=&MultiplyLinearColor(
     Linear Color=$Particles.R5 Color
     Scale Factor=&MakeVector4FromVectorAndFloat(
      "Vector (RGB)"=&VectorFromFloat(Value=&FloatFromCurve(
        FloatCurve=DataInterface() {propertyValues {Curve {keys [
            <at -0.17 16.4711 RCIM_Cubic>
            {arriveTangent=-27.6507, interpMode=RCIM_Cubic, leaveTangent=-27.6507, time=0.1072, value=17.8443}
            {arriveTangent=-3.2592, interpMode=RCIM_Cubic, leaveTangent=-3.2592, tangentMode=RCTM_User, time=0.3385, value=2.4113}
            {arriveTangent=2.5648, interpMode=RCIM_Cubic, leaveTangent=2.5648, tangentMode=RCTM_User, time=0.64, value=1.973}
            {arriveTangent=11.7336, interpMode=RCIM_Cubic, leaveTangent=11.7336, time=1.0714, value=20.2083}
            {arriveTangent=-34.897, interpMode=RCIM_Cubic, leaveTangent=-34.897, time=1.3482, value=10.2825}
            {arriveTangent=-8.1043, interpMode=RCIM_Cubic, leaveTangent=-8.1043, time=1.5789, value=2.4972}
            <at 1.7789 6.7918 RCIM_Cubic>
           ]}}}
        Scale Curve=&MakeCustomFloatFromBool(Bool=$User.bIsPainted, True Float=$User.Power_EmissivePaintedR4, False Float=0.7)
       ))
      "Float (Alpha)"=&FloatFromCurve(
       FloatCurve=DataInterface() {propertyValues {Curve {keys [
           <at -0.0556 3.887 RCIM_Cubic>
           {arriveTangent=-1.3925, interpMode=RCIM_Cubic, leaveTangent=-1.3925, time=0.111, value=0.7392}
           {arriveTangent=-0.2383, interpMode=RCIM_Cubic, leaveTangent=-0.2383, tangentMode=RCTM_User, time=0.4793, value=3.1423}
           {arriveTangent=-6.3349, interpMode=RCIM_Cubic, leaveTangent=-6.3349, time=0.8329, value=0.2261}
           <at 1.01 -0.2199 RCIM_Cubic>
          ]}}}
       Scale Curve=&MakeCustomFloatFromBool(Bool=$User.bIsPainted, True Float=$User.Power_AlphaPaintedR4, False Float=0.3)
      )
     )
    )
   ) {isSetParameters=true}
   &ScaleRibbonWidth(Ribbon Width Scale=&FloatFromCurve(FloatCurve=DataInterface() {propertyValues {Curve {keys [
         {arriveTangent=-0.2605, interpMode=RCIM_Cubic, leaveTangent=-0.2605, tangentMode=RCTM_User, time=0.0043, value=1.4579}
         {arriveTangent=-1.5709, interpMode=RCIM_Cubic, leaveTangent=-1.5709, tangentMode=RCTM_User, time=0.3137, value=1.1734}
         {arriveTangent=-1.5995, interpMode=RCIM_Cubic, leaveTangent=-1.5995, time=0.5043, value=0.5604}
         {arriveTangent=-1.126, interpMode=RCIM_Cubic, leaveTangent=-1.126, time=0.6024, value=0.7117}
         {arriveTangent=-1.0469, tangentMode=RCTM_User, time=1.002, value=0}
        ]}}})) {enabled=false, module=ScaleRibbonWidth001}
   &ScaleRibbonWidth(Ribbon Width Scale=&FloatFromCurve(FloatCurve=DataInterface() {propertyValues {Curve {keys [
         {arriveTangent=-0.2605, interpMode=RCIM_Cubic, leaveTangent=-0.2605, tangentMode=RCTM_User, time=0.0043, value=1.4579}
         {arriveTangent=-1.5709, interpMode=RCIM_Cubic, leaveTangent=-1.5709, tangentMode=RCTM_User, time=0.3137, value=1.1734}
         {arriveTangent=-2.0722, interpMode=RCIM_Cubic, leaveTangent=-2.0722, time=0.5043, value=0.5604}
         {arriveTangent=-1.126, interpMode=RCIM_Cubic, leaveTangent=-1.126, time=0.5945, value=0.5916}
         {arriveTangent=-1.0469, tangentMode=RCTM_User, time=1.002, value=0}
        ]}}}))
   /SetVariables(
    Particles.Ribbon_BlackColor:LinearColor=<rgba 0 0 0 1>
    Particles.WaveScaleCurve:NiagaraFloat=&FloatFromCurve(
     FloatCurve=DataInterface() {propertyValues {Curve {keys [
         {arriveTangent=39.5227, interpMode=RCIM_Cubic, leaveTangent=39.5227, tangentMode=RCTM_User, time=0.0092, value=0}
         {
          arriveTangent=0.7219
          arriveTangentWeight=0.0795
          interpMode=RCIM_Cubic
          leaveTangent=0.1806
          leaveTangentWeight=0.0696
          tangentMode=RCTM_Break
          tangentWeightMode=RCTWM_WeightedBoth
          time=0.125
          value=1.25
         }
         {arriveTangent=0.1988, interpMode=RCIM_Cubic, leaveTangent=0.1988, tangentMode=RCTM_User, time=0.45, value=0.625}
        ]}}}
     Scale Curve=10
    )
   ) {isSetParameters=true, module=SetVariables_#5}
   /SetVariables(
    Particles.CustomPosition_R1:NiagaraPosition=&AddVectorToPosition(
     Position=$Particles.Position
     Vector=&Multiply_VectorByFloat(
      Vector=&MakeVector(
       X=&Sine(
        Normalized Angle=&Add_Float(A=0, B=&Subtract_Float(A=$Emitter.NormalizedLoopAge, B=$Particles.NormalizedAge))
        Period=$User.WavePeriod
        Scale=&Multiply_Float(
         A=&SelectFloatFromArray(Direct Array Index=$Particles.SpawnGroup, Float Selection Array=DataInterface() {propertyValues {FloatData [1,-1]}})
         B=$User.WaveMaxScale
        )
        Bias=0
       )
       Y=&Sine(
        Normalized Angle=&Add_Float(A=0, B=&Subtract_Float(A=$Emitter.NormalizedLoopAge, B=$Particles.NormalizedAge))
        Period=$User.WavePeriod
        Scale=&Multiply_Float(
         A=&SelectFloatFromArray(Direct Array Index=$Particles.SpawnGroup, Float Selection Array=DataInterface() {propertyValues {FloatData [1,-1]}})
         B=$User.WaveMaxScale
        )
        Bias=0
       )
       Z=&Cosine(
        Normalized Angle=&Add_Float(A=0, B=&Subtract_Float(A=$Emitter.NormalizedLoopAge, B=$Particles.NormalizedAge))
        Period=$User.WavePeriod
        Scale=$User.WaveMaxScale
        Bias=0
       )
      )
      Float=$Particles.WaveScaleCurve
     )
    )
    Particles.CustomPosition_R2:NiagaraPosition=&AddVectorToPosition(
     Position=$Particles.Position
     Vector=&Multiply_VectorByFloat(
      Vector=&MakeVector(
       X=&Sine(
        Normalized Angle=&Add_Float(
         A=&Multiply_Float(A=$User.WavePeriod, B=0.5)
         B=&Subtract_Float(A=$Emitter.NormalizedLoopAge, B=$Particles.NormalizedAge)
        )
        Period=$User.WavePeriod
        Scale=&Multiply_Float(
         A=&SelectFloatFromArray(Direct Array Index=$Particles.SpawnGroup, Float Selection Array=DataInterface() {propertyValues {FloatData [1,-1]}})
         B=$User.WaveMaxScale
        )
        Bias=0
       )
       Y=&Sine(
        Normalized Angle=&Add_Float(
         A=&Multiply_Float(A=$User.WavePeriod, B=0.5)
         B=&Subtract_Float(A=$Emitter.NormalizedLoopAge, B=$Particles.NormalizedAge)
        )
        Period=$User.WavePeriod
        Scale=&Multiply_Float(
         A=&SelectFloatFromArray(Direct Array Index=$Particles.SpawnGroup, Float Selection Array=DataInterface() {propertyValues {FloatData [1,-1]}})
         B=$User.WaveMaxScale
        )
        Bias=0
       )
       Z=&Cosine(
        Normalized Angle=&Add_Float(
         A=&Multiply_Float(A=$User.WavePeriod, B=0.5)
         B=&Subtract_Float(A=$Emitter.NormalizedLoopAge, B=$Particles.NormalizedAge)
        )
        Period=$User.WavePeriod
        Scale=$User.WaveMaxScale
        Bias=0
       )
      )
      Float=$Particles.WaveScaleCurve
     )
    )
   ) {enabled=false, isSetParameters=true, module=SetVariables_#6}
   /SetVariables(
    Particles.CustomPosition_R1:NiagaraPosition=&AddVectorToPosition(
     Position=$Particles.Position
     Vector=&Multiply_VectorByFloat(
      Vector=&MakeVector(
       X=&Sine(
        Normalized Angle=&Add_Float(A=0, B=&Subtract_Float(A=$Emitter.NormalizedLoopAge, B=$Particles.NormalizedAge))
        Period=$User.WavePeriod
        Scale=&Multiply_Float(
         A=&SelectFloatFromArray(Direct Array Index=$Particles.SpawnGroup, Float Selection Array=DataInterface() {propertyValues {FloatData [1,-1]}})
         B=$User.WaveMaxScale
        )
        Bias=0
       )
       Y=&Sine(
        Normalized Angle=&Add_Float(A=0, B=&Subtract_Float(A=$Emitter.NormalizedLoopAge, B=$Particles.NormalizedAge))
        Period=$User.WavePeriod
        Scale=&Multiply_Float(
         A=&SelectFloatFromArray(Direct Array Index=$Particles.SpawnGroup, Float Selection Array=DataInterface() {propertyValues {FloatData [1,1]}})
         B=$User.WaveMaxScale
        )
        Bias=0
       )
       Z=&Cosine(
        Normalized Angle=&Add_Float(A=0, B=&Subtract_Float(A=$Emitter.NormalizedLoopAge, B=$Particles.NormalizedAge))
        Period=$User.WavePeriod
        Scale=&Multiply_Float(
         A=&SelectFloatFromArray(Direct Array Index=$Particles.SpawnGroup, Float Selection Array=DataInterface() {propertyValues {FloatData [-1,1]}})
         B=$User.WaveMaxScale
        )
        Bias=0
       )
      )
      Float=$Particles.WaveScaleCurve
     )
    )
    Particles.CustomPosition_R2:NiagaraPosition=&AddVectorToPosition(
     Position=$Particles.Position
     Vector=&Multiply_VectorByFloat(
      Vector=&MakeVector(
       X=&Sine(
        Normalized Angle=&Add_Float(
         A=&Multiply_Float(A=$User.WavePeriod, B=0.5)
         B=&Subtract_Float(A=$Emitter.NormalizedLoopAge, B=$Particles.NormalizedAge)
        )
        Period=$User.WavePeriod
        Scale=&Multiply_Float(
         A=&SelectFloatFromArray(Direct Array Index=$Particles.SpawnGroup, Float Selection Array=DataInterface() {propertyValues {FloatData [1,1]}})
         B=$User.WaveMaxScale
        )
        Bias=0
       )
       Y=&Sine(
        Normalized Angle=&Add_Float(
         A=&Multiply_Float(A=$User.WavePeriod, B=0.5)
         B=&Subtract_Float(A=$Emitter.NormalizedLoopAge, B=$Particles.NormalizedAge)
        )
        Period=$User.WavePeriod
        Scale=&Multiply_Float(
         A=&SelectFloatFromArray(Direct Array Index=$Particles.SpawnGroup, Float Selection Array=DataInterface() {propertyValues {FloatData [1,1]}})
         B=$User.WaveMaxScale
        )
        Bias=0
       )
       Z=&Cosine(
        Normalized Angle=&Add_Float(
         A=&Multiply_Float(A=$User.WavePeriod, B=0.5)
         B=&Subtract_Float(A=$Emitter.NormalizedLoopAge, B=$Particles.NormalizedAge)
        )
        Period=$User.WavePeriod
        Scale=&Multiply_Float(
         A=&SelectFloatFromArray(Direct Array Index=$Particles.SpawnGroup, Float Selection Array=DataInterface() {propertyValues {FloatData [-1,1]}})
         B=$User.WaveMaxScale
        )
        Bias=0
       )
      )
      Float=$Particles.WaveScaleCurve
     )
    )
   ) {isSetParameters=true}
   /SetVariables(Particles.DynamicMP_R2:Vector4f=&MakeVector4(
     X=&FloatFromCurve(FloatCurve=DataInterface() {propertyValues {Curve {
         defaultValue=1
         keys [
          {tangentMode=RCTM_User, time=0, value=0.5}
          {arriveTangent=-0.0633, interpMode=RCIM_Cubic, leaveTangent=-0.0633, tangentMode=RCTM_User, time=0.65, value=0.5}
          {arriveTangent=-0.0699, interpMode=RCIM_Cubic, leaveTangent=-0.0699, tangentMode=RCTM_User, time=0.75, value=-2}
         ]
        }}})
     Y=0
     Z=0
     W=0
    )) {isSetParameters=true, module=SetVariables_#7}
   &DynamicMaterialParameters(
    Index 0 Param 1=&FloatFromCurve(FloatCurve=DataInterface() {propertyValues {Curve {
        defaultValue=1
        keys [
         {
          arriveTangent=0.0405
          arriveTangentWeight=0.1835
          interpMode=RCIM_Cubic
          leaveTangent=0.0405
          leaveTangentWeight=0.1835
          tangentMode=RCTM_User
          time=0
          value=0
         }
         {
          arriveTangent=0.0201
          arriveTangentWeight=0.1834
          interpMode=RCIM_Cubic
          leaveTangent=0.0201
          leaveTangentWeight=0.1834
          tangentMode=RCTM_User
          time=0.5
          value=1
         }
        ]
       }}})
    Module.Index0Float0Write=true
    Module.Index0Float1Write=true
    Module.Index0Float2Write=true
    Module.Index0Float3Write=true
   )
  ]
  rendererClasses [&NiagaraRibbonRendererProperties]
  renderers [
   {rendererClass=&NiagaraRibbonRendererProperties}
   {rendererClass=&NiagaraRibbonRendererProperties, rendererIndex=1}
   {rendererClass=&NiagaraRibbonRendererProperties, rendererIndex=2}
   {rendererClass=&NiagaraRibbonRendererProperties, rendererIndex=3}
  ]
 }
}
summary {
 systemName=NS_BlackFlame_Experiment
 emitters [
  {emitterName=NE_BlacFlame_Aura, rendererClasses [&NiagaraRibbonRendererProperties]}
  {bEnabled=false, emitterName=NE_BlackFlame_Base1, rendererClasses [&NiagaraRibbonRendererProperties]}
  {bEnabled=false, emitterName=NE_BlackFlame_Base2, rendererClasses [&NiagaraRibbonRendererProperties]}
  {bEnabled=false, emitterName=NE_BlackFlame_Mid2, rendererClasses [&NiagaraRibbonRendererProperties]}
  {bEnabled=false, emitterName=NE_BlackFlame_Mid1, rendererClasses [&NiagaraRibbonRendererProperties]}
  {bEnabled=false, emitterName=NE_BlackFlame_Base1001, rendererClasses [&NiagaraRibbonRendererProperties]}
  {emitterName=NE_BlackFlame_Base1003, rendererClasses [&NiagaraRibbonRendererProperties]}
  {emitterName=NE_BlackFlame_Base2001, rendererClasses [&NiagaraRibbonRendererProperties]}
 ]
 userVariables [
  User.bIsFrontendPreview {type=NiagaraBool, defaultValue=true}
  User.bIsPainted {type=NiagaraBool}
  User.ColorVariantID {type=NiagaraInt32}
  User.DTCarOffset {type=Vector3f, description="", defaultValue {struct=Vector3f, value=<0 0 0>}}
  User.GameModeID {type=NiagaraInt32}
  User.PaintedColor {type=LinearColor, description="", defaultValue {struct=LinearColor, value=<rgba 1 1 1 1>}}
  User.PaintedScalar {type=NiagaraFloat, defaultValue=1, description=""}
  User.PortCount {type=NiagaraInt32, defaultValue=2}
  User.Power_AlphaPaintedR4 {type=NiagaraFloat, defaultValue=1, description=""}
  User.Power_EmissivePaintedR4 {type=NiagaraFloat, defaultValue=1, description=""}
  User.SingleBoostPos {type=Vector3f, description="", defaultValue {struct=Vector3f, value=<0 0 0>}}
  User.WaveMaxScale {type=NiagaraFloat, defaultValue=1, description=""}
  User.WavePeriod {type=NiagaraFloat, defaultValue=1, description=""}
 ]
}
